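# Comic panel 4 widget: two side-by-side animated frames, each showing layered Synth artwork with hands, powers and sparks.
# One animation per frame slides in the background and character, slowly rotates and scales the layers, then throws the left frame up and the right down.
# Nested CanvasPanel controllers act as transform pivots.
# The event graph only holds empty events, and a translucent Scene_07 reference screenshot sits on top.
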WidgetBlueprint /PistolCanary/Layout/ComicPanels_v2/Blueprints/Panel_04_v03
animations [
 P4_L_Frame {
  binding_count=8
  length_seconds=7.6167
  object=&self:P4_L_Frame
  sequence {
   bindings [
    L_BG_Controller {
     id=#1
     possessed_class=CanvasPanel
     tracks [{
       type=MovieScene2DTransformTrack
       sections [{
         type=MovieScene2DTransformSection
         end_frame=15
         start_frame=0
         channels [
          Translation.X {type=MovieSceneScriptingFloatChannel, keys [[0,-607.2714],[15,117.2038]]}
          Translation.Y {type=MovieSceneScriptingFloatChannel, keys [[0,1146.4131],[15,24.4778]]}
          Angle {type=MovieSceneScriptingFloatChannel, default=0}
          Scale.X {type=MovieSceneScriptingFloatChannel, default=2, const [2.2,1]}
          Scale.Y {type=MovieSceneScriptingFloatChannel, default=2, const [2.2,1]}
          Shear.X {type=MovieSceneScriptingFloatChannel, default=0}
          Shear.Y {type=MovieSceneScriptingFloatChannel, default=0}
         ]
        }]
      }]
    }
    L_Synth_R_Power_Controller {
     id=#2
     possessed_class=CanvasPanel
     tracks [{
       type=MovieScene2DTransformTrack
       sections [{
         type=MovieScene2DTransformSection
         end_frame=457
         start_frame=0
         channels [
          Translation.X {type=MovieSceneScriptingFloatChannel, const [0,3]}
          Translation.Y {type=MovieSceneScriptingFloatChannel, const [0,3]}
          Angle {type=MovieSceneScriptingFloatChannel, keys [[0,0],[457,-10]]}
          Scale.X {type=MovieSceneScriptingFloatChannel, keys [[15.474,0],[36.106,1],[457,1.2]]}
          Scale.Y {type=MovieSceneScriptingFloatChannel, keys [[15.474,0],[36.106,1],[457,1.2]]}
          Shear.X {type=MovieSceneScriptingFloatChannel, default=0}
          Shear.Y {type=MovieSceneScriptingFloatChannel, default=0}
         ]
        }]
      }]
    }
    L_Frame {
     id=#3
     possessed_class=CanvasPanel
     tracks [{
       type=MovieScene2DTransformTrack
       sections [{
         type=MovieScene2DTransformSection
         end_frame=457
         start_frame=0
         channels [
          Translation.X {type=MovieSceneScriptingFloatChannel, default=0}
          Translation.Y {type=MovieSceneScriptingFloatChannel, keys [[446.978,0],[457,-1367.4224]]}
          Angle {type=MovieSceneScriptingFloatChannel, default=0}
          Scale.X {type=MovieSceneScriptingFloatChannel, default=1}
          Scale.Y {type=MovieSceneScriptingFloatChannel, default=1}
          Shear.X {type=MovieSceneScriptingFloatChannel, default=0}
          Shear.Y {type=MovieSceneScriptingFloatChannel, default=0}
         ]
        }]
      }]
    }
    L_Synth_L_Power_Controller {
     id=#4
     possessed_class=CanvasPanel
     tracks [{
       type=MovieScene2DTransformTrack
       sections [{
         type=MovieScene2DTransformSection
         end_frame=457
         start_frame=0
         channels [
          Translation.X {type=MovieSceneScriptingFloatChannel, const [0,3]}
          Translation.Y {type=MovieSceneScriptingFloatChannel, const [0,3]}
          Angle {type=MovieSceneScriptingFloatChannel, keys [[0,0],[457,10]]}
          Scale.X {type=MovieSceneScriptingFloatChannel, keys [[15.474,0],[36.106,1.1],[457,1.2]]}
          Scale.Y {type=MovieSceneScriptingFloatChannel, keys [[15.474,0],[36.106,1.1],[457,1.2]]}
          Shear.X {type=MovieSceneScriptingFloatChannel, default=0}
          Shear.Y {type=MovieSceneScriptingFloatChannel, default=0}
         ]
        }]
      }]
    }
    L_Synth_L_Hand_Controller {
     id=#5
     possessed_class=CanvasPanel
     tracks [{
       type=MovieScene2DTransformTrack
       sections [{
         type=MovieScene2DTransformSection
         end_frame=457
         start_frame=0
         channels [
          Translation.X {type=MovieSceneScriptingFloatChannel, default=0}
          Translation.Y {type=MovieSceneScriptingFloatChannel, default=0}
          Angle {type=MovieSceneScriptingFloatChannel, keys [[0,0],[457,18.7271]]}
          Scale.X {type=MovieSceneScriptingFloatChannel, default=1}
          Scale.Y {type=MovieSceneScriptingFloatChannel, default=1}
          Shear.X {type=MovieSceneScriptingFloatChannel, default=0}
          Shear.Y {type=MovieSceneScriptingFloatChannel, default=0}
         ]
        }]
      }]
    }
    L_Synth {
     id=#6
     possessed_class=CanvasPanel
     tracks [{
       type=MovieScene2DTransformTrack
       sections [{
         type=MovieScene2DTransformSection
         end_frame=457
         start_frame=0
         channels [
          Translation.X {type=MovieSceneScriptingFloatChannel, keys [[0,-707.7817],[15.474,-707.7817],[36.106,0],[457,40.3339]]}
          Translation.Y {type=MovieSceneScriptingFloatChannel, keys [[0,-388.8787],[15.474,-388.8787],[36.106,0],[457,-20.2096]]}
          Angle {type=MovieSceneScriptingFloatChannel, keys [[0,20.9972],[15.474,20.9972],[36.106,0],[457,-3]]}
          Scale.X {type=MovieSceneScriptingFloatChannel, keys [[0,0.2],[15.474,0.2],[36.106,1],[457,1.1]]}
          Scale.Y {type=MovieSceneScriptingFloatChannel, keys [[0,0.2],[15.474,0.2],[36.106,1],[457,1.1]]}
          Shear.X {type=MovieSceneScriptingFloatChannel, const [0,1]}
          Shear.Y {type=MovieSceneScriptingFloatChannel, const [0,1]}
         ]
        }]
      }]
    }
    L_Sparks_Controller {
     id=#7
     possessed_class=CanvasPanel
     tracks [{
       type=MovieScene2DTransformTrack
       sections [{
         type=MovieScene2DTransformSection
         end_frame=457
         start_frame=0
         channels [
          Translation.X {type=MovieSceneScriptingFloatChannel, default=0}
          Translation.Y {type=MovieSceneScriptingFloatChannel, default=0}
          Angle {type=MovieSceneScriptingFloatChannel, default=0}
          Scale.X {type=MovieSceneScriptingFloatChannel, keys [[0,1],[457,1.2]]}
          Scale.Y {type=MovieSceneScriptingFloatChannel, keys [[0,1],[457,1.2]]}
          Shear.X {type=MovieSceneScriptingFloatChannel, default=0}
          Shear.Y {type=MovieSceneScriptingFloatChannel, default=0}
         ]
        }]
      }]
    }
    L_Synth_R_Hand_Controller {
     id=#8
     possessed_class=CanvasPanel
     tracks [{
       type=MovieScene2DTransformTrack
       sections [{
         type=MovieScene2DTransformSection
         end_frame=457
         start_frame=0
         channels [
          Translation.X {type=MovieSceneScriptingFloatChannel, default=329.5019}
          Translation.Y {type=MovieSceneScriptingFloatChannel, default=0}
          Angle {type=MovieSceneScriptingFloatChannel, keys [[0,0],[457,-15.633]]}
          Scale.X {type=MovieSceneScriptingFloatChannel, default=1}
          Scale.Y {type=MovieSceneScriptingFloatChannel, default=1}
          Shear.X {type=MovieSceneScriptingFloatChannel, default=0}
          Shear.Y {type=MovieSceneScriptingFloatChannel, default=0}
         ]
        }]
      }]
    }
   ]
   display_rate {numerator=60}
  }
 }
 P4_R_Frame {
  binding_count=8
  length_seconds=6.2333
  object=&self:P4_R_Frame
  sequence {
   bindings [
    R_Embers {
     id=#9
     possessed_class=Image
     tracks [{
       type=MovieScene2DTransformTrack
       sections [{
         type=MovieScene2DTransformSection
         end_frame=374
         start_frame=0
         channels [
          Translation.X {type=MovieSceneScriptingFloatChannel, keys [[0,788.252],[12.303,788.252],[24.605,0],[374,-66.0925]]}
          Translation.Y {type=MovieSceneScriptingFloatChannel, keys [[0,152.7908],[12.303,152.7908],[24.605,0],[374,0]]}
          Angle {type=MovieSceneScriptingFloatChannel, default=0, keys [[0,9.4774],[12.303,9.4774],[24.605,0],[374,0]]}
          Scale.X {type=MovieSceneScriptingFloatChannel, keys [[0,0.8],[12.303,0.8],[24.605,1],[374,1.2]]}
          Scale.Y {type=MovieSceneScriptingFloatChannel, keys [[0,0.8],[12.303,0.8],[24.605,1],[374,1]]}
          Shear.X {type=MovieSceneScriptingFloatChannel, default=0}
          Shear.Y {type=MovieSceneScriptingFloatChannel, default=0}
         ]
        }]
      }]
    }
    R_R_Hand_Controller {
     id=#10
     possessed_class=CanvasPanel
     tracks [{
       type=MovieScene2DTransformTrack
       sections [{
         type=MovieScene2DTransformSection
         end_frame=374
         start_frame=0
         channels [
          Translation.X {type=MovieSceneScriptingFloatChannel, default=0}
          Translation.Y {type=MovieSceneScriptingFloatChannel, default=0}
          Angle {type=MovieSceneScriptingFloatChannel, keys [[0,0],[374,20]]}
          Scale.X {type=MovieSceneScriptingFloatChannel, default=1}
          Scale.Y {type=MovieSceneScriptingFloatChannel, default=1}
          Shear.X {type=MovieSceneScriptingFloatChannel, default=0}
          Shear.Y {type=MovieSceneScriptingFloatChannel, default=0}
         ]
        }]
      }]
    }
    R_Powers {
     id=#11
     possessed_class=Image
     tracks [{
       type=MovieScene2DTransformTrack
       sections [{
         type=MovieScene2DTransformSection
         end_frame=374
         start_frame=0
         channels [
          Translation.X {type=MovieSceneScriptingFloatChannel, default=0}
          Translation.Y {type=MovieSceneScriptingFloatChannel, default=0}
          Angle {type=MovieSceneScriptingFloatChannel, default=0}
          Scale.X {type=MovieSceneScriptingFloatChannel, keys [[0,1],[374,1.2]]}
          Scale.Y {type=MovieSceneScriptingFloatChannel, keys [[0,1],[374,1.2]]}
          Shear.X {type=MovieSceneScriptingFloatChannel, default=0}
          Shear.Y {type=MovieSceneScriptingFloatChannel, default=0}
         ]
        }]
      }]
    }
    R_BG_Controller {
     id=#12
     possessed_class=CanvasPanel
     tracks [{
       type=MovieScene2DTransformTrack
       sections [{
         type=MovieScene2DTransformSection
         end_frame=15
         start_frame=0
         channels [
          Translation.X {type=MovieSceneScriptingFloatChannel, keys [[0,157.2426],[15,-19.7455]]}
          Translation.Y {type=MovieSceneScriptingFloatChannel, keys [[0,-1213.5093],[15,24.6162]]}
          Angle {type=MovieSceneScriptingFloatChannel, default=0}
          Scale.X {type=MovieSceneScriptingFloatChannel, default=2, const [2.1,1]}
          Scale.Y {type=MovieSceneScriptingFloatChannel, default=2, const [2.1,1]}
          Shear.X {type=MovieSceneScriptingFloatChannel, default=0}
          Shear.Y {type=MovieSceneScriptingFloatChannel, default=0}
         ]
        }]
      }]
    }
    R_Synth {
     id=#13
     possessed_class=CanvasPanel
     tracks [{
       type=MovieScene2DTransformTrack
       sections [{
         type=MovieScene2DTransformSection
         end_frame=374
         start_frame=0
         channels [
          Translation.X {type=MovieSceneScriptingFloatChannel, keys [[0,772.779],[12.303,772.779],[24.605,0],[374,-38.2837]]}
          Translation.Y {type=MovieSceneScriptingFloatChannel, keys [[0,421.278],[12.303,421.278],[24.605,0],[374,-11.5701]]}
          Angle {type=MovieSceneScriptingFloatChannel, keys [[0,11.8808],[12.303,11.8808],[24.605,0],[374,-3.3526]]}
          Scale.X {type=MovieSceneScriptingFloatChannel, keys [[0,0.8],[12.303,0.8],[24.605,1],[374,1]]}
          Scale.Y {type=MovieSceneScriptingFloatChannel, keys [[0,0.8],[12.303,0.8],[24.605,1],[374,1]]}
          Shear.X {type=MovieSceneScriptingFloatChannel, default=0}
          Shear.Y {type=MovieSceneScriptingFloatChannel, default=0}
         ]
        }]
      }]
    }
    R_Sparks_Controller {
     id=#14
     possessed_class=CanvasPanel
     tracks [{
       type=MovieScene2DTransformTrack
       sections [{
         type=MovieScene2DTransformSection
         end_frame=374
         start_frame=0
         channels [
          Translation.X {type=MovieSceneScriptingFloatChannel, default=0}
          Translation.Y {type=MovieSceneScriptingFloatChannel, default=0}
          Angle {type=MovieSceneScriptingFloatChannel, default=0}
          Scale.X {type=MovieSceneScriptingFloatChannel, keys [[0,1],[374,1.3]]}
          Scale.Y {type=MovieSceneScriptingFloatChannel, keys [[0,1],[374,1.3]]}
          Shear.X {type=MovieSceneScriptingFloatChannel, default=0}
          Shear.Y {type=MovieSceneScriptingFloatChannel, default=0}
         ]
        }]
      }]
    }
    R_Frame {
     id=#15
     possessed_class=CanvasPanel
     tracks [{
       type=MovieScene2DTransformTrack
       sections [{
         type=MovieScene2DTransformSection
         end_frame=374
         start_frame=0
         channels [
          Translation.X {type=MovieSceneScriptingFloatChannel, const [0,3]}
          Translation.Y {type=MovieSceneScriptingFloatChannel, keys [[0,0],[365.798,0],[374,1285.5883]]}
          Angle {type=MovieSceneScriptingFloatChannel, default=0}
          Scale.X {type=MovieSceneScriptingFloatChannel, default=1}
          Scale.Y {type=MovieSceneScriptingFloatChannel, default=1}
          Shear.X {type=MovieSceneScriptingFloatChannel, default=0}
          Shear.Y {type=MovieSceneScriptingFloatChannel, default=0}
         ]
        }]
      }]
    }
    R_L_Hand_Controller {
     id=#16
     possessed_class=CanvasPanel
     tracks [{
       type=MovieScene2DTransformTrack
       sections [{
         type=MovieScene2DTransformSection
         end_frame=374
         start_frame=0
         channels [
          Translation.X {type=MovieSceneScriptingFloatChannel, default=0}
          Translation.Y {type=MovieSceneScriptingFloatChannel, default=0}
          Angle {type=MovieSceneScriptingFloatChannel, keys [[0,0],[374,20]]}
          Scale.X {type=MovieSceneScriptingFloatChannel, default=1}
          Scale.Y {type=MovieSceneScriptingFloatChannel, default=1}
          Shear.X {type=MovieSceneScriptingFloatChannel, default=0}
          Shear.Y {type=MovieSceneScriptingFloatChannel, default=0}
         ]
        }]
      }]
    }
   ]
   display_rate {numerator=60}
  }
 }
]
graphs [EventGraph {nodes [
   Event PreConstruct()
   Event Construct()
   Event Tick()
  ]}]
variables []
widgets {
 info {inheritedWidgetCount=0, parentClass=&UserWidget, rootWidgetClass=&CanvasPanel, widgetCount=32}
 widgets [
  Panel_04 {widgetClassPath=&CanvasPanel, widget=Panel_04, properties {clipping=ClipToBounds, pixelSnapping=Disabled}}
  L_Frame {
   widgetClassPath=&CanvasPanel
   parent=Panel_04
   widget=L_Frame
   slot {class=CanvasPanelSlot, properties {bAutoSize=true, layoutData=~}}
  }
  L_BG_Controller {
   widgetClassPath=&CanvasPanel
   parent=L_Frame
   widget=L_BG_Controller
   properties {renderTransform {angle=0, scale=<2 2>, shear=<0 0>, translation=<0 0>}}
   slot {
    class=CanvasPanelSlot
    properties {bAutoSize=true, layoutData {alignment=<0 0>, anchors {maximum=<0 0>, minimum=<0 0>}, offsets=<ltrb 48 284 100 30>}}
   }
  }
  L_BG_50 {
   widgetClassPath=&Image
   bIsVariable=true
   parent=L_BG_Controller
   widget=L_BG_50
   properties {brush {
     bIsDynamicallyLoaded=false
     drawAs=Image
     imageType=FullColor
     mirroring=NoMirror
     resourceName=None
     resourceObject=/PistolCanary/Layout/ComicPanels_v2/ComicPanelTextures/Panel4/L_Frame/L_P4_BG_50percent
     tiling=NoTile
     imageSize=<516 564>
     margin=<ltrb 0 0 0 0>
     outlineSettings {
      bUseBrushTransparency=false
      roundingType=HalfHeightRadius
      width=0
      color {colorUseRule=UseColor_Specified, specifiedColor=<rgba 0 0 0 0>}
      cornerRadii=<0 0 0 0>
     }
     tintColor {colorUseRule=UseColor_Specified, specifiedColor=<rgba 1 1 1 1>}
     uVRegion {bIsValid=false, max=<0 0>, min=<0 0>}
    }}
   slot {class=CanvasPanelSlot, properties {layoutData {alignment=<0 0>, anchors {maximum=<0 0>, minimum=<0 0>}, offsets=<ltrb 0 0 320 564>}}}
  }
  L_Synth {
   widgetClassPath=&CanvasPanel
   parent=L_Frame
   widget=L_Synth
   slot {
    class=CanvasPanelSlot
    properties {bAutoSize=true, layoutData {alignment=<0 0>, anchors {maximum=<0 0>, minimum=<0 0>}, offsets=<ltrb 0 0 513.7143 1098.5715>}}
   }
  }
  L_Synth_Body {
   widgetClassPath=&Image
   bIsVariable=true
   parent=L_Synth
   widget=L_Synth_Body
   properties {brush {
     bIsDynamicallyLoaded=false
     drawAs=Image
     imageType=FullColor
     mirroring=NoMirror
     resourceName=None
     resourceObject=/PistolCanary/Layout/ComicPanels_v2/ComicPanelTextures/Panel4/L_Frame/L_P4_Synth_Body
     tiling=NoTile
     imageSize=<1032 1128>
     margin=<ltrb 0 0 0 0>
     outlineSettings {
      bUseBrushTransparency=false
      roundingType=HalfHeightRadius
      width=0
      color {colorUseRule=UseColor_Specified, specifiedColor=<rgba 0 0 0 0>}
      cornerRadii=<0 0 0 0>
     }
     tintColor {colorUseRule=UseColor_Specified, specifiedColor=<rgba 1 1 1 1>}
     uVRegion {bIsValid=false, max=<0 0>, min=<0 0>}
    }}
   slot {
    class=CanvasPanelSlot
    properties {zOrder=1, layoutData {alignment=<0 0>, anchors {maximum=<0 0>, minimum=<0 0>}, offsets=<ltrb -80 196 548 672>}}
   }
  }
  L_Synth_L_Hand_Controller {
   widgetClassPath=&CanvasPanel
   parent=L_Synth
   widget=L_Synth_L_Hand_Controller
   slot {
    class=CanvasPanelSlot
    properties {zOrder=6, layoutData {alignment=<0 0>, anchors {maximum=<0 0>, minimum=<0 0>}, offsets=<ltrb 220 104 289.7143 351.1429>}}
   }
  }
  L_Synth_L_Hand {
   widgetClassPath=&Image
   bIsVariable=true
   parent=L_Synth_L_Hand_Controller
   widget=L_Synth_L_Hand
   properties {brush {
     bIsDynamicallyLoaded=false
     drawAs=Image
     imageType=FullColor
     mirroring=NoMirror
     resourceName=None
     resourceObject=/PistolCanary/Layout/ComicPanels_v2/ComicPanelTextures/Panel4/L_Frame/L_P4_L_Hand
     tiling=NoTile
     imageSize=<1032 1128>
     margin=<ltrb 0 0 0 0>
     outlineSettings {
      bUseBrushTransparency=false
      roundingType=HalfHeightRadius
      width=0
      color {colorUseRule=UseColor_Specified, specifiedColor=<rgba 0 0 0 0>}
      cornerRadii=<0 0 0 0>
     }
     tintColor {colorUseRule=UseColor_Specified, specifiedColor=<rgba 1 1 1 1>}
     uVRegion {bIsValid=false, max=<0 0>, min=<0 0>}
    }}
   slot {
    class=CanvasPanelSlot
    properties {zOrder=6, layoutData {alignment=<0 0>, anchors {maximum=<0.5 0.5>, minimum=<0.5 0.5>}, offsets=<ltrb -128.8571 -171.5714 284 356>}}
   }
  }
  L_Synth_R_Hand_Controller {
   widgetClassPath=&CanvasPanel
   parent=L_Synth
   widget=L_Synth_R_Hand_Controller
   slot {
    class=CanvasPanelSlot
    properties {zOrder=6, layoutData {alignment=<0 0>, anchors {maximum=<0.5 0.5>, minimum=<0.5 0.5>}, offsets=<ltrb -626 269 217.2245 197.5102>}}
   }
  }
  L_Synth_R_Hand {
   widgetClassPath=&Image
   bIsVariable=true
   parent=L_Synth_R_Hand_Controller
   widget=L_Synth_R_Hand
   properties {brush {
     bIsDynamicallyLoaded=false
     drawAs=Image
     imageType=FullColor
     mirroring=NoMirror
     resourceName=None
     resourceObject=/PistolCanary/Layout/ComicPanels_v2/ComicPanelTextures/Panel4/L_Frame/L_P4_R_Hand
     tiling=NoTile
     imageSize=<1032 1128>
     margin=<ltrb 0 0 0 0>
     outlineSettings {
      bUseBrushTransparency=false
      roundingType=HalfHeightRadius
      width=0
      color {colorUseRule=UseColor_Specified, specifiedColor=<rgba 0 0 0 0>}
      cornerRadii=<0 0 0 0>
     }
     tintColor {colorUseRule=UseColor_Specified, specifiedColor=<rgba 1 1 1 1>}
     uVRegion {bIsValid=false, max=<0 0>, min=<0 0>}
    }}
   slot {
    class=CanvasPanelSlot
    properties {zOrder=6, layoutData {alignment=<0 0>, anchors {maximum=<0 0>, minimum=<0 0>}, offsets=<ltrb -12 -8 224 204>}}
   }
  }
  L_Synth_L_Power_Controller {
   widgetClassPath=&CanvasPanel
   parent=L_Synth
   widget=L_Synth_L_Power_Controller
   slot {
    class=CanvasPanelSlot
    properties {zOrder=5, layoutData {alignment=<0 0>, anchors {maximum=<0 0>, minimum=<0 0>}, offsets=<ltrb 232 112 289.7143 351.1429>}}
   }
  }
  L_S_L_Power {
   widgetClassPath=&Image
   bIsVariable=true
   parent=L_Synth_L_Power_Controller
   widget=L_S_L_Power
   properties {brush {
     bIsDynamicallyLoaded=false
     drawAs=Image
     imageType=FullColor
     mirroring=NoMirror
     resourceName=None
     resourceObject=/PistolCanary/Layout/ComicPanels_v2/ComicPanelTextures/Panel4/L_Frame/L_P4_L_Powers
     tiling=NoTile
     imageSize=<1032 1128>
     margin=<ltrb 0 0 0 0>
     outlineSettings {
      bUseBrushTransparency=false
      roundingType=HalfHeightRadius
      width=0
      color {colorUseRule=UseColor_Specified, specifiedColor=<rgba 0 0 0 0>}
      cornerRadii=<0 0 0 0>
     }
     tintColor {colorUseRule=UseColor_Specified, specifiedColor=<rgba 1 1 1 1>}
     uVRegion {bIsValid=false, max=<0 0>, min=<0 0>}
    }}
   slot {
    class=CanvasPanelSlot
    properties {zOrder=5, layoutData {alignment=<0 0>, anchors {maximum=<0 0>, minimum=<0 0>}, offsets=<ltrb -148 -112 620 536>}}
   }
  }
  L_Synth_R_Power_Controller {
   widgetClassPath=&CanvasPanel
   parent=L_Synth
   widget=L_Synth_R_Power_Controller
   slot {
    class=CanvasPanelSlot
    properties {zOrder=5, layoutData {alignment=<0 0>, anchors {maximum=<0 0>, minimum=<0 0>}, offsets=<ltrb 16 730.2857 215.4286 195.7143>}}
   }
  }
  L_S_R_Power {
   widgetClassPath=&Image
   bIsVariable=true
   parent=L_Synth_R_Power_Controller
   widget=L_S_R_Power
   properties {brush {
     bIsDynamicallyLoaded=false
     drawAs=Image
     imageType=FullColor
     mirroring=NoMirror
     resourceName=None
     resourceObject=/PistolCanary/Layout/ComicPanels_v2/ComicPanelTextures/Panel4/L_Frame/L_P4_R_Powers
     tiling=NoTile
     imageSize=<1032 1128>
     margin=<ltrb 0 0 0 0>
     outlineSettings {
      bUseBrushTransparency=false
      roundingType=HalfHeightRadius
      width=0
      color {colorUseRule=UseColor_Specified, specifiedColor=<rgba 0 0 0 0>}
      cornerRadii=<0 0 0 0>
     }
     tintColor {colorUseRule=UseColor_Specified, specifiedColor=<rgba 1 1 1 1>}
     uVRegion {bIsValid=false, max=<0 0>, min=<0 0>}
    }}
   slot {
    class=CanvasPanelSlot
    properties {zOrder=5, layoutData {alignment=<0 0>, anchors {maximum=<0 0>, minimum=<0 0>}, offsets=<ltrb -184 -156 540 540>}}
   }
  }
  L_Sparks_Controller {
   widgetClassPath=&CanvasPanel
   parent=L_Synth
   widget=L_Sparks_Controller
   slot {
    class=CanvasPanelSlot
    properties {layoutData {alignment=<0 0>, anchors {maximum=<0 0>, minimum=<0 0>}, offsets=<ltrb 196 508 101.1429 90.5714>}}
   }
  }
  L_T_Sparks {
   widgetClassPath=&Image
   bIsVariable=true
   parent=L_Sparks_Controller
   widget=L_T_Sparks
   properties {brush {
     bIsDynamicallyLoaded=false
     drawAs=Image
     imageType=FullColor
     mirroring=NoMirror
     resourceName=None
     resourceObject=/PistolCanary/Layout/ComicPanels_v2/ComicPanelTextures/Panel4/L_Frame/L_P4_T_Sparks
     tiling=NoTile
     imageSize=<844 344>
     margin=<ltrb 0 0 0 0>
     outlineSettings {
      bUseBrushTransparency=false
      roundingType=HalfHeightRadius
      width=0
      color {colorUseRule=UseColor_Specified, specifiedColor=<rgba 0 0 0 0>}
      cornerRadii=<0 0 0 0>
     }
     tintColor {colorUseRule=UseColor_Specified, specifiedColor=<rgba 1 1 1 1>}
     uVRegion {bIsValid=false, max=<0 0>, min=<0 0>}
    }}
   slot {
    class=CanvasPanelSlot
    properties {bAutoSize=true, layoutData {alignment=<0 0>, anchors {maximum=<0 0>, minimum=<0 0>}, offsets=<ltrb -216 -508 100 30>}}
   }
  }
  L_B_Sparks {
   widgetClassPath=&Image
   bIsVariable=true
   parent=L_Sparks_Controller
   widget=L_B_Sparks
   properties {brush {
     bIsDynamicallyLoaded=false
     drawAs=Image
     imageType=FullColor
     mirroring=NoMirror
     resourceName=None
     resourceObject=/PistolCanary/Layout/ComicPanels_v2/ComicPanelTextures/Panel4/L_Frame/L_P4_B_Sparks
     tiling=NoTile
     imageSize=<648 508>
     margin=<ltrb 0 0 0 0>
     outlineSettings {
      bUseBrushTransparency=false
      roundingType=HalfHeightRadius
      width=0
      color {colorUseRule=UseColor_Specified, specifiedColor=<rgba 0 0 0 0>}
      cornerRadii=<0 0 0 0>
     }
     tintColor {colorUseRule=UseColor_Specified, specifiedColor=<rgba 1 1 1 1>}
     uVRegion {bIsValid=false, max=<0 0>, min=<0 0>}
    }}
   slot {
    class=CanvasPanelSlot
    properties {bAutoSize=true, layoutData {alignment=<0 0>, anchors {maximum=<0 0>, minimum=<0 0>}, offsets=<ltrb -408 112 100 30>}}
   }
  }
  R_Frame {
   widgetClassPath=&CanvasPanel
   parent=Panel_04
   widget=R_Frame
   slot {class=CanvasPanelSlot, properties {bAutoSize=true, layoutData=~}}
  }
  R_BG_Controller {
   widgetClassPath=&CanvasPanel
   parent=R_Frame
   widget=R_BG_Controller
   properties {renderTransform {angle=0, scale=<2.1 2.1>, shear=<0 0>, translation=<17.2701 4.7448>}}
   slot {
    class=CanvasPanelSlot
    properties {bAutoSize=true, layoutData {alignment=<0 0>, anchors {maximum=<0 0>, minimum=<0 0>}, offsets=<ltrb 1376 280 100 30>}}
   }
  }
  R_BG_50 {
   widgetClassPath=&Image
   bIsVariable=true
   parent=R_BG_Controller
   widget=R_BG_50
   properties {
    brush {
     bIsDynamicallyLoaded=false
     drawAs=Image
     imageType=FullColor
     mirroring=NoMirror
     resourceName=None
     resourceObject=/PistolCanary/Layout/ComicPanels_v2/ComicPanelTextures/Panel4/R_Frame/R_P4_BG_50percent
     tiling=NoTile
     imageSize=<444 564>
     margin=<ltrb 0 0 0 0>
     outlineSettings {
      bUseBrushTransparency=false
      roundingType=HalfHeightRadius
      width=0
      color {colorUseRule=UseColor_Specified, specifiedColor=<rgba 0 0 0 0>}
      cornerRadii=<0 0 0 0>
     }
     tintColor {colorUseRule=UseColor_Specified, specifiedColor=<rgba 1 1 1 1>}
     uVRegion {bIsValid=false, max=<0 0>, min=<0 0>}
    }
    renderTransform {angle=0, scale=<1.1 1.1>, shear=<0 0>, translation=<0 0>}
   }
   slot {class=CanvasPanelSlot, properties {layoutData {alignment=<0 0>, anchors {maximum=<0 0>, minimum=<0 0>}, offsets=<ltrb 0 0 352 564>}}}
  }
  R_Embers {
   widgetClassPath=&Image
   bIsVariable=true
   parent=R_Frame
   widget=R_Embers
   properties {brush {
     bIsDynamicallyLoaded=false
     drawAs=Image
     imageType=FullColor
     mirroring=NoMirror
     resourceName=None
     resourceObject=/PistolCanary/Layout/ComicPanels_v2/ComicPanelTextures/Panel4/R_Frame/R_P4_Synth_Embers_50percent
     tiling=NoTile
     imageSize=<444 236>
     margin=<ltrb 0 0 0 0>
     outlineSettings {
      bUseBrushTransparency=false
      roundingType=HalfHeightRadius
      width=0
      color {colorUseRule=UseColor_Specified, specifiedColor=<rgba 0 0 0 0>}
      cornerRadii=<0 0 0 0>
     }
     tintColor {colorUseRule=UseColor_Specified, specifiedColor=<rgba 1 1 1 1>}
     uVRegion {bIsValid=false, max=<0 0>, min=<0 0>}
    }}
   slot {class=CanvasPanelSlot, properties {layoutData {alignment=<0 0>, anchors {maximum=<0 0>, minimum=<0 0>}, offsets=<ltrb 1184 696 888 468>}}}
  }
  R_Synth {
   widgetClassPath=&CanvasPanel
   parent=R_Frame
   widget=R_Synth
   slot {
    class=CanvasPanelSlot
    properties {zOrder=1, layoutData {alignment=<0 0>, anchors {maximum=<0 0>, minimum=<0 0>}, offsets=<ltrb 1716 852 132 121.4286>}}
   }
  }
  R_Synth_Body {
   widgetClassPath=&Image
   bIsVariable=true
   parent=R_Synth
   widget=R_Synth_Body
   properties {brush {
     bIsDynamicallyLoaded=false
     drawAs=Image
     imageType=FullColor
     mirroring=NoMirror
     resourceName=None
     resourceObject=/PistolCanary/Layout/ComicPanels_v2/ComicPanelTextures/Panel4/R_Frame/R_P4_Synth_Body
     tiling=NoTile
     imageSize=<540 648>
     margin=<ltrb 0 0 0 0>
     outlineSettings {
      bUseBrushTransparency=false
      roundingType=HalfHeightRadius
      width=0
      color {colorUseRule=UseColor_Specified, specifiedColor=<rgba 0 0 0 0>}
      cornerRadii=<0 0 0 0>
     }
     tintColor {colorUseRule=UseColor_Specified, specifiedColor=<rgba 1 1 1 1>}
     uVRegion {bIsValid=false, max=<0 0>, min=<0 0>}
    }}
   slot {
    class=CanvasPanelSlot
    properties {zOrder=1, layoutData {alignment=<0 0>, anchors {maximum=<0 0>, minimum=<0 0>}, offsets=<ltrb -220 -388 540 648>}}
   }
  }
  R_Powers {
   widgetClassPath=&Image
   bIsVariable=true
   parent=R_Synth
   widget=R_Powers
   properties {brush {
     bIsDynamicallyLoaded=false
     drawAs=Image
     imageType=FullColor
     mirroring=NoMirror
     resourceName=None
     resourceObject=/PistolCanary/Layout/ComicPanels_v2/ComicPanelTextures/Panel4/R_Frame/R_P4_Synth_Powers_50percent
     tiling=NoTile
     imageSize=<444 564>
     margin=<ltrb 0 0 0 0>
     outlineSettings {
      bUseBrushTransparency=false
      roundingType=HalfHeightRadius
      width=0
      color {colorUseRule=UseColor_Specified, specifiedColor=<rgba 0 0 0 0>}
      cornerRadii=<0 0 0 0>
     }
     tintColor {colorUseRule=UseColor_Specified, specifiedColor=<rgba 1 1 1 1>}
     uVRegion {bIsValid=false, max=<0 0>, min=<0 0>}
    }}
   slot {
    class=CanvasPanelSlot
    properties {zOrder=1, layoutData {alignment=<0 0>, anchors {maximum=<0 0>, minimum=<0 0>}, offsets=<ltrb -576 -868 888 1128>}}
   }
  }
  R_Sparks_Controller {
   widgetClassPath=&CanvasPanel
   parent=R_Synth
   widget=R_Sparks_Controller
   slot {
    class=CanvasPanelSlot
    properties {layoutData {alignment=<0 0>, anchors {maximum=<0 0>, minimum=<0 0>}, offsets=<ltrb -76 -196 102.2857 96.2857>}}
   }
  }
  R_T_Sparks {
   widgetClassPath=&Image
   bIsVariable=true
   parent=R_Sparks_Controller
   widget=R_T_Sparks
   properties {brush {
     bIsDynamicallyLoaded=false
     drawAs=Image
     imageType=FullColor
     mirroring=NoMirror
     resourceName=None
     resourceObject=/PistolCanary/Layout/ComicPanels_v2/ComicPanelTextures/Panel4/R_Frame/R_P4_Synth_T_Sparks
     tiling=NoTile
     imageSize=<524 373>
     margin=<ltrb 0 0 0 0>
     outlineSettings {
      bUseBrushTransparency=false
      roundingType=HalfHeightRadius
      width=0
      color {colorUseRule=UseColor_Specified, specifiedColor=<rgba 0 0 0 0>}
      cornerRadii=<0 0 0 0>
     }
     tintColor {colorUseRule=UseColor_Specified, specifiedColor=<rgba 1 1 1 1>}
     uVRegion {bIsValid=false, max=<0 0>, min=<0 0>}
    }}
   slot {
    class=CanvasPanelSlot
    properties {bAutoSize=true, layoutData {alignment=<0 0>, anchors {maximum=<0 1>, minimum=<0 0>}, offsets=<ltrb -496 -640 100 363.2857>}}
   }
  }
  R_B_Sparks {
   widgetClassPath=&Image
   bIsVariable=true
   parent=R_Sparks_Controller
   widget=R_B_Sparks
   properties {brush {
     bIsDynamicallyLoaded=false
     drawAs=Image
     imageType=FullColor
     mirroring=NoMirror
     resourceName=None
     resourceObject=/PistolCanary/Layout/ComicPanels_v2/ComicPanelTextures/Panel4/R_Frame/R_P4_Synth_B_Sparks
     tiling=NoTile
     imageSize=<364 400>
     margin=<ltrb 0 0 0 0>
     outlineSettings {
      bUseBrushTransparency=false
      roundingType=HalfHeightRadius
      width=0
      color {colorUseRule=UseColor_Specified, specifiedColor=<rgba 0 0 0 0>}
      cornerRadii=<0 0 0 0>
     }
     tintColor {colorUseRule=UseColor_Specified, specifiedColor=<rgba 1 1 1 1>}
     uVRegion {bIsValid=false, max=<0 0>, min=<0 0>}
    }}
   slot {
    class=CanvasPanelSlot
    properties {bAutoSize=true, layoutData {alignment=<0 0>, anchors {maximum=<0 0>, minimum=<0 0>}, offsets=<ltrb -496 60 100 30>}}
   }
  }
  R_L_Hand_Controller {
   widgetClassPath=&CanvasPanel
   parent=R_Synth
   widget=R_L_Hand_Controller
   slot {
    class=CanvasPanelSlot
    properties {zOrder=5, layoutData {alignment=<0 0>, anchors {maximum=<0 0>, minimum=<0 0>}, offsets=<ltrb -188 -164 22.8571 22>}}
   }
  }
  R_L_Hand {
   widgetClassPath=&Image
   bIsVariable=true
   parent=R_L_Hand_Controller
   widget=R_L_Hand
   properties {
    pixelSnapping=Disabled
    brush {
     bIsDynamicallyLoaded=false
     drawAs=Image
     imageType=FullColor
     mirroring=NoMirror
     resourceName=None
     resourceObject=/PistolCanary/Layout/ComicPanels_v2/ComicPanelTextures/Panel4/R_Frame/R_P4_Synth_L_Hand
     tiling=NoTile
     imageSize=<108 100>
     margin=<ltrb 0 0 0 0>
     outlineSettings {
      bUseBrushTransparency=false
      roundingType=HalfHeightRadius
      width=0
      color {colorUseRule=UseColor_Specified, specifiedColor=<rgba 0 0 0 0>}
      cornerRadii=<0 0 0 0>
     }
     tintColor {colorUseRule=UseColor_Specified, specifiedColor=<rgba 1 1 1 1>}
     uVRegion {bIsValid=false, max=<0 0>, min=<0 0>}
    }
   }
   slot {
    class=CanvasPanelSlot
    properties {zOrder=5, layoutData {alignment=<0 0>, anchors {maximum=<0 0>, minimum=<0 0>}, offsets=<ltrb -88 -72 108 100>}}
   }
  }
  R_R_Hand_Controller {
   widgetClassPath=&CanvasPanel
   parent=R_Synth
   widget=R_R_Hand_Controller
   slot {
    class=CanvasPanelSlot
    properties {zOrder=5, layoutData {alignment=<0 0>, anchors {maximum=<0 0>, minimum=<0 0>}, offsets=<ltrb -208 -336 22.8571 22>}}
   }
  }
  R_R_Hand {
   widgetClassPath=&Image
   bIsVariable=true
   parent=R_R_Hand_Controller
   widget=R_R_Hand
   properties {brush {
     bIsDynamicallyLoaded=false
     drawAs=Image
     imageType=FullColor
     mirroring=NoMirror
     resourceName=None
     resourceObject=/PistolCanary/Layout/ComicPanels_v2/ComicPanelTextures/Panel4/R_Frame/R_P4_Synth_R_Hand
     tiling=NoTile
     imageSize=<100 84>
     margin=<ltrb 0 0 0 0>
     outlineSettings {
      bUseBrushTransparency=false
      roundingType=HalfHeightRadius
      width=0
      color {colorUseRule=UseColor_Specified, specifiedColor=<rgba 0 0 0 0>}
      cornerRadii=<0 0 0 0>
     }
     tintColor {colorUseRule=UseColor_Specified, specifiedColor=<rgba 1 1 1 1>}
     uVRegion {bIsValid=false, max=<0 0>, min=<0 0>}
    }}
   slot {
    class=CanvasPanelSlot
    properties {zOrder=5, layoutData {alignment=<0 0>, anchors {maximum=<0 0>, minimum=<0 0>}, offsets=<ltrb -56 -64 100 84>}}
   }
  }
  Ref {
   widgetClassPath=&Image
   bIsVariable=true
   parent=Panel_04
   widget=Ref
   properties {
    brush {
     bIsDynamicallyLoaded=false
     drawAs=Image
     imageType=FullColor
     mirroring=NoMirror
     resourceName=None
     resourceObject=/PistolCanary/Layout/ComicPanels_v2/Screenshots/Scene_07
     tiling=NoTile
     imageSize=<1920 1080>
     margin=<ltrb 0 0 0 0>
     outlineSettings {
      bUseBrushTransparency=false
      roundingType=HalfHeightRadius
      width=0
      color {colorUseRule=UseColor_Specified, specifiedColor=<rgba 0 0 0 0>}
      cornerRadii=<0 0 0 0>
     }
     tintColor {colorUseRule=UseColor_Specified, specifiedColor=<rgba 1 1 1 1>}
     uVRegion {bIsValid=false, max=<0 0>, min=<0 0>}
    }
    colorAndOpacity=<rgba 1 1 1 0.3>
   }
   slot {class=CanvasPanelSlot, properties {bAutoSize=true, zOrder=100, layoutData=~}}
  }
 ]
}
