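# Panel_05_v02 is a UMG widget for an animated two-frame comic panel of gunfire and explosions.
# P5_L_Frame and P5_R_Frame each slide in a background, animate layered character and muzzle-flash art, then slide their frame off vertically in opposite directions.
# Nested CanvasPanel controllers move image groups through keyed 2D transforms, with staggered overshoot scale pops for Boom, Ping and Bang lettering.
# The event graph holds only empty events, so it never starts either animation.
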WidgetBlueprint /PistolCanary/Layout/ComicPanels_v2/Blueprints/Panel_05_v02
animations [
 P5_L_Frame {
  binding_count=7
  length_seconds=7.9667
  object=&self:P5_L_Frame
  sequence {
   bindings [
    L_F_L_Hand_Controller {
     id=#1
     possessed_class=CanvasPanel
     tracks [{
       type=MovieScene2DTransformTrack
       sections [{
         type=MovieScene2DTransformSection
         end_frame=478
         start_frame=0
         channels [
          Translation.X {type=MovieSceneScriptingFloatChannel, default=0}
          Translation.Y {type=MovieSceneScriptingFloatChannel, default=0}
          Angle {type=MovieSceneScriptingFloatChannel, keys [[0,6.8197],[42,6.8197],[50,0],[80,-3.0991],[478,0.7195]]}
          Scale.X {type=MovieSceneScriptingFloatChannel, default=1}
          Scale.Y {type=MovieSceneScriptingFloatChannel, default=1}
          Shear.X {type=MovieSceneScriptingFloatChannel, default=0}
          Shear.Y {type=MovieSceneScriptingFloatChannel, default=0}
         ]
        }]
      }]
    }
    L_Frame {
     id=#2
     possessed_class=CanvasPanel
     tracks [{
       type=MovieScene2DTransformTrack
       sections [{
         type=MovieScene2DTransformSection
         end_frame=478
         start_frame=0
         channels [
          Translation.X {type=MovieSceneScriptingFloatChannel, const [0,3]}
          Translation.Y {type=MovieSceneScriptingFloatChannel, keys [[0,0],[468,0],[478,1548.6045]]}
          Angle {type=MovieSceneScriptingFloatChannel, default=0}
          Scale.X {type=MovieSceneScriptingFloatChannel, default=1}
          Scale.Y {type=MovieSceneScriptingFloatChannel, default=1}
          Shear.X {type=MovieSceneScriptingFloatChannel, default=0}
          Shear.Y {type=MovieSceneScriptingFloatChannel, default=0}
         ]
        }]
      }]
    }
    L_F_R_Flare_Controller {
     id=#3
     possessed_class=CanvasPanel
     tracks [{
       type=MovieScene2DTransformTrack
       sections [{
         type=MovieScene2DTransformSection
         end_frame=478
         start_frame=0
         channels [
          Translation.X {type=MovieSceneScriptingFloatChannel, default=0}
          Translation.Y {type=MovieSceneScriptingFloatChannel, default=0}
          Angle {type=MovieSceneScriptingFloatChannel, default=0}
          Scale.X {type=MovieSceneScriptingFloatChannel, keys [[0,1],[478,1.2]]}
          Scale.Y {type=MovieSceneScriptingFloatChannel, keys [[0,1],[478,1.2]]}
          Shear.X {type=MovieSceneScriptingFloatChannel, default=0}
          Shear.Y {type=MovieSceneScriptingFloatChannel, default=0}
         ]
        }]
      }]
    }
    L_BG_Controller {
     id=#4
     possessed_class=CanvasPanel
     tracks [{
       type=MovieScene2DTransformTrack
       sections [{
         type=MovieScene2DTransformSection
         end_frame=15
         start_frame=0
         channels [
          Translation.X {type=MovieSceneScriptingFloatChannel, keys [[0,-121.6136],[15,0]]}
          Translation.Y {type=MovieSceneScriptingFloatChannel, keys [[0,-1235.63],[15,0]]}
          Angle {type=MovieSceneScriptingFloatChannel, default=0}
          Scale.X {type=MovieSceneScriptingFloatChannel, default=0.6}
          Scale.Y {type=MovieSceneScriptingFloatChannel, default=0.6}
          Shear.X {type=MovieSceneScriptingFloatChannel, default=0}
          Shear.Y {type=MovieSceneScriptingFloatChannel, default=0}
         ]
        }]
      }]
    }
    L_Kill_Controller {
     id=#5
     possessed_class=CanvasPanel
     tracks [{
       type=MovieScene2DTransformTrack
       sections [{
         type=MovieScene2DTransformSection
         end_frame=478
         start_frame=0
         channels [
          Translation.X {type=MovieSceneScriptingFloatChannel, keys [[15,390.7174],[30,7.1489],[478,-6.8349]]}
          Translation.Y {type=MovieSceneScriptingFloatChannel, keys [[15,1159.0286],[30,18.1728],[478,-21.0255]]}
          Angle {type=MovieSceneScriptingFloatChannel, const [0,1]}
          Scale.X {type=MovieSceneScriptingFloatChannel, keys [[15,0.2],[30,0.55]]}
          Scale.Y {type=MovieSceneScriptingFloatChannel, keys [[15,0.2],[30,0.55]]}
          Shear.X {type=MovieSceneScriptingFloatChannel, default=0}
          Shear.Y {type=MovieSceneScriptingFloatChannel, default=0}
         ]
        }]
      }]
    }
    L_F_L_Flare_Controller {
     id=#6
     possessed_class=CanvasPanel
     tracks [{
       type=MovieScene2DTransformTrack
       sections [{
         type=MovieScene2DTransformSection
         end_frame=478
         start_frame=0
         channels [
          Translation.X {type=MovieSceneScriptingFloatChannel, keys [[42,84.0896],[50,0]]}
          Translation.Y {type=MovieSceneScriptingFloatChannel, const [0,2]}
          Angle {type=MovieSceneScriptingFloatChannel, keys [[42,0.55],[50,0]]}
          Scale.X {type=MovieSceneScriptingFloatChannel, keys [[0,0],[42,0],[46,1],[478,1.2]]}
          Scale.Y {type=MovieSceneScriptingFloatChannel, keys [[0,0],[42,0],[46,1],[478,1.2]]}
          Shear.X {type=MovieSceneScriptingFloatChannel, const [0,1]}
          Shear.Y {type=MovieSceneScriptingFloatChannel, const [0,1]}
         ]
        }]
      }]
    }
    L_Muzzle_Splash {
     id=#7
     possessed_class=Image
     tracks [{
       type=MovieScene2DTransformTrack
       sections [{
         type=MovieScene2DTransformSection
         end_frame=478
         start_frame=15
         channels [
          Translation.X {type=MovieSceneScriptingFloatChannel, const [0,3]}
          Translation.Y {type=MovieSceneScriptingFloatChannel, keys [[15,0],[30,0],[478,-93.0193]]}
          Angle {type=MovieSceneScriptingFloatChannel, default=0}
          Scale.X {type=MovieSceneScriptingFloatChannel, keys [[15,0.1],[30,1],[478,1.2]]}
          Scale.Y {type=MovieSceneScriptingFloatChannel, keys [[15,0.1],[30,1],[478,1.2]]}
          Shear.X {type=MovieSceneScriptingFloatChannel, default=0}
          Shear.Y {type=MovieSceneScriptingFloatChannel, default=0}
         ]
        }]
      }]
    }
   ]
   display_rate {numerator=60}
  }
 }
 P5_R_Frame {
  binding_count=17
  length_seconds=9.4167
  object=&self:P5_R_Frame
  sequence {
   bindings [
    R_Muzzle_Splash {
     id=#8
     possessed_class=Image
     tracks [{
       type=MovieScene2DTransformTrack
       sections [{
         type=MovieScene2DTransformSection
         end_frame=510
         start_frame=78
         channels [
          Translation.X {type=MovieSceneScriptingFloatChannel, keys [[78,109.0228],[89,0]]}
          Translation.Y {type=MovieSceneScriptingFloatChannel, keys [[78,105.0969],[89,0]]}
          Angle {type=MovieSceneScriptingFloatChannel, default=0}
          Scale.X {type=MovieSceneScriptingFloatChannel, keys [[78,0],[89,1],[122,1.1],[508,1.2]]}
          Scale.Y {type=MovieSceneScriptingFloatChannel, keys [[78,0],[89,1],[122,1.1],[510,1.2]]}
          Shear.X {type=MovieSceneScriptingFloatChannel, default=0}
          Shear.Y {type=MovieSceneScriptingFloatChannel, default=0}
         ]
        }]
      }]
    }
    L_Kill_Muzzle_Splash_B {
     id=#9
     possessed_class=Image
     tracks [{
       type=MovieScene2DTransformTrack
       sections [{
         type=MovieScene2DTransformSection
         end_frame=487
         start_frame=55
         channels [
          Translation.X {type=MovieSceneScriptingFloatChannel, keys [[55,197.8459],[66,0]]}
          Translation.Y {type=MovieSceneScriptingFloatChannel, keys [[55,-388.5678],[66,0]]}
          Angle {type=MovieSceneScriptingFloatChannel, const [0,1]}
          Scale.X {type=MovieSceneScriptingFloatChannel, keys [[55,0],[66,1],[99,1.1],[485,1.2]]}
          Scale.Y {type=MovieSceneScriptingFloatChannel, keys [[55,0],[66,1],[99,1.1],[487,1.2]]}
          Shear.X {type=MovieSceneScriptingFloatChannel, const [0,1]}
          Shear.Y {type=MovieSceneScriptingFloatChannel, const [0,1]}
         ]
        }]
      }]
    }
    Boom_02 {
     id=#10
     possessed_class=Image
     tracks [{
       type=MovieScene2DTransformTrack
       sections [{
         type=MovieScene2DTransformSection
         end_frame=536
         start_frame=0
         channels [
          Translation.X {type=MovieSceneScriptingFloatChannel, const [0,1]}
          Translation.Y {type=MovieSceneScriptingFloatChannel, const [0,1]}
          Angle {type=MovieSceneScriptingFloatChannel, const [0,1]}
          Scale.X {type=MovieSceneScriptingFloatChannel, keys [[0,0],[77,0],[83,1.05],[86,0.99],[89,1],[536,1.1]]}
          Scale.Y {type=MovieSceneScriptingFloatChannel, keys [[0,0],[77,0],[83,1.05],[86,0.99],[89,1],[536,1.1]]}
          Shear.X {type=MovieSceneScriptingFloatChannel, const [0,1]}
          Shear.Y {type=MovieSceneScriptingFloatChannel, const [0,1]}
         ]
        }]
      }]
    }
    R_Hand_Controller {
     id=#11
     possessed_class=CanvasPanel
     tracks [{
       type=MovieScene2DTransformTrack
       sections [{
         type=MovieScene2DTransformSection
         end_frame=510
         start_frame=78
         channels [
          Translation.X {type=MovieSceneScriptingFloatChannel, default=0}
          Translation.Y {type=MovieSceneScriptingFloatChannel, default=0}
          Angle {type=MovieSceneScriptingFloatChannel, keys [[78,0],[85,15],[107,2.6253],[510,-2.2909]]}
          Scale.X {type=MovieSceneScriptingFloatChannel, default=1}
          Scale.Y {type=MovieSceneScriptingFloatChannel, default=1}
          Shear.X {type=MovieSceneScriptingFloatChannel, default=0}
          Shear.Y {type=MovieSceneScriptingFloatChannel, default=0}
         ]
        }]
      }]
    }
    L_Hand_Controller {
     id=#12
     possessed_class=CanvasPanel
     tracks [{
       type=MovieScene2DTransformTrack
       sections [{
         type=MovieScene2DTransformSection
         end_frame=487
         start_frame=55
         channels [
          Translation.X {type=MovieSceneScriptingFloatChannel, const [0,1]}
          Translation.Y {type=MovieSceneScriptingFloatChannel, const [0,1]}
          Angle {type=MovieSceneScriptingFloatChannel, keys [[55,0],[62,10],[84,2.6253],[487,-2.2909]]}
          Scale.X {type=MovieSceneScriptingFloatChannel, const [1,1]}
          Scale.Y {type=MovieSceneScriptingFloatChannel, const [1,1]}
          Shear.X {type=MovieSceneScriptingFloatChannel, const [0,1]}
          Shear.Y {type=MovieSceneScriptingFloatChannel, const [0,1]}
         ]
        }]
      }]
    }
    R_Debris_Controller {
     id=#13
     possessed_class=CanvasPanel
     tracks [{
       type=MovieScene2DTransformTrack
       sections [{
         type=MovieScene2DTransformSection
         end_frame=478
         start_frame=0
         channels [
          Translation.X {type=MovieSceneScriptingFloatChannel, keys [[15,0],[478,37.9215]]}
          Translation.Y {type=MovieSceneScriptingFloatChannel, keys [[15,0],[478,-58.4046]]}
          Angle {type=MovieSceneScriptingFloatChannel, keys [[15,0],[478,2.9289]]}
          Scale.X {type=MovieSceneScriptingFloatChannel, keys [[0,1.0145],[15,0],[25,1],[478,1.1]]}
          Scale.Y {type=MovieSceneScriptingFloatChannel, keys [[0,0],[15,0],[25,1],[478,1.1]]}
          Shear.X {type=MovieSceneScriptingFloatChannel, default=0}
          Shear.Y {type=MovieSceneScriptingFloatChannel, default=0}
         ]
        }]
      }]
    }
    Ping_02 {
     id=#14
     possessed_class=Image
     tracks [{
       type=MovieScene2DTransformTrack
       sections [{
         type=MovieScene2DTransformSection
         end_frame=487
         start_frame=0
         channels [
          Translation.X {type=MovieSceneScriptingFloatChannel, const [0,1]}
          Translation.Y {type=MovieSceneScriptingFloatChannel, const [0,1]}
          Angle {type=MovieSceneScriptingFloatChannel, const [0,1]}
          Scale.X {type=MovieSceneScriptingFloatChannel, keys [[0,0],[28,0],[34,1.05],[37,0.99],[40,1],[487,1.1]]}
          Scale.Y {type=MovieSceneScriptingFloatChannel, keys [[0,0],[28,0],[34,1.05],[37,0.99],[40,1],[487,1.1]]}
          Shear.X {type=MovieSceneScriptingFloatChannel, const [0,1]}
          Shear.Y {type=MovieSceneScriptingFloatChannel, const [0,1]}
         ]
        }]
      }]
    }
    Boom_01 {
     id=#15
     possessed_class=Image
     tracks [{
       type=MovieScene2DTransformTrack
       sections [{
         type=MovieScene2DTransformSection
         end_frame=478
         start_frame=0
         channels [
          Translation.X {type=MovieSceneScriptingFloatChannel, default=0}
          Translation.Y {type=MovieSceneScriptingFloatChannel, default=0}
          Angle {type=MovieSceneScriptingFloatChannel, default=0}
          Scale.X {type=MovieSceneScriptingFloatChannel, keys [[0,0],[19,0],[25,1.05],[28,0.99],[31,1],[478,1.1]]}
          Scale.Y {type=MovieSceneScriptingFloatChannel, keys [[0,0],[19,0],[25,1.05],[28,0.99],[31,1],[478,1.1]]}
          Shear.X {type=MovieSceneScriptingFloatChannel, const [0,1]}
          Shear.Y {type=MovieSceneScriptingFloatChannel, const [0,1]}
         ]
        }]
      }]
    }
    Ping_01 {
     id=#16
     possessed_class=Image
     tracks [{
       type=MovieScene2DTransformTrack
       sections [{
         type=MovieScene2DTransformSection
         end_frame=485
         start_frame=0
         channels [
          Translation.X {type=MovieSceneScriptingFloatChannel, default=0}
          Translation.Y {type=MovieSceneScriptingFloatChannel, default=0}
          Angle {type=MovieSceneScriptingFloatChannel, default=0}
          Scale.X {type=MovieSceneScriptingFloatChannel, keys [[0,0],[26,0],[32,1.05],[35,0.99],[38,1],[485,1.1]]}
          Scale.Y {type=MovieSceneScriptingFloatChannel, keys [[0,0],[26,0],[32,1.05],[35,0.99],[38,1],[485,1.1]]}
          Shear.X {type=MovieSceneScriptingFloatChannel, default=0}
          Shear.Y {type=MovieSceneScriptingFloatChannel, default=0}
         ]
        }]
      }]
    }
    Fire_Controller {
     id=#17
     possessed_class=CanvasPanel
     tracks [{
       type=MovieScene2DTransformTrack
       sections [{
         type=MovieScene2DTransformSection
         end_frame=478
         start_frame=0
         channels [
          Translation.X {type=MovieSceneScriptingFloatChannel, default=0, keys [[25,0],[478,13.4196]]}
          Translation.Y {type=MovieSceneScriptingFloatChannel, default=0, const [0,2]}
          Angle {type=MovieSceneScriptingFloatChannel, default=0}
          Scale.X {type=MovieSceneScriptingFloatChannel, keys [[0,1.0145],[15,0],[25,1],[478,1.1]]}
          Scale.Y {type=MovieSceneScriptingFloatChannel, keys [[0,0],[15,0],[25,1],[478,1.1]]}
          Shear.X {type=MovieSceneScriptingFloatChannel, default=0}
          Shear.Y {type=MovieSceneScriptingFloatChannel, default=0}
         ]
        }]
      }]
    }
    Bang_01 {
     id=#18
     possessed_class=Image
     tracks [{
       type=MovieScene2DTransformTrack
       sections [{
         type=MovieScene2DTransformSection
         end_frame=515
         start_frame=0
         channels [
          Translation.X {type=MovieSceneScriptingFloatChannel, const [0,1]}
          Translation.Y {type=MovieSceneScriptingFloatChannel, const [0,1]}
          Angle {type=MovieSceneScriptingFloatChannel, const [0,1]}
          Scale.X {type=MovieSceneScriptingFloatChannel, keys [[0,0],[54,0],[60,1.05],[63,0.99],[66,1],[515,1.1]]}
          Scale.Y {type=MovieSceneScriptingFloatChannel, keys [[0,0],[54,0],[60,1.05],[63,0.99],[66,1],[515,1.1]]}
          Shear.X {type=MovieSceneScriptingFloatChannel, const [0,1]}
          Shear.Y {type=MovieSceneScriptingFloatChannel, const [0,1]}
         ]
        }]
      }]
    }
    L_Kill_Muzzle_Splash_T {
     id=#19
     possessed_class=Image
     tracks [{
       type=MovieScene2DTransformTrack
       sections [{
         type=MovieScene2DTransformSection
         end_frame=487
         start_frame=55
         channels [
          Translation.X {type=MovieSceneScriptingFloatChannel, const [0,1]}
          Translation.Y {type=MovieSceneScriptingFloatChannel, const [0,1]}
          Angle {type=MovieSceneScriptingFloatChannel, const [0,1]}
          Scale.X {type=MovieSceneScriptingFloatChannel, keys [[55,0],[66,1],[99,1.1],[485,1.2]]}
          Scale.Y {type=MovieSceneScriptingFloatChannel, keys [[55,0],[66,1],[99,1.1],[487,1.2]]}
          Shear.X {type=MovieSceneScriptingFloatChannel, const [0,1]}
          Shear.Y {type=MovieSceneScriptingFloatChannel, const [0,1]}
         ]
        }]
      }]
    }
    R_Muzzle_Controller {
     id=#20
     possessed_class=CanvasPanel
     tracks [{
       type=MovieScene2DTransformTrack
       sections [{
         type=MovieScene2DTransformSection
         end_frame=565
         start_frame=0
         channels [
          Translation.X {type=MovieSceneScriptingFloatChannel, default=0}
          Translation.Y {type=MovieSceneScriptingFloatChannel, default=0}
          Angle {type=MovieSceneScriptingFloatChannel, default=0}
          Scale.X {type=MovieSceneScriptingFloatChannel, thinned_from=6, keys [[0,0],[78,0],[81,1],[122,1.1],[510,1.3]]}
          Scale.Y {type=MovieSceneScriptingFloatChannel, thinned_from=6, keys [[0,0],[78,0],[81,1],[122,1.1],[510,1.3]]}
          Shear.X {type=MovieSceneScriptingFloatChannel, default=0}
          Shear.Y {type=MovieSceneScriptingFloatChannel, default=0}
         ]
        }]
      }]
    }
    L_Muzzle_Controller {
     id=#21
     possessed_class=CanvasPanel
     tracks [{
       type=MovieScene2DTransformTrack
       sections [{
         type=MovieScene2DTransformSection
         end_frame=487
         start_frame=0
         channels [
          Translation.X {type=MovieSceneScriptingFloatChannel, default=0}
          Translation.Y {type=MovieSceneScriptingFloatChannel, default=0}
          Angle {type=MovieSceneScriptingFloatChannel, default=0}
          Scale.X {type=MovieSceneScriptingFloatChannel, keys [[0,0],[55,0],[58,1],[99,1.1],[487,1.3]]}
          Scale.Y {type=MovieSceneScriptingFloatChannel, keys [[0,0],[55,0],[58,1],[99,1.1],[487,1.3]]}
          Shear.X {type=MovieSceneScriptingFloatChannel, default=0}
          Shear.Y {type=MovieSceneScriptingFloatChannel, default=0}
         ]
        }]
      }]
    }
    R_Kill_Controller {
     id=#22
     possessed_class=CanvasPanel
     tracks [{
       type=MovieScene2DTransformTrack
       sections [{
         type=MovieScene2DTransformSection
         end_frame=480
         start_frame=0
         channels [
          Translation.X {type=MovieSceneScriptingFloatChannel, keys [[0,1210.6006],[25,1210.6006],[46,0],[104,-80.3032],[479,-153.5947]]}
          Translation.Y {type=MovieSceneScriptingFloatChannel, keys [[0,989.6136],[25,989.6136],[46,-58.929],[105,-51.6649],[480,72.6561]]}
          Angle {type=MovieSceneScriptingFloatChannel, keys [[0,23.1235],[25,23.1235],[46,0],[480,-4.4341]]}
          Scale.X {type=MovieSceneScriptingFloatChannel, default=1}
          Scale.Y {type=MovieSceneScriptingFloatChannel, default=1}
          Shear.X {type=MovieSceneScriptingFloatChannel, default=0}
          Shear.Y {type=MovieSceneScriptingFloatChannel, default=0}
         ]
        }]
      }]
    }
    R_Frame {
     id=#23
     possessed_class=CanvasPanel
     tracks [{
       type=MovieScene2DTransformTrack
       sections [{
         type=MovieScene2DTransformSection
         end_frame=478
         start_frame=0
         channels [
          Translation.X {type=MovieSceneScriptingFloatChannel, const [0,3]}
          Translation.Y {type=MovieSceneScriptingFloatChannel, keys [[0,0],[384,0],[394,-1160.2108]]}
          Angle {type=MovieSceneScriptingFloatChannel, default=0}
          Scale.X {type=MovieSceneScriptingFloatChannel, default=0.55}
          Scale.Y {type=MovieSceneScriptingFloatChannel, default=0.55}
          Shear.X {type=MovieSceneScriptingFloatChannel, default=0}
          Shear.Y {type=MovieSceneScriptingFloatChannel, default=0}
         ]
        }]
      }]
    }
    R_BG_Controller {
     id=#24
     possessed_class=CanvasPanel
     tracks [{
       type=MovieScene2DTransformTrack
       sections [{
         type=MovieScene2DTransformSection
         end_frame=15
         start_frame=0
         channels [
          Translation.X {type=MovieSceneScriptingFloatChannel, keys [[0,794.0942],[15,573.6088]]}
          Translation.Y {type=MovieSceneScriptingFloatChannel, keys [[0,2635.5718],[15,489.965]]}
          Angle {type=MovieSceneScriptingFloatChannel, default=0}
          Scale.X {type=MovieSceneScriptingFloatChannel, default=1}
          Scale.Y {type=MovieSceneScriptingFloatChannel, default=1}
          Shear.X {type=MovieSceneScriptingFloatChannel, default=0}
          Shear.Y {type=MovieSceneScriptingFloatChannel, default=0}
         ]
        }]
      }]
    }
   ]
   display_rate {numerator=60}
  }
 }
]
graphs [EventGraph {nodes [
   Event PreConstruct()
   Event Construct()
   Event Tick()
  ]}]
variables []
widgets {
 info {inheritedWidgetCount=0, parentClass=&UserWidget, rootWidgetClass=&CanvasPanel, widgetCount=44}
 widgets [
  Panel_05 {widgetClassPath=&CanvasPanel, widget=Panel_05, properties {clipping=ClipToBounds, pixelSnapping=Disabled}}
  L_Frame {
   widgetClassPath=&CanvasPanel
   parent=Panel_05
   widget=L_Frame
   slot {class=CanvasPanelSlot, properties {bAutoSize=true, layoutData=~}}
  }
  L_BG_Controller {
   widgetClassPath=&CanvasPanel
   parent=L_Frame
   widget=L_BG_Controller
   properties {renderTransform {angle=0, scale=<0.6 0.6>, shear=<0 0>, translation=<0 0>}}
   slot {
    class=CanvasPanelSlot
    properties {layoutData {alignment=<0 0>, anchors {maximum=<0 0>, minimum=<0 0>}, offsets=<ltrb -28 948 156 162.5714>}}
   }
  }
  L_BG {
   widgetClassPath=&Image
   bIsVariable=true
   parent=L_BG_Controller
   widget=L_BG
   properties {brush {
     bIsDynamicallyLoaded=false
     drawAs=Image
     imageType=FullColor
     mirroring=NoMirror
     resourceName=None
     resourceObject=/PistolCanary/Layout/ComicPanels_v2/ComicPanelTextures/Panel5/L_Frame/L_F_BG_50percent
     tiling=NoTile
     imageSize=<604 1132>
     margin=<ltrb 0 0 0 0>
     outlineSettings {
      bUseBrushTransparency=false
      roundingType=HalfHeightRadius
      width=0
      color {colorUseRule=UseColor_Specified, specifiedColor=<rgba 0 0 0 0>}
      cornerRadii=<0 0 0 0>
     }
     tintColor {colorUseRule=UseColor_Specified, specifiedColor=<rgba 1 1 1 1>}
     uVRegion {bIsValid=false, max=<0 0>, min=<0 0>}
    }}
   slot {
    class=CanvasPanelSlot
    properties {layoutData {alignment=<0 0>, anchors {maximum=<0 0>, minimum=<0 0>}, offsets=<ltrb -104 -1880 1212 2268>}}
   }
  }
  L_Kill_Controller {
   widgetClassPath=&CanvasPanel
   parent=L_Frame
   widget=L_Kill_Controller
   properties {renderTransform {angle=0, scale=<0.55 0.55>, shear=<0 0>, translation=<0 0>}}
   slot {
    class=CanvasPanelSlot
    properties {layoutData {alignment=<0 0>, anchors {maximum=<0 0>, minimum=<0 0>}, offsets=<ltrb 444 884 108 105.4286>}}
   }
  }
  L_Kill {
   widgetClassPath=&Image
   bIsVariable=true
   parent=L_Kill_Controller
   widget=L_Kill
   properties {brush {
     bIsDynamicallyLoaded=false
     drawAs=Image
     imageType=FullColor
     mirroring=NoMirror
     resourceName=None
     resourceObject=/PistolCanary/Layout/ComicPanels_v2/ComicPanelTextures/Panel5/L_Frame/L_F_Kill_Body_30percent
     tiling=NoTile
     imageSize=<336 352>
     margin=<ltrb 0 0 0 0>
     outlineSettings {
      bUseBrushTransparency=false
      roundingType=HalfHeightRadius
      width=0
      color {colorUseRule=UseColor_Specified, specifiedColor=<rgba 0 0 0 0>}
      cornerRadii=<0 0 0 0>
     }
     tintColor {colorUseRule=UseColor_Specified, specifiedColor=<rgba 1 1 1 1>}
     uVRegion {bIsValid=false, max=<0 0>, min=<0 0>}
    }}
   slot {
    class=CanvasPanelSlot
    properties {zOrder=1, layoutData {alignment=<0 0>, anchors {maximum=<0 0>, minimum=<0 0>}, offsets=<ltrb -776 -812 1128 1172>}}
   }
  }
  L_Muzzle_Splash {
   widgetClassPath=&Image
   bIsVariable=true
   parent=L_Kill_Controller
   widget=L_Muzzle_Splash
   properties {brush {
     bIsDynamicallyLoaded=false
     drawAs=Image
     imageType=FullColor
     mirroring=NoMirror
     resourceName=None
     resourceObject=/PistolCanary/Layout/ComicPanels_v2/ComicPanelTextures/Panel5/L_Frame/L_F_Muzzle_Splash_50percent
     tiling=NoTile
     imageSize=<432 536>
     margin=<ltrb 0 0 0 0>
     outlineSettings {
      bUseBrushTransparency=false
      roundingType=HalfHeightRadius
      width=0
      color {colorUseRule=UseColor_Specified, specifiedColor=<rgba 0 0 0 0>}
      cornerRadii=<0 0 0 0>
     }
     tintColor {colorUseRule=UseColor_Specified, specifiedColor=<rgba 1 1 1 1>}
     uVRegion {bIsValid=false, max=<0 0>, min=<0 0>}
    }}
   slot {class=CanvasPanelSlot, properties {layoutData {alignment=<0 0>, anchors {maximum=<0 0>, minimum=<0 0>}, offsets=<ltrb -852 -1552 860 1072>}}}
  }
  L_F_L_Hand_Controller {
   widgetClassPath=&CanvasPanel
   parent=L_Kill_Controller
   widget=L_F_L_Hand_Controller
   slot {
    class=CanvasPanelSlot
    properties {zOrder=2, layoutData {alignment=<0 0>, anchors {maximum=<0 0>, minimum=<0 0>}, offsets=<ltrb -396 -824 88.5714 80.2857>}}
   }
  }
  L_Hand {
   widgetClassPath=&Image
   bIsVariable=true
   parent=L_F_L_Hand_Controller
   widget=L_Hand
   properties {brush {
     bIsDynamicallyLoaded=false
     drawAs=Image
     imageType=FullColor
     mirroring=NoMirror
     resourceName=None
     resourceObject=/PistolCanary/Layout/ComicPanels_v2/ComicPanelTextures/Panel5/L_Frame/L_F_L_Hand_50percent
     tiling=NoTile
     imageSize=<256 476>
     margin=<ltrb 0 0 0 0>
     outlineSettings {
      bUseBrushTransparency=false
      roundingType=HalfHeightRadius
      width=0
      color {colorUseRule=UseColor_Specified, specifiedColor=<rgba 0 0 0 0>}
      cornerRadii=<0 0 0 0>
     }
     tintColor {colorUseRule=UseColor_Specified, specifiedColor=<rgba 1 1 1 1>}
     uVRegion {bIsValid=false, max=<0 0>, min=<0 0>}
    }}
   slot {class=CanvasPanelSlot, properties {layoutData {alignment=<0 0>, anchors {maximum=<0 0>, minimum=<0 0>}, offsets=<ltrb -344 -800 512 952>}}}
  }
  L_F_L_Flare_Controller {
   widgetClassPath=&CanvasPanel
   parent=L_Kill_Controller
   widget=L_F_L_Flare_Controller
   slot {
    class=CanvasPanelSlot
    properties {layoutData {alignment=<0 0>, anchors {maximum=<0 0>, minimum=<0 0>}, offsets=<ltrb -700 -1464 80.5714 78>}}
   }
  }
  L_B_Muzzle {
   widgetClassPath=&Image
   bIsVariable=true
   parent=L_F_L_Flare_Controller
   widget=L_B_Muzzle
   properties {
    brush {
     bIsDynamicallyLoaded=false
     drawAs=Image
     imageType=FullColor
     mirroring=NoMirror
     resourceName=None
     resourceObject=/PistolCanary/Layout/ComicPanels_v2/ComicPanelTextures/Panel5/L_Frame/L_F_L_B_Muzzle_Flare_50percent
     tiling=NoTile
     imageSize=<388 276>
     margin=<ltrb 0 0 0 0>
     outlineSettings {
      bUseBrushTransparency=false
      roundingType=HalfHeightRadius
      width=0
      color {colorUseRule=UseColor_Specified, specifiedColor=<rgba 0 0 0 0>}
      cornerRadii=<0 0 0 0>
     }
     tintColor {colorUseRule=UseColor_Specified, specifiedColor=<rgba 1 1 1 1>}
     uVRegion {bIsValid=false, max=<0 0>, min=<0 0>}
    }
    renderTransform {angle=0, scale=<2 2>, shear=<0 0>, translation=<-197 -282>}
   }
   slot {
    class=CanvasPanelSlot
    properties {layoutData {alignment=<0 0>, anchors {maximum=<1 1>, minimum=<1 1>}, offsets=<ltrb -80.5714 -78 388 276>}}
   }
  }
  L_T_Muzzle {
   widgetClassPath=&Image
   bIsVariable=true
   parent=L_F_L_Flare_Controller
   widget=L_T_Muzzle
   properties {brush {
     bIsDynamicallyLoaded=false
     drawAs=Image
     imageType=FullColor
     mirroring=NoMirror
     resourceName=None
     resourceObject=/PistolCanary/Layout/ComicPanels_v2/ComicPanelTextures/Panel5/L_Frame/L_F_L_T_Muzzle_Flare_70percent
     tiling=NoTile
     imageSize=<548 908>
     margin=<ltrb 0 0 0 0>
     outlineSettings {
      bUseBrushTransparency=false
      roundingType=HalfHeightRadius
      width=0
      color {colorUseRule=UseColor_Specified, specifiedColor=<rgba 0 0 0 0>}
      cornerRadii=<0 0 0 0>
     }
     tintColor {colorUseRule=UseColor_Specified, specifiedColor=<rgba 1 1 1 1>}
     uVRegion {bIsValid=false, max=<0 0>, min=<0 0>}
    }}
   slot {
    class=CanvasPanelSlot
    properties {zOrder=5, layoutData {alignment=<0 0>, anchors {maximum=<0 0>, minimum=<0 0>}, offsets=<ltrb -392 -1176 780 1300>}}
   }
  }
  L_F_R_Flare_Controller {
   widgetClassPath=&CanvasPanel
   parent=L_Kill_Controller
   widget=L_F_R_Flare_Controller
   slot {
    class=CanvasPanelSlot
    properties {layoutData {alignment=<0 0>, anchors {maximum=<0 0>, minimum=<0 0>}, offsets=<ltrb -172 -684 105.2245 103.7959>}}
   }
  }
  R_B_Muzzle {
   widgetClassPath=&Image
   bIsVariable=true
   parent=L_F_R_Flare_Controller
   widget=R_B_Muzzle
   properties {brush {
     bIsDynamicallyLoaded=false
     drawAs=Image
     imageType=FullColor
     mirroring=NoMirror
     resourceName=None
     resourceObject=/PistolCanary/Layout/ComicPanels_v2/ComicPanelTextures/Panel5/L_Frame/L_F_R_B_Muzzle_Flare
     tiling=NoTile
     imageSize=<252 640>
     margin=<ltrb 0 0 0 0>
     outlineSettings {
      bUseBrushTransparency=false
      roundingType=HalfHeightRadius
      width=0
      color {colorUseRule=UseColor_Specified, specifiedColor=<rgba 0 0 0 0>}
      cornerRadii=<0 0 0 0>
     }
     tintColor {colorUseRule=UseColor_Specified, specifiedColor=<rgba 1 1 1 1>}
     uVRegion {bIsValid=false, max=<0 0>, min=<0 0>}
    }}
   slot {
    class=CanvasPanelSlot
    properties {bAutoSize=true, layoutData {alignment=<0 0>, anchors {maximum=<0 0>, minimum=<0 0>}, offsets=<ltrb -16 -504 100 30>}}
   }
  }
  R_T_Muzzle {
   widgetClassPath=&Image
   bIsVariable=true
   parent=L_F_R_Flare_Controller
   widget=R_T_Muzzle
   properties {brush {
     bIsDynamicallyLoaded=false
     drawAs=Image
     imageType=FullColor
     mirroring=NoMirror
     resourceName=None
     resourceObject=/PistolCanary/Layout/ComicPanels_v2/ComicPanelTextures/Panel5/L_Frame/L_F_R_T_Muzzle_Flare
     tiling=NoTile
     imageSize=<252 640>
     margin=<ltrb 0 0 0 0>
     outlineSettings {
      bUseBrushTransparency=false
      roundingType=HalfHeightRadius
      width=0
      color {colorUseRule=UseColor_Specified, specifiedColor=<rgba 0 0 0 0>}
      cornerRadii=<0 0 0 0>
     }
     tintColor {colorUseRule=UseColor_Specified, specifiedColor=<rgba 1 1 1 1>}
     uVRegion {bIsValid=false, max=<0 0>, min=<0 0>}
    }}
   slot {
    class=CanvasPanelSlot
    properties {bAutoSize=true, zOrder=6, layoutData {alignment=<0 0>, anchors {maximum=<0 0>, minimum=<0 0>}, offsets=<ltrb -16 -504 100 30>}}
   }
  }
  R_Frame {
   widgetClassPath=&CanvasPanel
   parent=Panel_05
   widget=R_Frame
   properties {renderTransform {angle=0, scale=<0.55 0.55>, shear=<0 0>, translation=<0 0>}}
   slot {class=CanvasPanelSlot, properties {bAutoSize=true, layoutData=~}}
  }
  R_BG_Controller {
   widgetClassPath=&CanvasPanel
   parent=R_Frame
   widget=R_BG_Controller
   slot {
    class=CanvasPanelSlot
    properties {layoutData {alignment=<0 0>, anchors {maximum=<0 0>, minimum=<0 0>}, offsets=<ltrb 1710.4762 696.9524 766.6666 692.4762>}}
   }
  }
  R_BG {
   widgetClassPath=&Image
   bIsVariable=true
   parent=R_BG_Controller
   widget=R_BG
   properties {brush {
     bIsDynamicallyLoaded=false
     drawAs=Image
     imageType=FullColor
     mirroring=NoMirror
     resourceName=None
     resourceObject=/PistolCanary/Layout/ComicPanels_v2/ComicPanelTextures/Panel5/R_Frame/R_F_BG_80percent
     tiling=NoTile
     imageSize=<1044 1772>
     margin=<ltrb 0 0 0 0>
     outlineSettings {
      bUseBrushTransparency=false
      roundingType=HalfHeightRadius
      width=0
      color {colorUseRule=UseColor_Specified, specifiedColor=<rgba 0 0 0 0>}
      cornerRadii=<0 0 0 0>
     }
     tintColor {colorUseRule=UseColor_Specified, specifiedColor=<rgba 1 1 1 1>}
     uVRegion {bIsValid=false, max=<0 0>, min=<0 0>}
    }}
   slot {
    class=CanvasPanelSlot
    properties {layoutData {alignment=<0 0>, anchors {maximum=<0 0>, minimum=<0 0>}, offsets=<ltrb -1081.8877 -1928 1304 2220>}}
   }
  }
  Fire_Controller {
   widgetClassPath=&CanvasPanel
   parent=R_BG_Controller
   widget=Fire_Controller
   slot {
    class=CanvasPanelSlot
    properties {layoutData {alignment=<0 0>, anchors {maximum=<0 0>, minimum=<0 0>}, offsets=<ltrb -584 168 97.4603 99.418>}}
   }
  }
  R_Flames {
   widgetClassPath=&Image
   bIsVariable=true
   parent=Fire_Controller
   widget=R_Flames
   properties {brush {
     bIsDynamicallyLoaded=false
     drawAs=Image
     imageType=FullColor
     mirroring=NoMirror
     resourceName=None
     resourceObject=/PistolCanary/Layout/ComicPanels_v2/ComicPanelTextures/Panel5/R_Frame/R_F_Flames_50percent
     tiling=NoTile
     imageSize=<416 620>
     margin=<ltrb 0 0 0 0>
     outlineSettings {
      bUseBrushTransparency=false
      roundingType=HalfHeightRadius
      width=0
      color {colorUseRule=UseColor_Specified, specifiedColor=<rgba 0 0 0 0>}
      cornerRadii=<0 0 0 0>
     }
     tintColor {colorUseRule=UseColor_Specified, specifiedColor=<rgba 1 1 1 1>}
     uVRegion {bIsValid=false, max=<0 0>, min=<0 0>}
    }}
   slot {
    class=CanvasPanelSlot
    properties {layoutData {alignment=<0 0>, anchors {maximum=<0 0>, minimum=<0 0>}, offsets=<ltrb -84.9925 -1124 832 1236>}}
   }
  }
  R_Embers {
   widgetClassPath=&Image
   bIsVariable=true
   parent=Fire_Controller
   widget=R_Embers
   properties {brush {
     bIsDynamicallyLoaded=false
     drawAs=Image
     imageType=FullColor
     mirroring=NoMirror
     resourceName=None
     resourceObject=/PistolCanary/Layout/ComicPanels_v2/ComicPanelTextures/Panel5/R_Frame/R_F_Embers_70percent
     tiling=NoTile
     imageSize=<640 688>
     margin=<ltrb 0 0 0 0>
     outlineSettings {
      bUseBrushTransparency=false
      roundingType=HalfHeightRadius
      width=0
      color {colorUseRule=UseColor_Specified, specifiedColor=<rgba 0 0 0 0>}
      cornerRadii=<0 0 0 0>
     }
     tintColor {colorUseRule=UseColor_Specified, specifiedColor=<rgba 1 1 1 1>}
     uVRegion {bIsValid=false, max=<0 0>, min=<0 0>}
    }}
   slot {class=CanvasPanelSlot, properties {layoutData {alignment=<0 0>, anchors {maximum=<0 0>, minimum=<0 0>}, offsets=<ltrb -148 -1052 916 980>}}}
  }
  BG_Particles {
   widgetClassPath=&Image
   bIsVariable=true
   parent=R_BG_Controller
   widget=BG_Particles
   properties {
    brush {
     bIsDynamicallyLoaded=false
     drawAs=Image
     imageType=FullColor
     mirroring=NoMirror
     resourceName=None
     resourceObject=/PistolCanary/Layout/ComicPanels_v2/ComicPanelTextures/Panel5/R_Frame/R_F_BG_Particles
     tiling=NoTile
     imageSize=<648 840>
     margin=<ltrb 0 0 0 0>
     outlineSettings {
      bUseBrushTransparency=false
      roundingType=HalfHeightRadius
      width=0
      color {colorUseRule=UseColor_Specified, specifiedColor=<rgba 0 0 0 0>}
      cornerRadii=<0 0 0 0>
     }
     tintColor {colorUseRule=UseColor_Specified, specifiedColor=<rgba 1 1 1 1>}
     uVRegion {bIsValid=false, max=<0 0>, min=<0 0>}
    }
    renderTransform {angle=0, scale=<2 2>, shear=<0 0>, translation=<0 0>}
   }
   slot {
    class=CanvasPanelSlot
    properties {layoutData {alignment=<0 0>, anchors {maximum=<0 0>, minimum=<0 0>}, offsets=<ltrb -469.695 -1319.6691 648 840>}}
   }
  }
  Building_v02 {
   widgetClassPath=&Image
   bIsVariable=true
   parent=R_BG_Controller
   widget=Building_v02
   properties {
    brush {
     bIsDynamicallyLoaded=false
     drawAs=Image
     imageType=FullColor
     mirroring=NoMirror
     resourceName=None
     resourceObject=/PistolCanary/Layout/ComicPanels_v2/ComicPanelTextures/Panel5/R_Frame/R_F_Building_Solo
     tiling=NoTile
     imageSize=<512 488>
     margin=<ltrb 0 0 0 0>
     outlineSettings {
      bUseBrushTransparency=false
      roundingType=HalfHeightRadius
      width=0
      color {colorUseRule=UseColor_Specified, specifiedColor=<rgba 0 0 0 0>}
      cornerRadii=<0 0 0 0>
     }
     tintColor {colorUseRule=UseColor_Specified, specifiedColor=<rgba 1 1 1 1>}
     uVRegion {bIsValid=false, max=<0 0>, min=<0 0>}
    }
    renderTransform {angle=0, scale=<2 2>, shear=<0 0>, translation=<0 0>}
   }
   slot {
    class=CanvasPanelSlot
    properties {layoutData {alignment=<0 0>, anchors {maximum=<0 0>, minimum=<0 0>}, offsets=<ltrb -249.0764 -459.9039 512 488>}}
   }
  }
  R_Debris_Controller {
   widgetClassPath=&CanvasPanel
   parent=R_BG_Controller
   widget=R_Debris_Controller
   slot {
    class=CanvasPanelSlot
    properties {layoutData {alignment=<0 0>, anchors {maximum=<0 0>, minimum=<0 0>}, offsets=<ltrb -792 184 182.2857 172.8572>}}
   }
  }
  Debris {
   widgetClassPath=&Image
   bIsVariable=true
   parent=R_Debris_Controller
   widget=Debris
   properties {brush {
     bIsDynamicallyLoaded=false
     drawAs=Image
     imageType=FullColor
     mirroring=NoMirror
     resourceName=None
     resourceObject=/PistolCanary/Layout/ComicPanels_v2/ComicPanelTextures/Panel5/R_Frame/R_F_Debris_50percent
     tiling=NoTile
     imageSize=<560 508>
     margin=<ltrb 0 0 0 0>
     outlineSettings {
      bUseBrushTransparency=false
      roundingType=HalfHeightRadius
      width=0
      color {colorUseRule=UseColor_Specified, specifiedColor=<rgba 0 0 0 0>}
      cornerRadii=<0 0 0 0>
     }
     tintColor {colorUseRule=UseColor_Specified, specifiedColor=<rgba 1 1 1 1>}
     uVRegion {bIsValid=false, max=<0 0>, min=<0 0>}
    }}
   slot {class=CanvasPanelSlot, properties {layoutData {alignment=<0 0>, anchors {maximum=<0 0>, minimum=<0 0>}, offsets=<ltrb 16 -880 1116 1012>}}}
  }
  Boom_01 {
   widgetClassPath=&Image
   bIsVariable=true
   parent=R_BG_Controller
   widget=Boom_01
   properties {brush {
     bIsDynamicallyLoaded=false
     drawAs=Image
     imageType=FullColor
     mirroring=NoMirror
     resourceName=None
     resourceObject=/PistolCanary/Layout/ComicPanels_v2/ComicPanelTextures/Panel5/R_Frame/R_F_Boom_01_50percent
     tiling=NoTile
     imageSize=<372 288>
     margin=<ltrb 0 0 0 0>
     outlineSettings {
      bUseBrushTransparency=false
      roundingType=HalfHeightRadius
      width=0
      color {colorUseRule=UseColor_Specified, specifiedColor=<rgba 0 0 0 0>}
      cornerRadii=<0 0 0 0>
     }
     tintColor {colorUseRule=UseColor_Specified, specifiedColor=<rgba 1 1 1 1>}
     uVRegion {bIsValid=false, max=<0 0>, min=<0 0>}
    }}
   slot {class=CanvasPanelSlot, properties {layoutData {alignment=<0 0>, anchors {maximum=<0 0>, minimum=<0 0>}, offsets=<ltrb -840 -532 744 576>}}}
  }
  Ping_01 {
   widgetClassPath=&Image
   bIsVariable=true
   parent=R_BG_Controller
   widget=Ping_01
   properties {brush {
     bIsDynamicallyLoaded=false
     drawAs=Image
     imageType=FullColor
     mirroring=NoMirror
     resourceName=None
     resourceObject=/PistolCanary/Layout/ComicPanels_v2/ComicPanelTextures/Panel5/R_Frame/R_F_Ping_01
     tiling=NoTile
     imageSize=<436 268>
     margin=<ltrb 0 0 0 0>
     outlineSettings {
      bUseBrushTransparency=false
      roundingType=HalfHeightRadius
      width=0
      color {colorUseRule=UseColor_Specified, specifiedColor=<rgba 0 0 0 0>}
      cornerRadii=<0 0 0 0>
     }
     tintColor {colorUseRule=UseColor_Specified, specifiedColor=<rgba 1 1 1 1>}
     uVRegion {bIsValid=false, max=<0 0>, min=<0 0>}
    }}
   slot {
    class=CanvasPanelSlot
    properties {bAutoSize=true, layoutData {alignment=<0 0>, anchors {maximum=<0 0>, minimum=<0 0>}, offsets=<ltrb -408 -696 100 30>}}
   }
  }
  Ping_02 {
   widgetClassPath=&Image
   bIsVariable=true
   parent=R_BG_Controller
   widget=Ping_02
   properties {brush {
     bIsDynamicallyLoaded=false
     drawAs=Image
     imageType=FullColor
     mirroring=NoMirror
     resourceName=None
     resourceObject=/PistolCanary/Layout/ComicPanels_v2/ComicPanelTextures/Panel5/R_Frame/R_F_Ping_02
     tiling=NoTile
     imageSize=<308 188>
     margin=<ltrb 0 0 0 0>
     outlineSettings {
      bUseBrushTransparency=false
      roundingType=HalfHeightRadius
      width=0
      color {colorUseRule=UseColor_Specified, specifiedColor=<rgba 0 0 0 0>}
      cornerRadii=<0 0 0 0>
     }
     tintColor {colorUseRule=UseColor_Specified, specifiedColor=<rgba 1 1 1 1>}
     uVRegion {bIsValid=false, max=<0 0>, min=<0 0>}
    }}
   slot {
    class=CanvasPanelSlot
    properties {bAutoSize=true, layoutData {alignment=<0 0>, anchors {maximum=<0 0>, minimum=<0 0>}, offsets=<ltrb -164 -928 100 30>}}
   }
  }
  Bang_01 {
   widgetClassPath=&Image
   bIsVariable=true
   parent=R_BG_Controller
   widget=Bang_01
   properties {brush {
     bIsDynamicallyLoaded=false
     drawAs=Image
     imageType=FullColor
     mirroring=NoMirror
     resourceName=None
     resourceObject=/PistolCanary/Layout/ComicPanels_v2/ComicPanelTextures/Panel5/R_Frame/R_F_Bang_01
     tiling=NoTile
     imageSize=<432 368>
     margin=<ltrb 0 0 0 0>
     outlineSettings {
      bUseBrushTransparency=false
      roundingType=HalfHeightRadius
      width=0
      color {colorUseRule=UseColor_Specified, specifiedColor=<rgba 0 0 0 0>}
      cornerRadii=<0 0 0 0>
     }
     tintColor {colorUseRule=UseColor_Specified, specifiedColor=<rgba 1 1 1 1>}
     uVRegion {bIsValid=false, max=<0 0>, min=<0 0>}
    }}
   slot {
    class=CanvasPanelSlot
    properties {bAutoSize=true, layoutData {alignment=<0 0>, anchors {maximum=<0 0>, minimum=<0 0>}, offsets=<ltrb -632 -1000 100 30>}}
   }
  }
  Boom_02 {
   widgetClassPath=&Image
   bIsVariable=true
   parent=R_BG_Controller
   widget=Boom_02
   properties {brush {
     bIsDynamicallyLoaded=false
     drawAs=Image
     imageType=FullColor
     mirroring=NoMirror
     resourceName=None
     resourceObject=/PistolCanary/Layout/ComicPanels_v2/ComicPanelTextures/Panel5/R_Frame/R_F_Boom_02
     tiling=NoTile
     imageSize=<408 424>
     margin=<ltrb 0 0 0 0>
     outlineSettings {
      bUseBrushTransparency=false
      roundingType=HalfHeightRadius
      width=0
      color {colorUseRule=UseColor_Specified, specifiedColor=<rgba 0 0 0 0>}
      cornerRadii=<0 0 0 0>
     }
     tintColor {colorUseRule=UseColor_Specified, specifiedColor=<rgba 1 1 1 1>}
     uVRegion {bIsValid=false, max=<0 0>, min=<0 0>}
    }}
   slot {
    class=CanvasPanelSlot
    properties {bAutoSize=true, layoutData {alignment=<0 0>, anchors {maximum=<0 0>, minimum=<0 0>}, offsets=<ltrb -956 -1736 100 30>}}
   }
  }
  R_Kill_Controller {
   widgetClassPath=&CanvasPanel
   parent=R_Frame
   widget=R_Kill_Controller
   slot {
    class=CanvasPanelSlot
    properties {layoutData {alignment=<0 0>, anchors {maximum=<0 0>, minimum=<0 0>}, offsets=<ltrb 2084 -120 144.5714 142>}}
   }
  }
  R_Hand_Controller {
   widgetClassPath=&CanvasPanel
   parent=R_Kill_Controller
   widget=R_Hand_Controller
   slot {
    class=CanvasPanelSlot
    properties {layoutData {alignment=<0 0>, anchors {maximum=<0 0>, minimum=<0 0>}, offsets=<ltrb -448 136 70.6122 64.449>}}
   }
  }
  R_Muzzle_Controller {
   widgetClassPath=&CanvasPanel
   parent=R_Hand_Controller
   widget=R_Muzzle_Controller
   slot {
    class=CanvasPanelSlot
    properties {layoutData {alignment=<0 0>, anchors {maximum=<0 0>, minimum=<0 0>}, offsets=<ltrb -176 156 70.6122 64.449>}}
   }
  }
  R_Muzzle_Splash {
   widgetClassPath=&Image
   bIsVariable=true
   parent=R_Muzzle_Controller
   widget=R_Muzzle_Splash
   properties {brush {
     bIsDynamicallyLoaded=false
     drawAs=Image
     imageType=FullColor
     mirroring=NoMirror
     resourceName=None
     resourceObject=/PistolCanary/Layout/ComicPanels_v2/ComicPanelTextures/Panel5/R_Frame/R_F_R_Muzzle_Splash
     tiling=NoTile
     imageSize=<244 132>
     margin=<ltrb 0 0 0 0>
     outlineSettings {
      bUseBrushTransparency=false
      roundingType=HalfHeightRadius
      width=0
      color {colorUseRule=UseColor_Specified, specifiedColor=<rgba 0 0 0 0>}
      cornerRadii=<0 0 0 0>
     }
     tintColor {colorUseRule=UseColor_Specified, specifiedColor=<rgba 1 1 1 1>}
     uVRegion {bIsValid=false, max=<0 0>, min=<0 0>}
    }}
   slot {
    class=CanvasPanelSlot
    properties {bAutoSize=true, layoutData {alignment=<0 0>, anchors {maximum=<0 0>, minimum=<0 0>}, offsets=<ltrb -256 -148 100 30>}}
   }
  }
  R_Muzzle {
   widgetClassPath=&Image
   bIsVariable=true
   parent=R_Muzzle_Controller
   widget=R_Muzzle
   properties {brush {
     bIsDynamicallyLoaded=false
     drawAs=Image
     imageType=FullColor
     mirroring=NoMirror
     resourceName=None
     resourceObject=/PistolCanary/Layout/ComicPanels_v2/ComicPanelTextures/Panel5/R_Frame/R_F_R_Muzzle_Flare
     tiling=NoTile
     imageSize=<396 412>
     margin=<ltrb 0 0 0 0>
     outlineSettings {
      bUseBrushTransparency=false
      roundingType=HalfHeightRadius
      width=0
      color {colorUseRule=UseColor_Specified, specifiedColor=<rgba 0 0 0 0>}
      cornerRadii=<0 0 0 0>
     }
     tintColor {colorUseRule=UseColor_Specified, specifiedColor=<rgba 1 1 1 1>}
     uVRegion {bIsValid=false, max=<0 0>, min=<0 0>}
    }}
   slot {
    class=CanvasPanelSlot
    properties {bAutoSize=true, layoutData {alignment=<0 0>, anchors {maximum=<0 0>, minimum=<0 0>}, offsets=<ltrb -340 -16 100 30>}}
   }
  }
  R_Kill_Hand {
   widgetClassPath=&Image
   bIsVariable=true
   parent=R_Hand_Controller
   widget=R_Kill_Hand
   properties {brush {
     bIsDynamicallyLoaded=false
     drawAs=Image
     imageType=FullColor
     mirroring=NoMirror
     resourceName=None
     resourceObject=/PistolCanary/Layout/ComicPanels_v2/ComicPanelTextures/Panel5/R_Frame/R_F_R_Hand
     tiling=NoTile
     imageSize=<240 248>
     margin=<ltrb 0 0 0 0>
     outlineSettings {
      bUseBrushTransparency=false
      roundingType=HalfHeightRadius
      width=0
      color {colorUseRule=UseColor_Specified, specifiedColor=<rgba 0 0 0 0>}
      cornerRadii=<0 0 0 0>
     }
     tintColor {colorUseRule=UseColor_Specified, specifiedColor=<rgba 1 1 1 1>}
     uVRegion {bIsValid=false, max=<0 0>, min=<0 0>}
    }}
   slot {
    class=CanvasPanelSlot
    properties {bAutoSize=true, layoutData {alignment=<0 0>, anchors {maximum=<0 0>, minimum=<0 0>}, offsets=<ltrb -176 -16 100 30>}}
   }
  }
  L_Hand_Controller {
   widgetClassPath=&CanvasPanel
   parent=R_Kill_Controller
   widget=L_Hand_Controller
   slot {
    class=CanvasPanelSlot
    properties {layoutData {alignment=<0 0>, anchors {maximum=<0 0>, minimum=<0 0>}, offsets=<ltrb -324 120 70.6122 64.449>}}
   }
  }
  L_Muzzle_Controller {
   widgetClassPath=&CanvasPanel
   parent=L_Hand_Controller
   widget=L_Muzzle_Controller
   slot {
    class=CanvasPanelSlot
    properties {layoutData {alignment=<0 0>, anchors {maximum=<0 0>, minimum=<0 0>}, offsets=<ltrb -156 172 70.6122 64.449>}}
   }
  }
  L_Kill_Muzzle_Splash_B {
   widgetClassPath=&Image
   bIsVariable=true
   parent=L_Muzzle_Controller
   widget=L_Kill_Muzzle_Splash_B
   properties {brush {
     bIsDynamicallyLoaded=false
     drawAs=Image
     imageType=FullColor
     mirroring=NoMirror
     resourceName=None
     resourceObject=/PistolCanary/Layout/ComicPanels_v2/ComicPanelTextures/Panel5/R_Frame/R_F_L_Muzzle_Splash_Bottom
     tiling=NoTile
     imageSize=<260 252>
     margin=<ltrb 0 0 0 0>
     outlineSettings {
      bUseBrushTransparency=false
      roundingType=HalfHeightRadius
      width=0
      color {colorUseRule=UseColor_Specified, specifiedColor=<rgba 0 0 0 0>}
      cornerRadii=<0 0 0 0>
     }
     tintColor {colorUseRule=UseColor_Specified, specifiedColor=<rgba 1 1 1 1>}
     uVRegion {bIsValid=false, max=<0 0>, min=<0 0>}
    }}
   slot {
    class=CanvasPanelSlot
    properties {layoutData {alignment=<0 0>, anchors {maximum=<0 0>, minimum=<0 0>}, offsets=<ltrb -317.0791 355.1543 260 252>}}
   }
  }
  L_Kill_Muzzle_Splash_T {
   widgetClassPath=&Image
   bIsVariable=true
   parent=L_Muzzle_Controller
   widget=L_Kill_Muzzle_Splash_T
   properties {brush {
     bIsDynamicallyLoaded=false
     drawAs=Image
     imageType=FullColor
     mirroring=NoMirror
     resourceName=None
     resourceObject=/PistolCanary/Layout/ComicPanels_v2/ComicPanelTextures/Panel5/R_Frame/R_F_L_Muzzle_Splash_Top
     tiling=NoTile
     imageSize=<24 64>
     margin=<ltrb 0 0 0 0>
     outlineSettings {
      bUseBrushTransparency=false
      roundingType=HalfHeightRadius
      width=0
      color {colorUseRule=UseColor_Specified, specifiedColor=<rgba 0 0 0 0>}
      cornerRadii=<0 0 0 0>
     }
     tintColor {colorUseRule=UseColor_Specified, specifiedColor=<rgba 1 1 1 1>}
     uVRegion {bIsValid=false, max=<0 0>, min=<0 0>}
    }}
   slot {
    class=CanvasPanelSlot
    properties {layoutData {alignment=<0 0>, anchors {maximum=<0 0>, minimum=<0 0>}, offsets=<ltrb 90.2062 75.1616 24 64>}}
   }
  }
  L_Kill_Muzzle {
   widgetClassPath=&Image
   bIsVariable=true
   parent=L_Muzzle_Controller
   widget=L_Kill_Muzzle
   properties {brush {
     bIsDynamicallyLoaded=false
     drawAs=Image
     imageType=FullColor
     mirroring=NoMirror
     resourceName=None
     resourceObject=/PistolCanary/Layout/ComicPanels_v2/ComicPanelTextures/Panel5/R_Frame/R_F_L_Muzzle_Flare
     tiling=NoTile
     imageSize=<296 332>
     margin=<ltrb 0 0 0 0>
     outlineSettings {
      bUseBrushTransparency=false
      roundingType=HalfHeightRadius
      width=0
      color {colorUseRule=UseColor_Specified, specifiedColor=<rgba 0 0 0 0>}
      cornerRadii=<0 0 0 0>
     }
     tintColor {colorUseRule=UseColor_Specified, specifiedColor=<rgba 1 1 1 1>}
     uVRegion {bIsValid=false, max=<0 0>, min=<0 0>}
    }}
   slot {
    class=CanvasPanelSlot
    properties {bAutoSize=true, layoutData {alignment=<0 0>, anchors {maximum=<0 0>, minimum=<0 0>}, offsets=<ltrb -184 0 100 30>}}
   }
  }
  L_Kill_Hand {
   widgetClassPath=&Image
   bIsVariable=true
   parent=L_Hand_Controller
   widget=L_Kill_Hand
   properties {brush {
     bIsDynamicallyLoaded=false
     drawAs=Image
     imageType=FullColor
     mirroring=NoMirror
     resourceName=None
     resourceObject=/PistolCanary/Layout/ComicPanels_v2/ComicPanelTextures/Panel5/R_Frame/R_F_L_Hand
     tiling=NoTile
     imageSize=<220 260>
     margin=<ltrb 0 0 0 0>
     outlineSettings {
      bUseBrushTransparency=false
      roundingType=HalfHeightRadius
      width=0
      color {colorUseRule=UseColor_Specified, specifiedColor=<rgba 0 0 0 0>}
      cornerRadii=<0 0 0 0>
     }
     tintColor {colorUseRule=UseColor_Specified, specifiedColor=<rgba 1 1 1 1>}
     uVRegion {bIsValid=false, max=<0 0>, min=<0 0>}
    }}
   slot {
    class=CanvasPanelSlot
    properties {bAutoSize=true, layoutData {alignment=<0 0>, anchors {maximum=<0 0>, minimum=<0 0>}, offsets=<ltrb -152 -20 100 30>}}
   }
  }
  R_Kill {
   widgetClassPath=&Image
   bIsVariable=true
   parent=R_Kill_Controller
   widget=R_Kill
   properties {brush {
     bIsDynamicallyLoaded=false
     drawAs=Image
     imageType=FullColor
     mirroring=NoMirror
     resourceName=None
     resourceObject=/PistolCanary/Layout/ComicPanels_v2/ComicPanelTextures/Panel5/R_Frame/R_F_Kill_Body_30percent
     tiling=NoTile
     imageSize=<332 360>
     margin=<ltrb 0 0 0 0>
     outlineSettings {
      bUseBrushTransparency=false
      roundingType=HalfHeightRadius
      width=0
      color {colorUseRule=UseColor_Specified, specifiedColor=<rgba 0 0 0 0>}
      cornerRadii=<0 0 0 0>
     }
     tintColor {colorUseRule=UseColor_Specified, specifiedColor=<rgba 1 1 1 1>}
     uVRegion {bIsValid=false, max=<0 0>, min=<0 0>}
    }}
   slot {class=CanvasPanelSlot, properties {layoutData {alignment=<0 0>, anchors {maximum=<0 0>, minimum=<0 0>}, offsets=<ltrb -832 -388 1096 1192>}}}
  }
  Ref {
   widgetClassPath=&Image
   bIsVariable=true
   parent=Panel_05
   widget=Ref
   properties {
    brush {
     bIsDynamicallyLoaded=false
     drawAs=Image
     imageType=FullColor
     mirroring=NoMirror
     resourceName=None
     resourceObject=/PistolCanary/Layout/ComicPanels_v2/Screenshots/Scene_08
     tiling=NoTile
     imageSize=<1920 1080>
     margin=<ltrb 0 0 0 0>
     outlineSettings {
      bUseBrushTransparency=false
      roundingType=HalfHeightRadius
      width=0
      color {colorUseRule=UseColor_Specified, specifiedColor=<rgba 0 0 0 0>}
      cornerRadii=<0 0 0 0>
     }
     tintColor {colorUseRule=UseColor_Specified, specifiedColor=<rgba 1 1 1 1>}
     uVRegion {bIsValid=false, max=<0 0>, min=<0 0>}
    }
    colorAndOpacity=<rgba 1 1 1 0.3>
   }
   slot {class=CanvasPanelSlot, properties {bAutoSize=true, zOrder=100, layoutData=~}}
  }
 ]
}
