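# Comic panel 6 widget: a two-frame animated scene with Myst figures, crows and smoke, built from layered images.
# Nested CanvasPanel "controllers" act as cutout rigs, so bodies and hands transform separately.
# P6_L_Frame and P6_R_Frame slide layers in, drift them slowly, then exit frames up and down respectively after about eight to nine seconds.
# The graph has only empty events, and a translucent Scene_09 screenshot overlays everything as "Ref".
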WidgetBlueprint /PistolCanary/Layout/ComicPanels_v2/Blueprints/Panel_06_v02
animations [
 P6_L_Frame {
  binding_count=5
  length_seconds=9.4333
  object=&self:P6_L_Frame
  sequence {
   bindings [
    L_Myst_R_Hand_Controller {
     id=#1
     possessed_class=CanvasPanel
     tracks [{
       type=MovieScene2DTransformTrack
       sections [{
         type=MovieScene2DTransformSection
         end_frame=566
         start_frame=0
         channels [
          Translation.X {type=MovieSceneScriptingFloatChannel, default=0}
          Translation.Y {type=MovieSceneScriptingFloatChannel, default=0}
          Angle {type=MovieSceneScriptingFloatChannel, keys [[0,0],[566,-22.8202]]}
          Scale.X {type=MovieSceneScriptingFloatChannel, default=1}
          Scale.Y {type=MovieSceneScriptingFloatChannel, default=1}
          Shear.X {type=MovieSceneScriptingFloatChannel, default=0}
          Shear.Y {type=MovieSceneScriptingFloatChannel, default=0}
         ]
        }]
      }]
    }
    L_BG_Controller {
     id=#2
     possessed_class=CanvasPanel
     tracks [{
       type=MovieScene2DTransformTrack
       sections [{
         type=MovieScene2DTransformSection
         end_frame=15
         start_frame=0
         channels [
          Translation.X {type=MovieSceneScriptingFloatChannel, keys [[0,-461.2941],[15,0]]}
          Translation.Y {type=MovieSceneScriptingFloatChannel, keys [[0,2072.2954],[15,0]]}
          Angle {type=MovieSceneScriptingFloatChannel, default=0}
          Scale.X {type=MovieSceneScriptingFloatChannel, default=1}
          Scale.Y {type=MovieSceneScriptingFloatChannel, default=1}
          Shear.X {type=MovieSceneScriptingFloatChannel, default=0}
          Shear.Y {type=MovieSceneScriptingFloatChannel, default=0}
         ]
        }]
      }]
    }
    L_Frame {
     id=#3
     possessed_class=CanvasPanel
     tracks [{
       type=MovieScene2DTransformTrack
       sections [{
         type=MovieScene2DTransformSection
         end_frame=566
         start_frame=0
         channels [
          Translation.X {type=MovieSceneScriptingFloatChannel, const [0,3]}
          Translation.Y {type=MovieSceneScriptingFloatChannel, keys [[0,0],[555.965,0],[566,-1259.4069]]}
          Angle {type=MovieSceneScriptingFloatChannel, default=0}
          Scale.X {type=MovieSceneScriptingFloatChannel, default=0.55}
          Scale.Y {type=MovieSceneScriptingFloatChannel, default=0.55}
          Shear.X {type=MovieSceneScriptingFloatChannel, default=0}
          Shear.Y {type=MovieSceneScriptingFloatChannel, default=0}
         ]
        }]
      }]
    }
    L_Myst_L_Hand_Controller {
     id=#4
     possessed_class=CanvasPanel
     tracks [{
       type=MovieScene2DTransformTrack
       sections [{
         type=MovieScene2DTransformSection
         end_frame=566
         start_frame=0
         channels [
          Translation.X {type=MovieSceneScriptingFloatChannel, default=0}
          Translation.Y {type=MovieSceneScriptingFloatChannel, default=0}
          Angle {type=MovieSceneScriptingFloatChannel, keys [[0,0],[566,17.8193]]}
          Scale.X {type=MovieSceneScriptingFloatChannel, default=1}
          Scale.Y {type=MovieSceneScriptingFloatChannel, default=1}
          Shear.X {type=MovieSceneScriptingFloatChannel, default=0}
          Shear.Y {type=MovieSceneScriptingFloatChannel, default=0}
         ]
        }]
      }]
    }
    L_Myst_Controller {
     id=#5
     possessed_class=CanvasPanel
     tracks [{
       type=MovieScene2DTransformTrack
       sections [{
         type=MovieScene2DTransformSection
         end_frame=566
         start_frame=0
         channels [
          Translation.X {type=MovieSceneScriptingFloatChannel, keys [[0,-859.5015],[15.053,-859.5015],[40.142,0],[566,83.6651]]}
          Translation.Y {type=MovieSceneScriptingFloatChannel, keys [[0,1166.1135],[15.053,1166.1135],[40.142,0],[566,-78.0464]]}
          Angle {type=MovieSceneScriptingFloatChannel, keys [[0,-26.3807],[15.053,-26.3807],[40.142,0],[566,3.5728]]}
          Scale.X {type=MovieSceneScriptingFloatChannel, keys [[0,0.4],[15.053,0.4],[40.142,1],[566,1.1]]}
          Scale.Y {type=MovieSceneScriptingFloatChannel, keys [[0,0.4],[15.053,0.4],[40.142,1],[566,1.1]]}
          Shear.X {type=MovieSceneScriptingFloatChannel, default=0}
          Shear.Y {type=MovieSceneScriptingFloatChannel, default=0}
         ]
        }]
      }]
    }
   ]
   display_rate {numerator=60}
  }
 }
 P6_R_Frame {
  binding_count=9
  length_seconds=9.4
  object=&self:P6_R_Frame
  sequence {
   bindings [
    R_Smoke {
     id=#6
     possessed_class=Image
     tracks [{
       type=MovieScene2DTransformTrack
       sections [{
         type=MovieScene2DTransformSection
         end_frame=564
         start_frame=0
         channels [
          Translation.X {type=MovieSceneScriptingFloatChannel, default=0, const [0,2]}
          Translation.Y {type=MovieSceneScriptingFloatChannel, default=0, keys [[0,0],[564,-27.5976]]}
          Angle {type=MovieSceneScriptingFloatChannel, default=0}
          Scale.X {type=MovieSceneScriptingFloatChannel, keys [[0,1],[564,1.05]]}
          Scale.Y {type=MovieSceneScriptingFloatChannel, keys [[0,1],[564,1.05]]}
          Shear.X {type=MovieSceneScriptingFloatChannel, default=0}
          Shear.Y {type=MovieSceneScriptingFloatChannel, default=0}
         ]
        }]
      }]
    }
    R_Myst_R_Hand_Controller {
     id=#7
     possessed_class=CanvasPanel
     tracks [{
       type=MovieScene2DTransformTrack
       sections [{
         type=MovieScene2DTransformSection
         end_frame=564
         start_frame=0
         channels [
          Translation.X {type=MovieSceneScriptingFloatChannel, default=0}
          Translation.Y {type=MovieSceneScriptingFloatChannel, default=0}
          Angle {type=MovieSceneScriptingFloatChannel, keys [[0,0],[564,-32.1323]]}
          Scale.X {type=MovieSceneScriptingFloatChannel, default=1}
          Scale.Y {type=MovieSceneScriptingFloatChannel, default=1}
          Shear.X {type=MovieSceneScriptingFloatChannel, default=0}
          Shear.Y {type=MovieSceneScriptingFloatChannel, default=0}
         ]
        }]
      }]
    }
    R_Myst_Controller {
     id=#8
     possessed_class=CanvasPanel
     tracks [{
       type=MovieScene2DTransformTrack
       sections [{
         type=MovieScene2DTransformSection
         end_frame=564
         start_frame=0
         channels [
          Translation.X {type=MovieSceneScriptingFloatChannel, keys [[0,695.9243],[15,695.9243],[30,0],[564,-55.8486]]}
          Translation.Y {type=MovieSceneScriptingFloatChannel, keys [[0,1683.7428],[15,1683.7428],[30,0],[564,-63.74]]}
          Angle {type=MovieSceneScriptingFloatChannel, keys [[0,-10.0792],[15,-10.0792],[30,0],[564,-1.1996]]}
          Scale.X {type=MovieSceneScriptingFloatChannel, keys [[0,0.8],[15,0.8],[30,1],[564,1.1]]}
          Scale.Y {type=MovieSceneScriptingFloatChannel, keys [[0,0.8],[15,0.8],[30,1],[564,1.1]]}
          Shear.X {type=MovieSceneScriptingFloatChannel, default=0}
          Shear.Y {type=MovieSceneScriptingFloatChannel, default=0}
         ]
        }]
      }]
    }
    BG_Crows {
     id=#9
     possessed_class=Image
     tracks [{
       type=MovieScene2DTransformTrack
       sections [{
         type=MovieScene2DTransformSection
         end_frame=564
         start_frame=0
         channels [
          Translation.X {type=MovieSceneScriptingFloatChannel, default=0, const [0,2]}
          Translation.Y {type=MovieSceneScriptingFloatChannel, default=0, keys [[0,0],[564,-36.9725]]}
          Angle {type=MovieSceneScriptingFloatChannel, keys [[0,0],[564,10.6173]]}
          Scale.X {type=MovieSceneScriptingFloatChannel, keys [[0,1.1],[564,1.2]]}
          Scale.Y {type=MovieSceneScriptingFloatChannel, keys [[0,1.1],[564,1.2]]}
          Shear.X {type=MovieSceneScriptingFloatChannel, default=0}
          Shear.Y {type=MovieSceneScriptingFloatChannel, default=0}
         ]
        }]
      }]
    }
    L_Crow_Controller {
     id=#10
     possessed_class=CanvasPanel
     tracks [{
       type=MovieScene2DTransformTrack
       sections [{
         type=MovieScene2DTransformSection
         end_frame=564
         start_frame=0
         channels [
          Translation.X {type=MovieSceneScriptingFloatChannel, keys [[0,-791.5851],[23,-791.5851],[38,0],[564,167.005]]}
          Translation.Y {type=MovieSceneScriptingFloatChannel, keys [[0,821.7805],[23,821.7805],[38,0],[564,-129.818]]}
          Angle {type=MovieSceneScriptingFloatChannel, keys [[0,15.7863],[23,15.7863],[38,0],[564,-8.245]]}
          Scale.X {type=MovieSceneScriptingFloatChannel, keys [[0,0.7],[15,0.7],[30,0.8],[564,0.9]]}
          Scale.Y {type=MovieSceneScriptingFloatChannel, keys [[0,0.7],[15,0.7],[30,0.8],[564,0.9]]}
          Shear.X {type=MovieSceneScriptingFloatChannel, default=0}
          Shear.Y {type=MovieSceneScriptingFloatChannel, default=0}
         ]
        }]
      }]
    }
    R_Frame {
     id=#11
     possessed_class=CanvasPanel
     tracks [{
       type=MovieScene2DTransformTrack
       sections [{
         type=MovieScene2DTransformSection
         end_frame=564
         start_frame=0
         channels [
          Translation.X {type=MovieSceneScriptingFloatChannel, const [0,3]}
          Translation.Y {type=MovieSceneScriptingFloatChannel, keys [[0,0],[473,0],[483,1271.9932]]}
          Angle {type=MovieSceneScriptingFloatChannel, default=0}
          Scale.X {type=MovieSceneScriptingFloatChannel, default=0.55}
          Scale.Y {type=MovieSceneScriptingFloatChannel, default=0.55}
          Shear.X {type=MovieSceneScriptingFloatChannel, default=0}
          Shear.Y {type=MovieSceneScriptingFloatChannel, default=0}
         ]
        }]
      }]
    }
    R_BG_Controller {
     id=#12
     possessed_class=CanvasPanel
     tracks [{
       type=MovieScene2DTransformTrack
       sections [{
         type=MovieScene2DTransformSection
         end_frame=15
         start_frame=0
         channels [
          Translation.X {type=MovieSceneScriptingFloatChannel, const [0,2]}
          Translation.Y {type=MovieSceneScriptingFloatChannel, keys [[0,-2287.2463],[15,0]]}
          Angle {type=MovieSceneScriptingFloatChannel, default=0}
          Scale.X {type=MovieSceneScriptingFloatChannel, default=1}
          Scale.Y {type=MovieSceneScriptingFloatChannel, default=1}
          Shear.X {type=MovieSceneScriptingFloatChannel, default=0}
          Shear.Y {type=MovieSceneScriptingFloatChannel, default=0}
         ]
        }]
      }]
    }
    R_Myst_L_Hand_Controller {
     id=#13
     possessed_class=CanvasPanel
     tracks [{
       type=MovieScene2DTransformTrack
       sections [{
         type=MovieScene2DTransformSection
         end_frame=564
         start_frame=0
         channels [
          Translation.X {type=MovieSceneScriptingFloatChannel, default=0}
          Translation.Y {type=MovieSceneScriptingFloatChannel, default=0}
          Angle {type=MovieSceneScriptingFloatChannel, keys [[0,0],[564,18.1579]]}
          Scale.X {type=MovieSceneScriptingFloatChannel, default=1}
          Scale.Y {type=MovieSceneScriptingFloatChannel, default=1}
          Shear.X {type=MovieSceneScriptingFloatChannel, default=0}
          Shear.Y {type=MovieSceneScriptingFloatChannel, default=0}
         ]
        }]
      }]
    }
    R_Crow_Controller {
     id=#14
     possessed_class=CanvasPanel
     tracks [{
       type=MovieScene2DTransformTrack
       sections [{
         type=MovieScene2DTransformSection
         end_frame=564
         start_frame=0
         channels [
          Translation.X {type=MovieSceneScriptingFloatChannel, keys [[0,-684.5891],[27,-684.5891],[45,0],[564,98.2496]]}
          Translation.Y {type=MovieSceneScriptingFloatChannel, keys [[0,-505.1443],[27,-505.1443],[45,0],[564,-61.1664]]}
          Angle {type=MovieSceneScriptingFloatChannel, keys [[0,31.1551],[27,31.1551],[45,0],[564,-5.2126]]}
          Scale.X {type=MovieSceneScriptingFloatChannel, keys [[0,0.5],[27,0.5],[45,0.8],[564,0.9]]}
          Scale.Y {type=MovieSceneScriptingFloatChannel, keys [[0,0.5],[27,0.5],[45,0.8],[564,0.9]]}
          Shear.X {type=MovieSceneScriptingFloatChannel, default=0}
          Shear.Y {type=MovieSceneScriptingFloatChannel, default=0}
         ]
        }]
      }]
    }
   ]
   display_rate {numerator=60}
  }
 }
]
graphs [EventGraph {nodes [
   Event PreConstruct()
   Event Construct()
   Event Tick()
  ]}]
variables []
widgets {
 info {inheritedWidgetCount=0, parentClass=&UserWidget, rootWidgetClass=&CanvasPanel, widgetCount=26}
 widgets [
  Panel_06 {widgetClassPath=&CanvasPanel, widget=Panel_06}
  L_Frame {
   widgetClassPath=&CanvasPanel
   parent=Panel_06
   widget=L_Frame
   properties {renderTransform {angle=0, scale=<0.55 0.55>, shear=<0 0>, translation=<0 0>}}
   slot {
    class=CanvasPanelSlot
    properties {bAutoSize=true, layoutData {alignment=<0 0>, anchors {maximum=<0.5 0.5>, minimum=<0.5 0.5>}, offsets=<ltrb -960 -540 100 30>}}
   }
  }
  L_BG_Controller {
   widgetClassPath=&CanvasPanel
   parent=L_Frame
   widget=L_BG_Controller
   slot {
    class=CanvasPanelSlot
    properties {bAutoSize=true, layoutData {alignment=<0 0>, anchors {maximum=<0.5 0.5>, minimum=<0.5 0.5>}, offsets=<ltrb -1746 -2072 169.0552 176.201>}}
   }
  }
  L_BG {
   widgetClassPath=&Image
   bIsVariable=true
   parent=L_BG_Controller
   widget=L_BG
   properties {brush {
     bIsDynamicallyLoaded=false
     drawAs=Image
     imageType=FullColor
     mirroring=NoMirror
     resourceName=None
     resourceObject=/PistolCanary/Layout/ComicPanels_v2/ComicPanelTextures/Panel6/L_Frame/P6_L_BG_80percent
     tiling=NoTile
     imageSize=<1280 1804>
     margin=<ltrb 0 0 0 0>
     outlineSettings {
      bUseBrushTransparency=false
      roundingType=HalfHeightRadius
      width=0
      color {colorUseRule=UseColor_Specified, specifiedColor=<rgba 0 0 0 0>}
      cornerRadii=<0 0 0 0>
     }
     tintColor {colorUseRule=UseColor_Specified, specifiedColor=<rgba 1 1 1 1>}
     uVRegion {bIsValid=false, max=<0 0>, min=<0 0>}
    }}
   slot {class=CanvasPanelSlot, properties {layoutData {alignment=<0 0>, anchors {maximum=<0 0>, minimum=<0 0>}, offsets=<ltrb 0 0 1604 2256>}}}
  }
  L_Myst_Controller {
   widgetClassPath=&CanvasPanel
   parent=L_Frame
   widget=L_Myst_Controller
   slot {
    class=CanvasPanelSlot
    properties {layoutData {alignment=<0 0>, anchors {maximum=<0.5 0.5>, minimum=<0.5 0.5>}, offsets=<ltrb -842 -916 169.0552 176.201>}}
   }
  }
  L_Myst_Body {
   widgetClassPath=&Image
   bIsVariable=true
   parent=L_Myst_Controller
   widget=L_Myst_Body
   properties {brush {
     bIsDynamicallyLoaded=false
     drawAs=Image
     imageType=FullColor
     mirroring=NoMirror
     resourceName=None
     resourceObject=/PistolCanary/Layout/ComicPanels_v2/ComicPanelTextures/Panel6/L_Frame/L_Myst_Body_60percent
     tiling=NoTile
     imageSize=<928 1260>
     margin=<ltrb 0 0 0 0>
     outlineSettings {
      bUseBrushTransparency=false
      roundingType=HalfHeightRadius
      width=0
      color {colorUseRule=UseColor_Specified, specifiedColor=<rgba 0 0 0 0>}
      cornerRadii=<0 0 0 0>
     }
     tintColor {colorUseRule=UseColor_Specified, specifiedColor=<rgba 1 1 1 1>}
     uVRegion {bIsValid=false, max=<0 0>, min=<0 0>}
    }}
   slot {
    class=CanvasPanelSlot
    properties {layoutData {alignment=<0 0>, anchors {maximum=<0 0>, minimum=<0 0>}, offsets=<ltrb -812 -1096 1544 2096>}}
   }
  }
  L_Myst_R_Hand_Controller {
   widgetClassPath=&CanvasPanel
   parent=L_Myst_Controller
   widget=L_Myst_R_Hand_Controller
   slot {
    class=CanvasPanelSlot
    properties {layoutData {alignment=<0 0>, anchors {maximum=<0.5 0.5>, minimum=<0.5 0.5>}, offsets=<ltrb -424.5276 -132.1005 157.9048 147.3333>}}
   }
  }
  L_Myst_R_Hand {
   widgetClassPath=&Image
   bIsVariable=true
   parent=L_Myst_R_Hand_Controller
   widget=L_Myst_R_Hand
   properties {brush {
     bIsDynamicallyLoaded=false
     drawAs=Image
     imageType=FullColor
     mirroring=NoMirror
     resourceName=None
     resourceObject=/PistolCanary/Layout/ComicPanels_v2/ComicPanelTextures/Panel6/L_Frame/L_Myst_R_Hand
     tiling=NoTile
     imageSize=<444 412>
     margin=<ltrb 0 0 0 0>
     outlineSettings {
      bUseBrushTransparency=false
      roundingType=HalfHeightRadius
      width=0
      color {colorUseRule=UseColor_Specified, specifiedColor=<rgba 0 0 0 0>}
      cornerRadii=<0 0 0 0>
     }
     tintColor {colorUseRule=UseColor_Specified, specifiedColor=<rgba 1 1 1 1>}
     uVRegion {bIsValid=false, max=<0 0>, min=<0 0>}
    }}
   slot {
    class=CanvasPanelSlot
    properties {bAutoSize=true, layoutData {alignment=<0 0>, anchors {maximum=<0 0>, minimum=<0 0>}, offsets=<ltrb -172 -160 100 30>}}
   }
  }
  L_Myst_L_Hand_Controller {
   widgetClassPath=&CanvasPanel
   parent=L_Myst_Controller
   widget=L_Myst_L_Hand_Controller
   slot {
    class=CanvasPanelSlot
    properties {layoutData {alignment=<0 0>, anchors {maximum=<0.5 0.5>, minimum=<0.5 0.5>}, offsets=<ltrb 175.4724 -864.1005 157.9048 147.3333>}}
   }
  }
  L_Myst_L_Hand {
   widgetClassPath=&Image
   bIsVariable=true
   parent=L_Myst_L_Hand_Controller
   widget=L_Myst_L_Hand
   properties {brush {
     bIsDynamicallyLoaded=false
     drawAs=Image
     imageType=FullColor
     mirroring=NoMirror
     resourceName=None
     resourceObject=/PistolCanary/Layout/ComicPanels_v2/ComicPanelTextures/Panel6/L_Frame/L_Myst_L_Hand
     tiling=NoTile
     imageSize=<292 260>
     margin=<ltrb 0 0 0 0>
     outlineSettings {
      bUseBrushTransparency=false
      roundingType=HalfHeightRadius
      width=0
      color {colorUseRule=UseColor_Specified, specifiedColor=<rgba 0 0 0 0>}
      cornerRadii=<0 0 0 0>
     }
     tintColor {colorUseRule=UseColor_Specified, specifiedColor=<rgba 1 1 1 1>}
     uVRegion {bIsValid=false, max=<0 0>, min=<0 0>}
    }}
   slot {
    class=CanvasPanelSlot
    properties {bAutoSize=true, layoutData {alignment=<0 0>, anchors {maximum=<0 0>, minimum=<0 0>}, offsets=<ltrb -92 -48 100 30>}}
   }
  }
  R_Frame {
   widgetClassPath=&CanvasPanel
   parent=Panel_06
   widget=R_Frame
   properties {renderTransform {angle=0, scale=<0.55 0.55>, shear=<0 0>, translation=<0 0>}}
   slot {
    class=CanvasPanelSlot
    properties {bAutoSize=true, layoutData {alignment=<0 0>, anchors {maximum=<0.5 0.5>, minimum=<0.5 0.5>}, offsets=<ltrb -960 -540 100 30>}}
   }
  }
  R_BG_Controller {
   widgetClassPath=&CanvasPanel
   parent=R_Frame
   widget=R_BG_Controller
   slot {
    class=CanvasPanelSlot
    properties {bAutoSize=true, layoutData {alignment=<0 0>, anchors {maximum=<0 0>, minimum=<0 0>}, offsets=<ltrb 972 -2004 100 30>}}
   }
  }
  R_BG {
   widgetClassPath=&Image
   bIsVariable=true
   parent=R_BG_Controller
   widget=R_BG
   properties {brush {
     bIsDynamicallyLoaded=false
     drawAs=Image
     imageType=FullColor
     mirroring=NoMirror
     resourceName=None
     resourceObject=/PistolCanary/Layout/ComicPanels_v2/ComicPanelTextures/Panel6/R_Frame/P6_R_BG_80percent
     tiling=NoTile
     imageSize=<1420 1804>
     margin=<ltrb 0 0 0 0>
     outlineSettings {
      bUseBrushTransparency=false
      roundingType=HalfHeightRadius
      width=0
      color {colorUseRule=UseColor_Specified, specifiedColor=<rgba 0 0 0 0>}
      cornerRadii=<0 0 0 0>
     }
     tintColor {colorUseRule=UseColor_Specified, specifiedColor=<rgba 1 1 1 1>}
     uVRegion {bIsValid=false, max=<0 0>, min=<0 0>}
    }}
   slot {class=CanvasPanelSlot, properties {layoutData {alignment=<0 0>, anchors {maximum=<0 0>, minimum=<0 0>}, offsets=<ltrb 0 0 1776 2256>}}}
  }
  BG_Crows {
   widgetClassPath=&Image
   bIsVariable=true
   parent=R_BG_Controller
   widget=BG_Crows
   properties {
    brush {
     bIsDynamicallyLoaded=false
     drawAs=Image
     imageType=FullColor
     mirroring=NoMirror
     resourceName=None
     resourceObject=/PistolCanary/Layout/ComicPanels_v2/ComicPanelTextures/Panel6/R_Frame/P6_BG_Crows_60percent
     tiling=NoTile
     imageSize=<644 724>
     margin=<ltrb 0 0 0 0>
     outlineSettings {
      bUseBrushTransparency=false
      roundingType=HalfHeightRadius
      width=0
      color {colorUseRule=UseColor_Specified, specifiedColor=<rgba 0 0 0 0>}
      cornerRadii=<0 0 0 0>
     }
     tintColor {colorUseRule=UseColor_Specified, specifiedColor=<rgba 1 1 1 1>}
     uVRegion {bIsValid=false, max=<0 0>, min=<0 0>}
    }
    renderTransform {angle=0, scale=<1.1 1.1>, shear=<0 0>, translation=<0 0>}
   }
   slot {class=CanvasPanelSlot, properties {layoutData {alignment=<0 0>, anchors {maximum=<0 0>, minimum=<0 0>}, offsets=<ltrb 176 420 1076 1208>}}}
  }
  R_Smoke {
   widgetClassPath=&Image
   bIsVariable=true
   parent=R_BG_Controller
   widget=R_Smoke
   properties {brush {
     bIsDynamicallyLoaded=false
     drawAs=Image
     imageType=FullColor
     mirroring=NoMirror
     resourceName=None
     resourceObject=/PistolCanary/Layout/ComicPanels_v2/ComicPanelTextures/Panel6/R_Frame/P6_BG_Smoke_50percent
     tiling=NoTile
     imageSize=<544 560>
     margin=<ltrb 0 0 0 0>
     outlineSettings {
      bUseBrushTransparency=false
      roundingType=HalfHeightRadius
      width=0
      color {colorUseRule=UseColor_Specified, specifiedColor=<rgba 0 0 0 0>}
      cornerRadii=<0 0 0 0>
     }
     tintColor {colorUseRule=UseColor_Specified, specifiedColor=<rgba 1 1 1 1>}
     uVRegion {bIsValid=false, max=<0 0>, min=<0 0>}
    }}
   slot {class=CanvasPanelSlot, properties {layoutData {alignment=<0 0>, anchors {maximum=<0 0>, minimum=<0 0>}, offsets=<ltrb 592 1212 1092 1120>}}}
  }
  L_Crow_Controller {
   widgetClassPath=&CanvasPanel
   parent=R_Frame
   widget=L_Crow_Controller
   properties {renderTransform {angle=0, scale=<0.8 0.8>, shear=<0 0>, translation=<0 0>}}
   slot {
    class=CanvasPanelSlot
    properties {layoutData {alignment=<0 0>, anchors {maximum=<0 0>, minimum=<0 0>}, offsets=<ltrb 1276 -112 90.0952 88.6667>}}
   }
  }
  L_Crow {
   widgetClassPath=&Image
   bIsVariable=true
   parent=L_Crow_Controller
   widget=L_Crow
   properties {brush {
     bIsDynamicallyLoaded=false
     drawAs=Image
     imageType=FullColor
     mirroring=NoMirror
     resourceName=None
     resourceObject=/PistolCanary/Layout/ComicPanels_v2/ComicPanelTextures/Panel6/R_Frame/P6_L_Crow_50percent
     tiling=NoTile
     imageSize=<484 420>
     margin=<ltrb 0 0 0 0>
     outlineSettings {
      bUseBrushTransparency=false
      roundingType=HalfHeightRadius
      width=0
      color {colorUseRule=UseColor_Specified, specifiedColor=<rgba 0 0 0 0>}
      cornerRadii=<0 0 0 0>
     }
     tintColor {colorUseRule=UseColor_Specified, specifiedColor=<rgba 1 1 1 1>}
     uVRegion {bIsValid=false, max=<0 0>, min=<0 0>}
    }}
   slot {class=CanvasPanelSlot, properties {layoutData {alignment=<0 0>, anchors {maximum=<0 0>, minimum=<0 0>}, offsets=<ltrb -368 -588 968 840>}}}
  }
  R_Crow_Controller {
   widgetClassPath=&CanvasPanel
   parent=R_Frame
   widget=R_Crow_Controller
   properties {renderTransform {angle=0, scale=<0.8 0.8>, shear=<0 0>, translation=<0 0>}}
   slot {
    class=CanvasPanelSlot
    properties {layoutData {alignment=<0 0>, anchors {maximum=<0 0>, minimum=<0 0>}, offsets=<ltrb 1940 -1540 102.0317 90.7831>}}
   }
  }
  R_Crow {
   widgetClassPath=&Image
   bIsVariable=true
   parent=R_Crow_Controller
   widget=R_Crow
   properties {brush {
     bIsDynamicallyLoaded=false
     drawAs=Image
     imageType=FullColor
     mirroring=NoMirror
     resourceName=None
     resourceObject=/PistolCanary/Layout/ComicPanels_v2/ComicPanelTextures/Panel6/R_Frame/P6_R_Crow_50percent
     tiling=NoTile
     imageSize=<676 424>
     margin=<ltrb 0 0 0 0>
     outlineSettings {
      bUseBrushTransparency=false
      roundingType=HalfHeightRadius
      width=0
      color {colorUseRule=UseColor_Specified, specifiedColor=<rgba 0 0 0 0>}
      cornerRadii=<0 0 0 0>
     }
     tintColor {colorUseRule=UseColor_Specified, specifiedColor=<rgba 1 1 1 1>}
     uVRegion {bIsValid=false, max=<0 0>, min=<0 0>}
    }}
   slot {class=CanvasPanelSlot, properties {layoutData {alignment=<0 0>, anchors {maximum=<0 0>, minimum=<0 0>}, offsets=<ltrb -740 -540 1356 848>}}}
  }
  R_Myst_Controller {
   widgetClassPath=&CanvasPanel
   parent=R_Frame
   widget=R_Myst_Controller
   slot {
    class=CanvasPanelSlot
    properties {layoutData {alignment=<0 0>, anchors {maximum=<0 0>, minimum=<0 0>}, offsets=<ltrb 1688 -872 102.0529 94.8677>}}
   }
  }
  R_Myst_Body {
   widgetClassPath=&Image
   bIsVariable=true
   parent=R_Myst_Controller
   widget=R_Myst_Body
   properties {brush {
     bIsDynamicallyLoaded=false
     drawAs=Image
     imageType=FullColor
     mirroring=NoMirror
     resourceName=None
     resourceObject=/PistolCanary/Layout/ComicPanels_v2/ComicPanelTextures/Panel6/R_Frame/P6_R_Myst_Body_50percent
     tiling=NoTile
     imageSize=<308 880>
     margin=<ltrb 0 0 0 0>
     outlineSettings {
      bUseBrushTransparency=false
      roundingType=HalfHeightRadius
      width=0
      color {colorUseRule=UseColor_Specified, specifiedColor=<rgba 0 0 0 0>}
      cornerRadii=<0 0 0 0>
     }
     tintColor {colorUseRule=UseColor_Specified, specifiedColor=<rgba 1 1 1 1>}
     uVRegion {bIsValid=false, max=<0 0>, min=<0 0>}
    }}
   slot {class=CanvasPanelSlot, properties {layoutData {alignment=<0 0>, anchors {maximum=<0 0>, minimum=<0 0>}, offsets=<ltrb -236 -724 616 1760>}}}
  }
  R_Myst_R_Hand_Controller {
   widgetClassPath=&CanvasPanel
   parent=R_Myst_Controller
   widget=R_Myst_R_Hand_Controller
   slot {
    class=CanvasPanelSlot
    properties {layoutData {alignment=<0 0>, anchors {maximum=<0 0>, minimum=<0 0>}, offsets=<ltrb -252 -20 99.1429 89.4286>}}
   }
  }
  R_Myst_R_Hand {
   widgetClassPath=&Image
   bIsVariable=true
   parent=R_Myst_R_Hand_Controller
   widget=R_Myst_R_Hand
   properties {brush {
     bIsDynamicallyLoaded=false
     drawAs=Image
     imageType=FullColor
     mirroring=NoMirror
     resourceName=None
     resourceObject=/PistolCanary/Layout/ComicPanels_v2/ComicPanelTextures/Panel6/R_Frame/P6_R_Myst_R_Hand
     tiling=NoTile
     imageSize=<112 76>
     margin=<ltrb 0 0 0 0>
     outlineSettings {
      bUseBrushTransparency=false
      roundingType=HalfHeightRadius
      width=0
      color {colorUseRule=UseColor_Specified, specifiedColor=<rgba 0 0 0 0>}
      cornerRadii=<0 0 0 0>
     }
     tintColor {colorUseRule=UseColor_Specified, specifiedColor=<rgba 1 1 1 1>}
     uVRegion {bIsValid=false, max=<0 0>, min=<0 0>}
    }}
   slot {
    class=CanvasPanelSlot
    properties {bAutoSize=true, layoutData {alignment=<0 0>, anchors {maximum=<0 0>, minimum=<0 0>}, offsets=<ltrb -48 16 100 30>}}
   }
  }
  R_Myst_L_Hand_Controller {
   widgetClassPath=&CanvasPanel
   parent=R_Myst_Controller
   widget=R_Myst_L_Hand_Controller
   slot {
    class=CanvasPanelSlot
    properties {layoutData {alignment=<0 0>, anchors {maximum=<0 0>, minimum=<0 0>}, offsets=<ltrb 180 -220 99.1429 89.4286>}}
   }
  }
  R_Myst_L_Hand {
   widgetClassPath=&Image
   bIsVariable=true
   parent=R_Myst_L_Hand_Controller
   widget=R_Myst_L_Hand
   properties {brush {
     bIsDynamicallyLoaded=false
     drawAs=Image
     imageType=FullColor
     mirroring=NoMirror
     resourceName=None
     resourceObject=/PistolCanary/Layout/ComicPanels_v2/ComicPanelTextures/Panel6/R_Frame/P6_R_Myst_L_Hand
     tiling=NoTile
     imageSize=<136 112>
     margin=<ltrb 0 0 0 0>
     outlineSettings {
      bUseBrushTransparency=false
      roundingType=HalfHeightRadius
      width=0
      color {colorUseRule=UseColor_Specified, specifiedColor=<rgba 0 0 0 0>}
      cornerRadii=<0 0 0 0>
     }
     tintColor {colorUseRule=UseColor_Specified, specifiedColor=<rgba 1 1 1 1>}
     uVRegion {bIsValid=false, max=<0 0>, min=<0 0>}
    }}
   slot {
    class=CanvasPanelSlot
    properties {bAutoSize=true, layoutData {alignment=<0 0>, anchors {maximum=<0 0>, minimum=<0 0>}, offsets=<ltrb -20 -24 100 30>}}
   }
  }
  Ref {
   widgetClassPath=&Image
   bIsVariable=true
   parent=Panel_06
   widget=Ref
   properties {
    brush {
     bIsDynamicallyLoaded=false
     drawAs=Image
     imageType=FullColor
     mirroring=NoMirror
     resourceName=None
     resourceObject=/PistolCanary/Layout/ComicPanels_v2/Screenshots/Scene_09
     tiling=NoTile
     imageSize=<1920 1080>
     margin=<ltrb 0 0 0 0>
     outlineSettings {
      bUseBrushTransparency=false
      roundingType=HalfHeightRadius
      width=0
      color {colorUseRule=UseColor_Specified, specifiedColor=<rgba 0 0 0 0>}
      cornerRadii=<0 0 0 0>
     }
     tintColor {colorUseRule=UseColor_Specified, specifiedColor=<rgba 1 1 1 1>}
     uVRegion {bIsValid=false, max=<0 0>, min=<0 0>}
    }
    colorAndOpacity=<rgba 1 1 1 0.3>
   }
   slot {class=CanvasPanelSlot, properties {bAutoSize=true, zOrder=100, layoutData=~}}
  }
 ]
}
