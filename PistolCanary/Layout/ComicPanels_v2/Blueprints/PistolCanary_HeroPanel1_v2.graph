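# Pistol Canary hero panel 1: a two-sided comic-style UMG widget showing layered Synth artwork with hands, powers and sparks.
# Both sides start hidden; the play event runs the left or right animation by PanelSide and reveals it, with slide-in, slow drift and fade-out.
# The load event replaces default, mostly low-res brushes with index-matched soft textures from TexturesHigh on non-mobile only.
# The unload event clears the brushes and forces garbage collection.
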WidgetBlueprint /PistolCanary/Layout/ComicPanels_v2/Blueprints/PistolCanary_HeroPanel1_v2
animations [
 LeftPanelAnim {
  binding_count=10
  length_seconds=7.6167
  object=&self:LeftPanelAnim
  sequence {
   bindings [
    L_BG_Controller {
     id=#1
     possessed_class=CanvasPanel
     tracks [{
       type=MovieScene2DTransformTrack
       sections [{
         type=MovieScene2DTransformSection
         end_frame=15
         start_frame=0
         channels [
          Translation.X {type=MovieSceneScriptingFloatChannel, keys [[0,-607.2714],[15,117.2038]]}
          Translation.Y {type=MovieSceneScriptingFloatChannel, keys [[0,1146.4131],[15,24.4778]]}
          Angle {type=MovieSceneScriptingFloatChannel, default=0}
          Scale.X {type=MovieSceneScriptingFloatChannel, default=2, const [2.2,1]}
          Scale.Y {type=MovieSceneScriptingFloatChannel, default=2, const [2.2,1]}
          Shear.X {type=MovieSceneScriptingFloatChannel, default=0}
          Shear.Y {type=MovieSceneScriptingFloatChannel, default=0}
         ]
        }]
      }]
    }
    L_Synth_R_Power_Controller {
     id=#2
     possessed_class=CanvasPanel
     tracks [{
       type=MovieScene2DTransformTrack
       sections [{
         type=MovieScene2DTransformSection
         end_frame=457
         start_frame=0
         channels [
          Translation.X {type=MovieSceneScriptingFloatChannel, const [0,3]}
          Translation.Y {type=MovieSceneScriptingFloatChannel, const [0,3]}
          Angle {type=MovieSceneScriptingFloatChannel, keys [[0,0],[457,-10]]}
          Scale.X {type=MovieSceneScriptingFloatChannel, keys [[15.474,0],[36.106,1],[457,1.2]]}
          Scale.Y {type=MovieSceneScriptingFloatChannel, keys [[15.474,0],[36.106,1],[457,1.2]]}
          Shear.X {type=MovieSceneScriptingFloatChannel, default=0}
          Shear.Y {type=MovieSceneScriptingFloatChannel, default=0}
         ]
        }]
      }]
    }
    L_Frame {
     id=#3
     possessed_class=CanvasPanel
     tracks [
      {
       type=MovieScene2DTransformTrack
       sections [{
         type=MovieScene2DTransformSection
         end_frame=457
         start_frame=0
         channels [
          Translation.X {type=MovieSceneScriptingFloatChannel, default=0}
          Translation.Y {type=MovieSceneScriptingFloatChannel, keys [[446.978,0],[457,-1367.4224]]}
          Angle {type=MovieSceneScriptingFloatChannel, default=0}
          Scale.X {type=MovieSceneScriptingFloatChannel, default=1}
          Scale.Y {type=MovieSceneScriptingFloatChannel, default=1}
          Shear.X {type=MovieSceneScriptingFloatChannel, default=0}
          Shear.Y {type=MovieSceneScriptingFloatChannel, default=0}
         ]
        }]
      }
      {
       type=MovieSceneFloatTrack
       display_name=Render Opacity
       sections [{
         type=MovieSceneFloatSection
         end_frame=457
         start_frame=0
         channels [None {type=MovieSceneScriptingFloatChannel, keys [[0,0],[14,1],[448,1],[457,0]]}]
        }]
      }
     ]
    }
    L_Synth_L_Power_Controller {
     id=#4
     possessed_class=CanvasPanel
     tracks [{
       type=MovieScene2DTransformTrack
       sections [{
         type=MovieScene2DTransformSection
         end_frame=457
         start_frame=0
         channels [
          Translation.X {type=MovieSceneScriptingFloatChannel, const [0,3]}
          Translation.Y {type=MovieSceneScriptingFloatChannel, const [0,3]}
          Angle {type=MovieSceneScriptingFloatChannel, keys [[0,0],[457,10]]}
          Scale.X {type=MovieSceneScriptingFloatChannel, keys [[15.474,0],[36.106,1.1],[457,1.2]]}
          Scale.Y {type=MovieSceneScriptingFloatChannel, keys [[15.474,0],[36.106,1.1],[457,1.2]]}
          Shear.X {type=MovieSceneScriptingFloatChannel, default=0}
          Shear.Y {type=MovieSceneScriptingFloatChannel, default=0}
         ]
        }]
      }]
    }
    L_Synth_L_Hand_Controller {
     id=#5
     possessed_class=CanvasPanel
     tracks [{
       type=MovieScene2DTransformTrack
       sections [{
         type=MovieScene2DTransformSection
         end_frame=457
         start_frame=0
         channels [
          Translation.X {type=MovieSceneScriptingFloatChannel, default=0}
          Translation.Y {type=MovieSceneScriptingFloatChannel, default=0}
          Angle {type=MovieSceneScriptingFloatChannel, keys [[0,0],[457,18.7271]]}
          Scale.X {type=MovieSceneScriptingFloatChannel, default=1}
          Scale.Y {type=MovieSceneScriptingFloatChannel, default=1}
          Shear.X {type=MovieSceneScriptingFloatChannel, default=0}
          Shear.Y {type=MovieSceneScriptingFloatChannel, default=0}
         ]
        }]
      }]
    }
    L_Synth {
     id=#6
     possessed_class=CanvasPanel
     tracks [{
       type=MovieScene2DTransformTrack
       sections [{
         type=MovieScene2DTransformSection
         end_frame=457
         start_frame=0
         channels [
          Translation.X {type=MovieSceneScriptingFloatChannel, keys [[0,-707.7817],[15.474,-707.7817],[36.106,0],[457,40.3339]]}
          Translation.Y {type=MovieSceneScriptingFloatChannel, keys [[0,-388.8787],[15.474,-388.8787],[36.106,0],[457,-20.2096]]}
          Angle {type=MovieSceneScriptingFloatChannel, keys [[0,20.9972],[15.474,20.9972],[36.106,0],[457,-3]]}
          Scale.X {type=MovieSceneScriptingFloatChannel, keys [[0,0.2],[15.474,0.2],[36.106,1],[457,1.1]]}
          Scale.Y {type=MovieSceneScriptingFloatChannel, keys [[0,0.2],[15.474,0.2],[36.106,1],[457,1.1]]}
          Shear.X {type=MovieSceneScriptingFloatChannel, const [0,1]}
          Shear.Y {type=MovieSceneScriptingFloatChannel, const [0,1]}
         ]
        }]
      }]
    }
    BlackBar_Top {
     id=#7
     possessed_class=Image
     tracks [{
       type=MovieSceneFloatTrack
       display_name=Render Opacity
       sections [{type=MovieSceneFloatSection, end_frame=14, start_frame=0, channels [None {type=MovieSceneScriptingFloatChannel, keys [[0,0],[14,1]]}]}]
      }]
    }
    L_Sparks_Controller {
     id=#8
     possessed_class=CanvasPanel
     tracks [{
       type=MovieScene2DTransformTrack
       sections [{
         type=MovieScene2DTransformSection
         end_frame=457
         start_frame=0
         channels [
          Translation.X {type=MovieSceneScriptingFloatChannel, default=0}
          Translation.Y {type=MovieSceneScriptingFloatChannel, default=0}
          Angle {type=MovieSceneScriptingFloatChannel, default=0}
          Scale.X {type=MovieSceneScriptingFloatChannel, keys [[0,1],[457,1.2]]}
          Scale.Y {type=MovieSceneScriptingFloatChannel, keys [[0,1],[457,1.2]]}
          Shear.X {type=MovieSceneScriptingFloatChannel, default=0}
          Shear.Y {type=MovieSceneScriptingFloatChannel, default=0}
         ]
        }]
      }]
    }
    BlackBar_Bottom {
     id=#9
     possessed_class=Image
     tracks [{
       type=MovieSceneFloatTrack
       display_name=Render Opacity
       sections [{type=MovieSceneFloatSection, end_frame=14, start_frame=0, channels [None {type=MovieSceneScriptingFloatChannel, keys [[0,0],[14,1]]}]}]
      }]
    }
    L_Synth_R_Hand_Controller {
     id=#10
     possessed_class=CanvasPanel
     tracks [{
       type=MovieScene2DTransformTrack
       sections [{
         type=MovieScene2DTransformSection
         end_frame=457
         start_frame=0
         channels [
          Translation.X {type=MovieSceneScriptingFloatChannel, default=329.5019}
          Translation.Y {type=MovieSceneScriptingFloatChannel, default=0}
          Angle {type=MovieSceneScriptingFloatChannel, keys [[0,0],[457,-15.633]]}
          Scale.X {type=MovieSceneScriptingFloatChannel, default=1}
          Scale.Y {type=MovieSceneScriptingFloatChannel, default=1}
          Shear.X {type=MovieSceneScriptingFloatChannel, default=0}
          Shear.Y {type=MovieSceneScriptingFloatChannel, default=0}
         ]
        }]
      }]
    }
   ]
   display_rate {numerator=60}
  }
 }
 RightPanelAnim {
  binding_count=10
  length_seconds=6.2333
  object=&self:RightPanelAnim
  sequence {
   bindings [
    R_Embers {
     id=#11
     possessed_class=Image
     tracks [{
       type=MovieScene2DTransformTrack
       sections [{
         type=MovieScene2DTransformSection
         end_frame=374
         start_frame=0
         channels [
          Translation.X {type=MovieSceneScriptingFloatChannel, keys [[0,788.252],[12.303,788.252],[24.605,0],[374,-66.0925]]}
          Translation.Y {type=MovieSceneScriptingFloatChannel, keys [[0,152.7908],[12.303,152.7908],[24.605,0],[374,0]]}
          Angle {type=MovieSceneScriptingFloatChannel, default=0, keys [[0,9.4774],[12.303,9.4774],[24.605,0],[374,0]]}
          Scale.X {type=MovieSceneScriptingFloatChannel, keys [[0,0.8],[12.303,0.8],[24.605,1],[374,1.2]]}
          Scale.Y {type=MovieSceneScriptingFloatChannel, keys [[0,0.8],[12.303,0.8],[24.605,1],[374,1]]}
          Shear.X {type=MovieSceneScriptingFloatChannel, default=0}
          Shear.Y {type=MovieSceneScriptingFloatChannel, default=0}
         ]
        }]
      }]
    }
    R_R_Hand_Controller {
     id=#12
     possessed_class=CanvasPanel
     tracks [{
       type=MovieScene2DTransformTrack
       sections [{
         type=MovieScene2DTransformSection
         end_frame=374
         start_frame=0
         channels [
          Translation.X {type=MovieSceneScriptingFloatChannel, default=0}
          Translation.Y {type=MovieSceneScriptingFloatChannel, default=0}
          Angle {type=MovieSceneScriptingFloatChannel, keys [[0,0],[374,20]]}
          Scale.X {type=MovieSceneScriptingFloatChannel, default=1}
          Scale.Y {type=MovieSceneScriptingFloatChannel, default=1}
          Shear.X {type=MovieSceneScriptingFloatChannel, default=0}
          Shear.Y {type=MovieSceneScriptingFloatChannel, default=0}
         ]
        }]
      }]
    }
    BlackBar_Bottom {
     id=#13
     possessed_class=Image
     tracks [{
       type=MovieSceneFloatTrack
       display_name=Render Opacity
       sections [{
         type=MovieSceneFloatSection
         end_frame=374
         start_frame=0
         channels [None {type=MovieSceneScriptingFloatChannel, keys [[0,1],[366,1],[374,0]]}]
        }]
      }]
    }
    BlackBar_Top {
     id=#14
     possessed_class=Image
     tracks [{
       type=MovieSceneFloatTrack
       display_name=Render Opacity
       sections [{
         type=MovieSceneFloatSection
         end_frame=374
         start_frame=0
         channels [None {type=MovieSceneScriptingFloatChannel, keys [[0,1],[366,1],[374,0]]}]
        }]
      }]
    }
    R_Powers {
     id=#15
     possessed_class=Image
     tracks [{
       type=MovieScene2DTransformTrack
       sections [{
         type=MovieScene2DTransformSection
         end_frame=374
         start_frame=0
         channels [
          Translation.X {type=MovieSceneScriptingFloatChannel, default=0}
          Translation.Y {type=MovieSceneScriptingFloatChannel, default=0}
          Angle {type=MovieSceneScriptingFloatChannel, default=0}
          Scale.X {type=MovieSceneScriptingFloatChannel, keys [[0,1],[374,1.2]]}
          Scale.Y {type=MovieSceneScriptingFloatChannel, keys [[0,1],[374,1.2]]}
          Shear.X {type=MovieSceneScriptingFloatChannel, default=0}
          Shear.Y {type=MovieSceneScriptingFloatChannel, default=0}
         ]
        }]
      }]
    }
    R_BG_Controller {
     id=#16
     possessed_class=CanvasPanel
     tracks [{
       type=MovieScene2DTransformTrack
       sections [{
         type=MovieScene2DTransformSection
         end_frame=15
         start_frame=0
         channels [
          Translation.X {type=MovieSceneScriptingFloatChannel, keys [[0,157.2426],[15,-19.7455]]}
          Translation.Y {type=MovieSceneScriptingFloatChannel, keys [[0,-1213.5093],[15,24.6162]]}
          Angle {type=MovieSceneScriptingFloatChannel, default=0}
          Scale.X {type=MovieSceneScriptingFloatChannel, default=2, const [2.1,1]}
          Scale.Y {type=MovieSceneScriptingFloatChannel, default=2, const [2.1,1]}
          Shear.X {type=MovieSceneScriptingFloatChannel, default=0}
          Shear.Y {type=MovieSceneScriptingFloatChannel, default=0}
         ]
        }]
      }]
    }
    R_Synth {
     id=#17
     possessed_class=CanvasPanel
     tracks [{
       type=MovieScene2DTransformTrack
       sections [{
         type=MovieScene2DTransformSection
         end_frame=374
         start_frame=0
         channels [
          Translation.X {type=MovieSceneScriptingFloatChannel, keys [[0,772.779],[12.303,772.779],[24.605,0],[374,-38.2837]]}
          Translation.Y {type=MovieSceneScriptingFloatChannel, keys [[0,421.278],[12.303,421.278],[24.605,0],[374,-11.5701]]}
          Angle {type=MovieSceneScriptingFloatChannel, keys [[0,11.8808],[12.303,11.8808],[24.605,0],[374,-3.3526]]}
          Scale.X {type=MovieSceneScriptingFloatChannel, keys [[0,0.8],[12.303,0.8],[24.605,1],[374,1]]}
          Scale.Y {type=MovieSceneScriptingFloatChannel, keys [[0,0.8],[12.303,0.8],[24.605,1],[374,1]]}
          Shear.X {type=MovieSceneScriptingFloatChannel, default=0}
          Shear.Y {type=MovieSceneScriptingFloatChannel, default=0}
         ]
        }]
      }]
    }
    R_Sparks_Controller {
     id=#18
     possessed_class=CanvasPanel
     tracks [{
       type=MovieScene2DTransformTrack
       sections [{
         type=MovieScene2DTransformSection
         end_frame=374
         start_frame=0
         channels [
          Translation.X {type=MovieSceneScriptingFloatChannel, default=0}
          Translation.Y {type=MovieSceneScriptingFloatChannel, default=0}
          Angle {type=MovieSceneScriptingFloatChannel, default=0}
          Scale.X {type=MovieSceneScriptingFloatChannel, keys [[0,1],[374,1.3]]}
          Scale.Y {type=MovieSceneScriptingFloatChannel, keys [[0,1],[374,1.3]]}
          Shear.X {type=MovieSceneScriptingFloatChannel, default=0}
          Shear.Y {type=MovieSceneScriptingFloatChannel, default=0}
         ]
        }]
      }]
    }
    R_Frame {
     id=#19
     possessed_class=CanvasPanel
     tracks [
      {
       type=MovieScene2DTransformTrack
       sections [{
         type=MovieScene2DTransformSection
         end_frame=374
         start_frame=0
         channels [
          Translation.X {type=MovieSceneScriptingFloatChannel, const [0,3]}
          Translation.Y {type=MovieSceneScriptingFloatChannel, keys [[0,0],[365.798,0],[374,1285.5883]]}
          Angle {type=MovieSceneScriptingFloatChannel, default=0}
          Scale.X {type=MovieSceneScriptingFloatChannel, default=1}
          Scale.Y {type=MovieSceneScriptingFloatChannel, default=1}
          Shear.X {type=MovieSceneScriptingFloatChannel, default=0}
          Shear.Y {type=MovieSceneScriptingFloatChannel, default=0}
         ]
        }]
      }
      {
       type=MovieSceneFloatTrack
       display_name=Render Opacity
       sections [{
         type=MovieSceneFloatSection
         end_frame=374
         start_frame=0
         channels [None {type=MovieSceneScriptingFloatChannel, keys [[0,0],[13,1],[366,1],[374,0]]}]
        }]
      }
     ]
    }
    R_L_Hand_Controller {
     id=#20
     possessed_class=CanvasPanel
     tracks [{
       type=MovieScene2DTransformTrack
       sections [{
         type=MovieScene2DTransformSection
         end_frame=374
         start_frame=0
         channels [
          Translation.X {type=MovieSceneScriptingFloatChannel, default=0}
          Translation.Y {type=MovieSceneScriptingFloatChannel, default=0}
          Angle {type=MovieSceneScriptingFloatChannel, keys [[0,0],[374,20]]}
          Scale.X {type=MovieSceneScriptingFloatChannel, default=1}
          Scale.Y {type=MovieSceneScriptingFloatChannel, default=1}
          Shear.X {type=MovieSceneScriptingFloatChannel, default=0}
          Shear.Y {type=MovieSceneScriptingFloatChannel, default=0}
         ]
        }]
      }]
    }
   ]
   display_rate {numerator=60}
  }
 }
]
graphs [EventGraph {nodes [
   n1 = SetRightPanelVisiblity!() -> (OutputDelegate, then, InVisibility:ESlateVisibility Enum=SelfHitTestInvisible) {is=title K2Node_CustomEvent}
   > SetVisibility(self=Get R_Frame() -> (R_Frame:Canvas Panel Object Reference), InVisibility=@n1.InVisibility)
   n2 = SetLeftPanelVisibility!() -> (OutputDelegate, then, InVisibility:ESlateVisibility Enum=SelfHitTestInvisible) {is=title K2Node_CustomEvent}
   > SetVisibility(
    self=[
     Get BlackBar_Top() -> (BlackBar_Top:Image Object Reference)
     Get BlackBar_Bottom() -> (BlackBar_Bottom:Image Object Reference)
     Get L_Frame() -> (L_Frame:Canvas Panel Object Reference)
    ]
    InVisibility=@n2.InVisibility
   )
   Event Construct()
   > SetLeftPanelVisibility!(execute, self, InVisibility:ESlateVisibility Enum=Hidden) -> (then) {is=title K2Node_CallFunction}
   > SetRightPanelVisiblity!(execute, self, InVisibility:ESlateVisibility Enum=Hidden) -> (then) {is=title K2Node_CallFunction}
   Event HeroPanel_UnloadTextures() {is=title K2Node_Event}
   n3 = Make Array!(
    "[0]":Image Object Reference=Get R_T_Sparks() -> (R_T_Sparks:Image Object Reference)
    "[1]":Image Object Reference=Get R_Synth_Body() -> (R_Synth_Body:Image Object Reference)
    "[2]":Image Object Reference=Get R_R_Hand() -> (R_R_Hand:Image Object Reference)
    "[3]":Image Object Reference=Get R_Powers() -> (R_Powers:Image Object Reference)
    "[4]":Image Object Reference=Get R_L_Hand() -> (R_L_Hand:Image Object Reference)
    "[5]":Image Object Reference=Get R_Embers() -> (R_Embers:Image Object Reference)
    "[6]":Image Object Reference=Get R_BG_50() -> (R_BG_50:Image Object Reference)
    "[7]":Image Object Reference=Get R_B_Sparks() -> (R_B_Sparks:Image Object Reference)
    "[8]":Image Object Reference=Get L_T_Sparks() -> (L_T_Sparks:Image Object Reference)
    "[9]":Image Object Reference=Get L_Synth_R_Hand() -> (L_Synth_R_Hand:Image Object Reference)
    "[10]":Image Object Reference=Get L_Synth_L_Hand() -> (L_Synth_L_Hand:Image Object Reference)
    "[11]":Image Object Reference=Get L_Synth_Body() -> (L_Synth_Body:Image Object Reference)
    "[12]":Image Object Reference=Get L_S_R_Power() -> (L_S_R_Power:Image Object Reference)
    "[13]":Image Object Reference=Get L_S_L_Power() -> (L_S_L_Power:Image Object Reference)
    "[14]":Image Object Reference=Get L_BG_50() -> (L_BG_50:Image Object Reference)
    "[15]":Image Object Reference=Get L_B_Sparks() -> (L_B_Sparks:Image Object Reference)
   ) -> (Array:Array of Image Object References)
   > n4 = For Each Loop(Array:Array of Image Object References=@n3) -> (Array Element:Image Object Reference)
   CollectGarbage!(execute=@n4.Completed) -> (then)
   > PrintString!(execute, InString=UNLOADED HERO PANEL 1 TEXTURES, bPrintToScreen, bPrintToLog, TextColor=(R=0.9931,G=0.9931,B=1,A=1), Duration=3, Key) -> (then)
   Is Valid(exec=@n4.LoopBody, InputObject=@n4)
   > SetBrushFromTexture!(execute, self:Image Object Reference=@n4, Texture:Texture 2D Object Reference, bMatchSize:bool=true) -> (then)
   Event HeroPanel_LoadTextures() {is=title K2Node_Event}
   > n5 = For Each Loop(Array=@n3)
   > n6 = Branch(Condition=IsMobilePlatform!() -> (ReturnValue:bool))
   SetBrushFromSoftTexture!(
    execute=@n6.else
    self:Image Object Reference=@n5
    SoftTexture:Texture 2D Soft Object Reference="Get (a copy)"(
     Array:Array of Texture 2D Soft Object References=Get TexturesHigh() -> (TexturesHigh:Array of Texture 2D Soft Object References)
     Dimension 1=@n5.Array Index
    ) -> (Output:Texture 2D Soft Object Reference)
    bMatchSize:bool=false
   ) -> (then)
   n7 = Event PlayPistolCanaryHeroPanelAnimation!() -> (OutputDelegate, then, HeroPanel:E_PistolCanary_HeroPanelSelect Enum=NewEnumerator0, PanelSide:E_PistolCanary_PanelSide Enum=NewEnumerator0) {is=title K2Node_Event}
   > PlayAnimation(InAnimation=Select!(
     NewEnumerator0:Widget Animation Object Reference=Get LeftPanelAnim() -> (LeftPanelAnim:Widget Animation Object Reference)
     NewEnumerator1:Widget Animation Object Reference=Get RightPanelAnim() -> (RightPanelAnim:Widget Animation Object Reference)
     NewEnumerator2:Widget Animation Object Reference
     Index:E_PistolCanary_PanelSide Enum=@n7.PanelSide
    ) -> (ReturnValue:Widget Animation Object Reference))
   > n8 = Switch on E_PistolCanary_PanelSide!(execute, Selection:E_PistolCanary_PanelSide Enum=@n7.PanelSide) -> (NewEnumerator0:Exec, NewEnumerator1:Exec, NewEnumerator2:Exec) {is=title K2Node_SwitchEnum}
   > SetLeftPanelVisibility(InVisibility=HitTestInvisible)
   SetRightPanelVisiblity(execute=@n8.NewEnumerator1, InVisibility=HitTestInvisible)
   Event PreConstruct()
   Event Tick()
  ]}]
variables [TexturesHigh {
  type="array<Texture2D>"
  default [
   /PistolCanary/Layout/ComicPanels_v2/ComicPanelTextures/Panel4/R_Frame/R_P4_Synth_T_Sparks
   /PistolCanary/Layout/ComicPanels_v2/ComicPanelTextures/Panel4/R_Frame/R_P4_Synth_Body
   /PistolCanary/Layout/ComicPanels_v2/ComicPanelTextures/Panel4/R_Frame/R_P4_Synth_R_Hand
   /PistolCanary/Layout/ComicPanels_v2/ComicPanelTextures/Panel4/R_Frame/R_P4_Synth_Powers_50percent
   /PistolCanary/Layout/ComicPanels_v2/ComicPanelTextures/Panel4/R_Frame/R_P4_Synth_L_Hand
   /PistolCanary/Layout/ComicPanels_v2/ComicPanelTextures/Panel4/R_Frame/R_P4_Synth_Embers_50percent
   /PistolCanary/Layout/ComicPanels_v2/ComicPanelTextures/Panel4/R_Frame/R_P4_BG_50percent
   /PistolCanary/Layout/ComicPanels_v2/ComicPanelTextures/Panel4/R_Frame/R_P4_Synth_B_Sparks
   /PistolCanary/Layout/ComicPanels_v2/ComicPanelTextures/Panel4/L_Frame/L_P4_T_Sparks
   /PistolCanary/Layout/ComicPanels_v2/ComicPanelTextures/Panel4/L_Frame/L_P4_R_Hand
   /PistolCanary/Layout/ComicPanels_v2/ComicPanelTextures/Panel4/L_Frame/L_P4_L_Hand
   /PistolCanary/Layout/ComicPanels_v2/ComicPanelTextures/Panel4/L_Frame/L_P4_Synth_Body
   /PistolCanary/Layout/ComicPanels_v2/ComicPanelTextures/Panel4/L_Frame/L_P4_R_Powers
   /PistolCanary/Layout/ComicPanels_v2/ComicPanelTextures/Panel4/L_Frame/L_P4_L_Powers
   /PistolCanary/Layout/ComicPanels_v2/ComicPanelTextures/Panel4/L_Frame/L_P4_BG_50percent
   /PistolCanary/Layout/ComicPanels_v2/ComicPanelTextures/Panel4/L_Frame/L_P4_B_Sparks
  ]
 }]
widgets {
 info {inheritedWidgetCount=0, parentClass=&UserWidget, rootWidgetClass=&CanvasPanel, widgetCount=33}
 widgets [
  Panel_04 {widgetClassPath=&CanvasPanel, widget=Panel_04, properties {clipping=ClipToBounds, pixelSnapping=Disabled}}
  L_Frame {
   widgetClassPath=&CanvasPanel
   bIsVariable=true
   parent=Panel_04
   widget=L_Frame
   properties {clipping=ClipToBounds}
   slot {
    class=CanvasPanelSlot
    properties {layoutData {alignment=<0 0>, anchors {maximum=<0 0.5>, minimum=<0 0.5>}, offsets=<ltrb 0 -540.5405 1000 1080>}}
   }
  }
  L_BG_Controller {
   widgetClassPath=&CanvasPanel
   parent=L_Frame
   widget=L_BG_Controller
   properties {clipping=ClipToBounds, renderTransform {angle=0, scale=<2 2>, shear=<0 0>, translation=<0 0>}}
   slot {
    class=CanvasPanelSlot
    properties {bAutoSize=true, layoutData {alignment=<0 0>, anchors {maximum=<0 0>, minimum=<0 0>}, offsets=<ltrb 48 284 100 30>}}
   }
  }
  L_BG_50 {
   widgetClassPath=&Image
   bIsVariable=true
   parent=L_BG_Controller
   widget=L_BG_50
   properties {
    clipping=ClipToBounds
    brush {
     bIsDynamicallyLoaded=false
     drawAs=Image
     imageType=FullColor
     mirroring=NoMirror
     resourceName=None
     resourceObject=/PistolCanary/Layout/ComicPanels_v2/ComicPanelTextures/Panel4/L_Frame/LowRes/L_P4_BG_50percent_low
     tiling=NoTile
     imageSize=<192 336>
     margin=<ltrb 0 0 0 0>
     outlineSettings {
      bUseBrushTransparency=false
      roundingType=HalfHeightRadius
      width=0
      color {colorUseRule=UseColor_Specified, specifiedColor=<rgba 0 0 0 0>}
      cornerRadii=<0 0 0 0>
     }
     tintColor {colorUseRule=UseColor_Specified, specifiedColor=<rgba 1 1 1 1>}
     uVRegion {bIsValid=false, max=<0 0>, min=<0 0>}
    }
   }
   slot {class=CanvasPanelSlot, properties {layoutData {alignment=<0 0>, anchors {maximum=<0 0>, minimum=<0 0>}, offsets=<ltrb 0 0 320 564>}}}
  }
  L_Synth {
   widgetClassPath=&CanvasPanel
   parent=L_Frame
   widget=L_Synth
   slot {
    class=CanvasPanelSlot
    properties {bAutoSize=true, layoutData {alignment=<0 0>, anchors {maximum=<0 0>, minimum=<0 0>}, offsets=<ltrb 0 0 513.7143 1098.5715>}}
   }
  }
  L_Synth_Body {
   widgetClassPath=&Image
   bIsVariable=true
   parent=L_Synth
   widget=L_Synth_Body
   properties {
    clipping=ClipToBounds
    visibility=HitTestInvisible
    brush {
     bIsDynamicallyLoaded=false
     drawAs=Image
     imageType=FullColor
     mirroring=NoMirror
     resourceName=None
     resourceObject=/PistolCanary/Layout/ComicPanels_v2/ComicPanelTextures/Panel4/L_Frame/LowRes/L_P4_Synth_Body_low
     tiling=NoTile
     imageSize=<328 404>
     margin=<ltrb 0 0 0 0>
     outlineSettings {
      bUseBrushTransparency=false
      roundingType=HalfHeightRadius
      width=0
      color {colorUseRule=UseColor_Specified, specifiedColor=<rgba 0 0 0 0>}
      cornerRadii=<0 0 0 0>
     }
     tintColor {colorUseRule=UseColor_Specified, specifiedColor=<rgba 1 1 1 1>}
     uVRegion {bIsValid=false, max=<0 0>, min=<0 0>}
    }
   }
   slot {
    class=CanvasPanelSlot
    properties {zOrder=1, layoutData {alignment=<0 0>, anchors {maximum=<0 0>, minimum=<0 0>}, offsets=<ltrb -80 196 548 672>}}
   }
  }
  L_Synth_L_Hand_Controller {
   widgetClassPath=&CanvasPanel
   parent=L_Synth
   widget=L_Synth_L_Hand_Controller
   properties {clipping=ClipToBounds}
   slot {
    class=CanvasPanelSlot
    properties {zOrder=6, layoutData {alignment=<0 0>, anchors {maximum=<0 0>, minimum=<0 0>}, offsets=<ltrb 220 104 289.7143 351.1429>}}
   }
  }
  L_Synth_L_Hand {
   widgetClassPath=&Image
   bIsVariable=true
   parent=L_Synth_L_Hand_Controller
   widget=L_Synth_L_Hand
   properties {
    clipping=ClipToBounds
    visibility=HitTestInvisible
    brush {
     bIsDynamicallyLoaded=false
     drawAs=Image
     imageType=FullColor
     mirroring=NoMirror
     resourceName=None
     resourceObject=/PistolCanary/Layout/ComicPanels_v2/ComicPanelTextures/Panel4/L_Frame/LowRes/L_P4_L_Hand_low
     tiling=NoTile
     imageSize=<228 284>
     margin=<ltrb 0 0 0 0>
     outlineSettings {
      bUseBrushTransparency=false
      roundingType=HalfHeightRadius
      width=0
      color {colorUseRule=UseColor_Specified, specifiedColor=<rgba 0 0 0 0>}
      cornerRadii=<0 0 0 0>
     }
     tintColor {colorUseRule=UseColor_Specified, specifiedColor=<rgba 1 1 1 1>}
     uVRegion {bIsValid=false, max=<0 0>, min=<0 0>}
    }
   }
   slot {
    class=CanvasPanelSlot
    properties {zOrder=6, layoutData {alignment=<0 0>, anchors {maximum=<0.5 0.5>, minimum=<0.5 0.5>}, offsets=<ltrb -128.8571 -171.5714 284 356>}}
   }
  }
  L_Synth_R_Hand_Controller {
   widgetClassPath=&CanvasPanel
   parent=L_Synth
   widget=L_Synth_R_Hand_Controller
   properties {clipping=ClipToBounds}
   slot {
    class=CanvasPanelSlot
    properties {zOrder=6, layoutData {alignment=<0 0>, anchors {maximum=<0.5 0.5>, minimum=<0.5 0.5>}, offsets=<ltrb -626 269 217.2245 197.5102>}}
   }
  }
  L_Synth_R_Hand {
   widgetClassPath=&Image
   bIsVariable=true
   parent=L_Synth_R_Hand_Controller
   widget=L_Synth_R_Hand
   properties {
    clipping=ClipToBounds
    visibility=HitTestInvisible
    brush {
     bIsDynamicallyLoaded=false
     drawAs=Image
     imageType=FullColor
     mirroring=NoMirror
     resourceName=None
     resourceObject=/PistolCanary/Layout/ComicPanels_v2/ComicPanelTextures/Panel4/L_Frame/LowRes/L_P4_R_Hand_low
     tiling=NoTile
     imageSize=<224 204>
     margin=<ltrb 0 0 0 0>
     outlineSettings {
      bUseBrushTransparency=false
      roundingType=HalfHeightRadius
      width=0
      color {colorUseRule=UseColor_Specified, specifiedColor=<rgba 0 0 0 0>}
      cornerRadii=<0 0 0 0>
     }
     tintColor {colorUseRule=UseColor_Specified, specifiedColor=<rgba 1 1 1 1>}
     uVRegion {bIsValid=false, max=<0 0>, min=<0 0>}
    }
   }
   slot {
    class=CanvasPanelSlot
    properties {zOrder=6, layoutData {alignment=<0 0>, anchors {maximum=<0 0>, minimum=<0 0>}, offsets=<ltrb -12 -8 224 204>}}
   }
  }
  L_Synth_L_Power_Controller {
   widgetClassPath=&CanvasPanel
   parent=L_Synth
   widget=L_Synth_L_Power_Controller
   slot {
    class=CanvasPanelSlot
    properties {zOrder=5, layoutData {alignment=<0 0>, anchors {maximum=<0 0>, minimum=<0 0>}, offsets=<ltrb 232 112 289.7143 351.1429>}}
   }
  }
  L_S_L_Power {
   widgetClassPath=&Image
   bIsVariable=true
   parent=L_Synth_L_Power_Controller
   widget=L_S_L_Power
   properties {
    clipping=ClipToBounds
    visibility=HitTestInvisible
    brush {
     bIsDynamicallyLoaded=false
     drawAs=Image
     imageType=FullColor
     mirroring=NoMirror
     resourceName=None
     resourceObject=/PistolCanary/Layout/ComicPanels_v2/ComicPanelTextures/Panel4/L_Frame/LowRes/L_P4_L_Powers_low
     tiling=NoTile
     imageSize=<372 324>
     margin=<ltrb 0 0 0 0>
     outlineSettings {
      bUseBrushTransparency=false
      roundingType=HalfHeightRadius
      width=0
      color {colorUseRule=UseColor_Specified, specifiedColor=<rgba 0 0 0 0>}
      cornerRadii=<0 0 0 0>
     }
     tintColor {colorUseRule=UseColor_Specified, specifiedColor=<rgba 1 1 1 1>}
     uVRegion {bIsValid=false, max=<0 0>, min=<0 0>}
    }
   }
   slot {
    class=CanvasPanelSlot
    properties {zOrder=5, layoutData {alignment=<0 0>, anchors {maximum=<0 0>, minimum=<0 0>}, offsets=<ltrb -148 -112 620 536>}}
   }
  }
  L_Synth_R_Power_Controller {
   widgetClassPath=&CanvasPanel
   parent=L_Synth
   widget=L_Synth_R_Power_Controller
   slot {
    class=CanvasPanelSlot
    properties {zOrder=5, layoutData {alignment=<0 0>, anchors {maximum=<0 0>, minimum=<0 0>}, offsets=<ltrb 16 730.2857 215.4286 195.7143>}}
   }
  }
  L_S_R_Power {
   widgetClassPath=&Image
   bIsVariable=true
   parent=L_Synth_R_Power_Controller
   widget=L_S_R_Power
   properties {
    clipping=ClipToBounds
    visibility=HitTestInvisible
    brush {
     bIsDynamicallyLoaded=false
     drawAs=Image
     imageType=FullColor
     mirroring=NoMirror
     resourceName=None
     resourceObject=/PistolCanary/Layout/ComicPanels_v2/ComicPanelTextures/Panel4/L_Frame/LowRes/L_P4_R_Powers_low
     tiling=NoTile
     imageSize=<324 324>
     margin=<ltrb 0 0 0 0>
     outlineSettings {
      bUseBrushTransparency=false
      roundingType=HalfHeightRadius
      width=0
      color {colorUseRule=UseColor_Specified, specifiedColor=<rgba 0 0 0 0>}
      cornerRadii=<0 0 0 0>
     }
     tintColor {colorUseRule=UseColor_Specified, specifiedColor=<rgba 1 1 1 1>}
     uVRegion {bIsValid=false, max=<0 0>, min=<0 0>}
    }
   }
   slot {
    class=CanvasPanelSlot
    properties {zOrder=5, layoutData {alignment=<0 0>, anchors {maximum=<0 0>, minimum=<0 0>}, offsets=<ltrb -184 -156 540 540>}}
   }
  }
  L_Sparks_Controller {
   widgetClassPath=&CanvasPanel
   parent=L_Synth
   widget=L_Sparks_Controller
   slot {
    class=CanvasPanelSlot
    properties {layoutData {alignment=<0 0>, anchors {maximum=<0 0>, minimum=<0 0>}, offsets=<ltrb 196 508 101.1429 90.5714>}}
   }
  }
  L_T_Sparks {
   widgetClassPath=&Image
   bIsVariable=true
   parent=L_Sparks_Controller
   widget=L_T_Sparks
   properties {
    clipping=ClipToBounds
    visibility=HitTestInvisible
    brush {
     bIsDynamicallyLoaded=false
     drawAs=Image
     imageType=FullColor
     mirroring=NoMirror
     resourceName=None
     resourceObject=/PistolCanary/Layout/ComicPanels_v2/ComicPanelTextures/Panel4/L_Frame/LowRes/L_P4_T_Sparks_low
     tiling=NoTile
     imageSize=<508 208>
     margin=<ltrb 0 0 0 0>
     outlineSettings {
      bUseBrushTransparency=false
      roundingType=HalfHeightRadius
      width=0
      color {colorUseRule=UseColor_Specified, specifiedColor=<rgba 0 0 0 0>}
      cornerRadii=<0 0 0 0>
     }
     tintColor {colorUseRule=UseColor_Specified, specifiedColor=<rgba 1 1 1 1>}
     uVRegion {bIsValid=false, max=<0 0>, min=<0 0>}
    }
   }
   slot {class=CanvasPanelSlot, properties {layoutData {alignment=<0 0>, anchors {maximum=<0 0>, minimum=<0 0>}, offsets=<ltrb -216 -508 844 344>}}}
  }
  L_B_Sparks {
   widgetClassPath=&Image
   bIsVariable=true
   parent=L_Sparks_Controller
   widget=L_B_Sparks
   properties {
    clipping=ClipToBounds
    visibility=HitTestInvisible
    brush {
     bIsDynamicallyLoaded=false
     drawAs=Image
     imageType=FullColor
     mirroring=NoMirror
     resourceName=None
     resourceObject=/PistolCanary/Layout/ComicPanels_v2/ComicPanelTextures/Panel4/L_Frame/LowRes/L_P4_B_Sparks_low
     tiling=NoTile
     imageSize=<392 304>
     margin=<ltrb 0 0 0 0>
     outlineSettings {
      bUseBrushTransparency=false
      roundingType=HalfHeightRadius
      width=0
      color {colorUseRule=UseColor_Specified, specifiedColor=<rgba 0 0 0 0>}
      cornerRadii=<0 0 0 0>
     }
     tintColor {colorUseRule=UseColor_Specified, specifiedColor=<rgba 1 1 1 1>}
     uVRegion {bIsValid=false, max=<0 0>, min=<0 0>}
    }
   }
   slot {class=CanvasPanelSlot, properties {layoutData {alignment=<0 0>, anchors {maximum=<0 0>, minimum=<0 0>}, offsets=<ltrb -408 112 648 508>}}}
  }
  R_Frame {
   widgetClassPath=&CanvasPanel
   bIsVariable=true
   parent=Panel_04
   widget=R_Frame
   properties {clipping=ClipToBounds}
   slot {
    class=CanvasPanelSlot
    properties {bAutoSize=true, layoutData {alignment=<0 0>, anchors {maximum=<1 0.5>, minimum=<1 0.5>}, offsets=<ltrb -1921.9219 -540.5405 100 30>}}
   }
  }
  R_BG_Controller {
   widgetClassPath=&CanvasPanel
   parent=R_Frame
   widget=R_BG_Controller
   properties {renderTransform {angle=0, scale=<2.1 2.1>, shear=<0 0>, translation=<17.2701 4.7448>}}
   slot {
    class=CanvasPanelSlot
    properties {bAutoSize=true, layoutData {alignment=<0 0>, anchors {maximum=<0 0>, minimum=<0 0>}, offsets=<ltrb 1376 280 100 30>}}
   }
  }
  R_BG_50 {
   widgetClassPath=&Image
   bIsVariable=true
   parent=R_BG_Controller
   widget=R_BG_50
   properties {
    clipping=ClipToBounds
    visibility=HitTestInvisible
    brush {
     bIsDynamicallyLoaded=false
     drawAs=Image
     imageType=FullColor
     mirroring=NoMirror
     resourceName=None
     resourceObject=/PistolCanary/Layout/ComicPanels_v2/ComicPanelTextures/Panel4/R_Frame/LowRes/R_P4_BG_50percent_low
     tiling=NoTile
     imageSize=<208 336>
     margin=<ltrb 0 0 0 0>
     outlineSettings {
      bUseBrushTransparency=false
      roundingType=HalfHeightRadius
      width=0
      color {colorUseRule=UseColor_Specified, specifiedColor=<rgba 0 0 0 0>}
      cornerRadii=<0 0 0 0>
     }
     tintColor {colorUseRule=UseColor_Specified, specifiedColor=<rgba 1 1 1 1>}
     uVRegion {bIsValid=false, max=<0 0>, min=<0 0>}
    }
    renderTransform {angle=0, scale=<1.1 1.1>, shear=<0 0>, translation=<0 0>}
   }
   slot {class=CanvasPanelSlot, properties {layoutData {alignment=<0 0>, anchors {maximum=<0 0>, minimum=<0 0>}, offsets=<ltrb 0 0 352 564>}}}
  }
  R_Embers {
   widgetClassPath=&Image
   bIsVariable=true
   parent=R_Frame
   widget=R_Embers
   properties {
    clipping=ClipToBounds
    visibility=HitTestInvisible
    brush {
     bIsDynamicallyLoaded=false
     drawAs=Image
     imageType=FullColor
     mirroring=NoMirror
     resourceName=None
     resourceObject=&R_P4_Synth_Embers_50percent
     tiling=NoTile
     imageSize=<444 236>
     margin=<ltrb 0 0 0 0>
     outlineSettings {
      bUseBrushTransparency=false
      roundingType=HalfHeightRadius
      width=0
      color {colorUseRule=UseColor_Specified, specifiedColor=<rgba 0 0 0 0>}
      cornerRadii=<0 0 0 0>
     }
     tintColor {colorUseRule=UseColor_Specified, specifiedColor=<rgba 1 1 1 1>}
     uVRegion {bIsValid=false, max=<0 0>, min=<0 0>}
    }
   }
   slot {class=CanvasPanelSlot, properties {layoutData {alignment=<0 0>, anchors {maximum=<0 0>, minimum=<0 0>}, offsets=<ltrb 1184 696 888 468>}}}
  }
  R_Synth {
   widgetClassPath=&CanvasPanel
   parent=R_Frame
   widget=R_Synth
   slot {
    class=CanvasPanelSlot
    properties {zOrder=1, layoutData {alignment=<0 0>, anchors {maximum=<0 0>, minimum=<0 0>}, offsets=<ltrb 1716 852 132 121.4286>}}
   }
  }
  R_Synth_Body {
   widgetClassPath=&Image
   bIsVariable=true
   parent=R_Synth
   widget=R_Synth_Body
   properties {
    clipping=ClipToBounds
    visibility=HitTestInvisible
    brush {
     bIsDynamicallyLoaded=false
     drawAs=Image
     imageType=FullColor
     mirroring=NoMirror
     resourceName=None
     resourceObject=/PistolCanary/Layout/ComicPanels_v2/ComicPanelTextures/Panel4/R_Frame/LowRes/R_P4_Synth_Body_low
     tiling=NoTile
     imageSize=<324 388>
     margin=<ltrb 0 0 0 0>
     outlineSettings {
      bUseBrushTransparency=false
      roundingType=HalfHeightRadius
      width=0
      color {colorUseRule=UseColor_Specified, specifiedColor=<rgba 0 0 0 0>}
      cornerRadii=<0 0 0 0>
     }
     tintColor {colorUseRule=UseColor_Specified, specifiedColor=<rgba 1 1 1 1>}
     uVRegion {bIsValid=false, max=<0 0>, min=<0 0>}
    }
   }
   slot {
    class=CanvasPanelSlot
    properties {zOrder=1, layoutData {alignment=<0 0>, anchors {maximum=<0 0>, minimum=<0 0>}, offsets=<ltrb -220 -388 540 648>}}
   }
  }
  R_Powers {
   widgetClassPath=&Image
   bIsVariable=true
   parent=R_Synth
   widget=R_Powers
   properties {
    clipping=ClipToBounds
    visibility=HitTestInvisible
    brush {
     bIsDynamicallyLoaded=false
     drawAs=Image
     imageType=FullColor
     mirroring=NoMirror
     resourceName=None
     resourceObject=/PistolCanary/Layout/ComicPanels_v2/ComicPanelTextures/Panel4/R_Frame/LowRes/R_P4_Synth_Powers_50percent_low
     tiling=NoTile
     imageSize=<264 336>
     margin=<ltrb 0 0 0 0>
     outlineSettings {
      bUseBrushTransparency=false
      roundingType=HalfHeightRadius
      width=0
      color {colorUseRule=UseColor_Specified, specifiedColor=<rgba 0 0 0 0>}
      cornerRadii=<0 0 0 0>
     }
     tintColor {colorUseRule=UseColor_Specified, specifiedColor=<rgba 1 1 1 1>}
     uVRegion {bIsValid=false, max=<0 0>, min=<0 0>}
    }
   }
   slot {
    class=CanvasPanelSlot
    properties {zOrder=1, layoutData {alignment=<0 0>, anchors {maximum=<0 0>, minimum=<0 0>}, offsets=<ltrb -576 -868 888 1128>}}
   }
  }
  R_Sparks_Controller {
   widgetClassPath=&CanvasPanel
   parent=R_Synth
   widget=R_Sparks_Controller
   slot {
    class=CanvasPanelSlot
    properties {layoutData {alignment=<0 0>, anchors {maximum=<0 0>, minimum=<0 0>}, offsets=<ltrb -76 -196 102.2857 96.2857>}}
   }
  }
  R_T_Sparks {
   widgetClassPath=&Image
   bIsVariable=true
   parent=R_Sparks_Controller
   widget=R_T_Sparks
   properties {
    clipping=ClipToBounds
    visibility=HitTestInvisible
    brush {
     bIsDynamicallyLoaded=false
     drawAs=Image
     imageType=FullColor
     mirroring=NoMirror
     resourceName=None
     resourceObject=/PistolCanary/Layout/ComicPanels_v2/ComicPanelTextures/Panel4/R_Frame/LowRes/R_P4_Synth_T_Sparks_low
     tiling=NoTile
     imageSize=<316 224>
     margin=<ltrb 0 0 0 0>
     outlineSettings {
      bUseBrushTransparency=false
      roundingType=HalfHeightRadius
      width=0
      color {colorUseRule=UseColor_Specified, specifiedColor=<rgba 0 0 0 0>}
      cornerRadii=<0 0 0 0>
     }
     tintColor {colorUseRule=UseColor_Specified, specifiedColor=<rgba 1 1 1 1>}
     uVRegion {bIsValid=false, max=<0 0>, min=<0 0>}
    }
   }
   slot {
    class=CanvasPanelSlot
    properties {layoutData {alignment=<0 0>, anchors {maximum=<0 1>, minimum=<0 0>}, offsets=<ltrb -496 -640 524 363.2857>}}
   }
  }
  R_B_Sparks {
   widgetClassPath=&Image
   bIsVariable=true
   parent=R_Sparks_Controller
   widget=R_B_Sparks
   properties {
    clipping=ClipToBounds
    visibility=HitTestInvisible
    brush {
     bIsDynamicallyLoaded=false
     drawAs=Image
     imageType=FullColor
     mirroring=NoMirror
     resourceName=None
     resourceObject=/PistolCanary/Layout/ComicPanels_v2/ComicPanelTextures/Panel4/R_Frame/LowRes/R_P4_Synth_B_Sparks_low
     tiling=NoTile
     imageSize=<216 240>
     margin=<ltrb 0 0 0 0>
     outlineSettings {
      bUseBrushTransparency=false
      roundingType=HalfHeightRadius
      width=0
      color {colorUseRule=UseColor_Specified, specifiedColor=<rgba 0 0 0 0>}
      cornerRadii=<0 0 0 0>
     }
     tintColor {colorUseRule=UseColor_Specified, specifiedColor=<rgba 1 1 1 1>}
     uVRegion {bIsValid=false, max=<0 0>, min=<0 0>}
    }
   }
   slot {class=CanvasPanelSlot, properties {layoutData {alignment=<0 0>, anchors {maximum=<0 0>, minimum=<0 0>}, offsets=<ltrb -496 60 364 400>}}}
  }
  R_L_Hand_Controller {
   widgetClassPath=&CanvasPanel
   parent=R_Synth
   widget=R_L_Hand_Controller
   slot {
    class=CanvasPanelSlot
    properties {zOrder=5, layoutData {alignment=<0 0>, anchors {maximum=<0 0>, minimum=<0 0>}, offsets=<ltrb -188 -164 22.8571 22>}}
   }
  }
  R_L_Hand {
   widgetClassPath=&Image
   bIsVariable=true
   parent=R_L_Hand_Controller
   widget=R_L_Hand
   properties {
    clipping=ClipToBounds
    pixelSnapping=Disabled
    visibility=HitTestInvisible
    brush {
     bIsDynamicallyLoaded=false
     drawAs=Image
     imageType=FullColor
     mirroring=NoMirror
     resourceName=None
     resourceObject=/PistolCanary/Layout/ComicPanels_v2/ComicPanelTextures/Panel4/R_Frame/LowRes/R_P4_Synth_L_Hand_low
     tiling=NoTile
     imageSize=<108 100>
     margin=<ltrb 0 0 0 0>
     outlineSettings {
      bUseBrushTransparency=false
      roundingType=HalfHeightRadius
      width=0
      color {colorUseRule=UseColor_Specified, specifiedColor=<rgba 0 0 0 0>}
      cornerRadii=<0 0 0 0>
     }
     tintColor {colorUseRule=UseColor_Specified, specifiedColor=<rgba 1 1 1 1>}
     uVRegion {bIsValid=false, max=<0 0>, min=<0 0>}
    }
   }
   slot {
    class=CanvasPanelSlot
    properties {zOrder=5, layoutData {alignment=<0 0>, anchors {maximum=<0 0>, minimum=<0 0>}, offsets=<ltrb -88 -72 108 100>}}
   }
  }
  R_R_Hand_Controller {
   widgetClassPath=&CanvasPanel
   parent=R_Synth
   widget=R_R_Hand_Controller
   slot {
    class=CanvasPanelSlot
    properties {zOrder=5, layoutData {alignment=<0 0>, anchors {maximum=<0 0>, minimum=<0 0>}, offsets=<ltrb -208 -336 22.8571 22>}}
   }
  }
  R_R_Hand {
   widgetClassPath=&Image
   bIsVariable=true
   parent=R_R_Hand_Controller
   widget=R_R_Hand
   properties {
    clipping=ClipToBounds
    visibility=HitTestInvisible
    brush {
     bIsDynamicallyLoaded=false
     drawAs=Image
     imageType=FullColor
     mirroring=NoMirror
     resourceName=None
     resourceObject=/PistolCanary/Layout/ComicPanels_v2/ComicPanelTextures/Panel4/R_Frame/LowRes/R_P4_Synth_R_Hand_low
     tiling=NoTile
     imageSize=<100 84>
     margin=<ltrb 0 0 0 0>
     outlineSettings {
      bUseBrushTransparency=false
      roundingType=HalfHeightRadius
      width=0
      color {colorUseRule=UseColor_Specified, specifiedColor=<rgba 0 0 0 0>}
      cornerRadii=<0 0 0 0>
     }
     tintColor {colorUseRule=UseColor_Specified, specifiedColor=<rgba 1 1 1 1>}
     uVRegion {bIsValid=false, max=<0 0>, min=<0 0>}
    }
   }
   slot {
    class=CanvasPanelSlot
    properties {zOrder=5, layoutData {alignment=<0 0>, anchors {maximum=<0 0>, minimum=<0 0>}, offsets=<ltrb -56 -64 100 84>}}
   }
  }
  BlackBar_Top {
   widgetClassPath=&Image
   bIsVariable=true
   parent=Panel_04
   widget=BlackBar_Top
   properties {brush {
     bIsDynamicallyLoaded=false
     drawAs=Image
     imageType=NoImage
     mirroring=NoMirror
     resourceName=None
     resourceObject=null
     tiling=NoTile
     imageSize=<32 32>
     margin=<ltrb 0 0 0 0>
     outlineSettings {
      bUseBrushTransparency=false
      roundingType=HalfHeightRadius
      width=0
      color {colorUseRule=UseColor_Specified, specifiedColor=<rgba 0 0 0 0>}
      cornerRadii=<0 0 0 0>
     }
     tintColor {colorUseRule=UseColor_Specified, specifiedColor=<rgba 0 0 0 1>}
     uVRegion {bIsValid=false, max=<0 0>, min=<0 0>}
    }}
   slot {
    class=CanvasPanelSlot
    properties {
     zOrder=10
     layoutData {alignment=<0 0>, anchors {maximum=<0.5 0.5>, minimum=<0.5 0.5>}, offsets=<ltrb -1202.0753 -931.8829 2278.6777 393.113>}
    }
   }
  }
  BlackBar_Bottom {
   widgetClassPath=&Image
   bIsVariable=true
   parent=Panel_04
   widget=BlackBar_Bottom
   properties {brush {
     bIsDynamicallyLoaded=false
     drawAs=Image
     imageType=NoImage
     mirroring=NoMirror
     resourceName=None
     resourceObject=null
     tiling=NoTile
     imageSize=<32 32>
     margin=<ltrb 0 0 0 0>
     outlineSettings {
      bUseBrushTransparency=false
      roundingType=HalfHeightRadius
      width=0
      color {colorUseRule=UseColor_Specified, specifiedColor=<rgba 0 0 0 0>}
      cornerRadii=<0 0 0 0>
     }
     tintColor {colorUseRule=UseColor_Specified, specifiedColor=<rgba 0 0 0 1>}
     uVRegion {bIsValid=false, max=<0 0>, min=<0 0>}
    }}
   slot {
    class=CanvasPanelSlot
    properties {zOrder=12, layoutData {alignment=<0 0>, anchors {maximum=<0.5 0.5>, minimum=<0.5 0.5>}, offsets=<ltrb -1154.2546 539.03 2278.6777 393.113>}}
   }
  }
 ]
}
