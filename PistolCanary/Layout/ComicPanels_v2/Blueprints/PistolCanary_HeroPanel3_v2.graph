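# Pistol Canary comic hero panel 3 widget: layered left and right frames of Myst, crows and backgrounds between black letterbox bars.
# Construct hides both sides; the play event reveals the chosen side and runs its animation, with bars fading in on left, out on right.
# Load swaps low-res default brushes for soft high-res textures on non-mobile, matching TexturesHigh to the image array by index.
# Unload clears those brushes and collects garbage.
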WidgetBlueprint /PistolCanary/Layout/ComicPanels_v2/Blueprints/PistolCanary_HeroPanel3_v2
animations [
 LeftPanelAnim {
  binding_count=7
  length_seconds=9.4333
  object=&self:LeftPanelAnim
  sequence {
   bindings [
    L_Myst_R_Hand_Controller {
     id=#1
     possessed_class=CanvasPanel
     tracks [{
       type=MovieScene2DTransformTrack
       sections [{
         type=MovieScene2DTransformSection
         end_frame=566
         start_frame=0
         channels [
          Translation.X {type=MovieSceneScriptingFloatChannel, default=0}
          Translation.Y {type=MovieSceneScriptingFloatChannel, default=0}
          Angle {type=MovieSceneScriptingFloatChannel, keys [[0,0],[566,-22.8202]]}
          Scale.X {type=MovieSceneScriptingFloatChannel, default=1}
          Scale.Y {type=MovieSceneScriptingFloatChannel, default=1}
          Shear.X {type=MovieSceneScriptingFloatChannel, default=0}
          Shear.Y {type=MovieSceneScriptingFloatChannel, default=0}
         ]
        }]
      }]
    }
    L_BG_Controller {
     id=#2
     possessed_class=CanvasPanel
     tracks [{
       type=MovieScene2DTransformTrack
       sections [{
         type=MovieScene2DTransformSection
         end_frame=15
         start_frame=0
         channels [
          Translation.X {type=MovieSceneScriptingFloatChannel, keys [[0,-461.2941],[15,0]]}
          Translation.Y {type=MovieSceneScriptingFloatChannel, keys [[0,2072.2954],[15,0]]}
          Angle {type=MovieSceneScriptingFloatChannel, default=0}
          Scale.X {type=MovieSceneScriptingFloatChannel, default=1}
          Scale.Y {type=MovieSceneScriptingFloatChannel, default=1}
          Shear.X {type=MovieSceneScriptingFloatChannel, default=0}
          Shear.Y {type=MovieSceneScriptingFloatChannel, default=0}
         ]
        }]
      }]
    }
    L_Frame {
     id=#3
     possessed_class=CanvasPanel
     tracks [
      {
       type=MovieScene2DTransformTrack
       sections [{
         type=MovieScene2DTransformSection
         end_frame=566
         start_frame=0
         channels [
          Translation.X {type=MovieSceneScriptingFloatChannel, const [0,3]}
          Translation.Y {type=MovieSceneScriptingFloatChannel, keys [[0,0],[555.965,0],[566,-1259.4069]]}
          Angle {type=MovieSceneScriptingFloatChannel, default=0}
          Scale.X {type=MovieSceneScriptingFloatChannel, default=0.55}
          Scale.Y {type=MovieSceneScriptingFloatChannel, default=0.55}
          Shear.X {type=MovieSceneScriptingFloatChannel, default=0}
          Shear.Y {type=MovieSceneScriptingFloatChannel, default=0}
         ]
        }]
      }
      {
       type=MovieSceneFloatTrack
       display_name=Render Opacity
       sections [{
         type=MovieSceneFloatSection
         end_frame=566
         start_frame=0
         channels [None {type=MovieSceneScriptingFloatChannel, keys [[0,0],[15.053,1],[555.965,1],[566,0]]}]
        }]
      }
     ]
    }
    L_Myst_L_Hand_Controller {
     id=#4
     possessed_class=CanvasPanel
     tracks [{
       type=MovieScene2DTransformTrack
       sections [{
         type=MovieScene2DTransformSection
         end_frame=566
         start_frame=0
         channels [
          Translation.X {type=MovieSceneScriptingFloatChannel, default=0}
          Translation.Y {type=MovieSceneScriptingFloatChannel, default=0}
          Angle {type=MovieSceneScriptingFloatChannel, keys [[0,0],[566,17.8193]]}
          Scale.X {type=MovieSceneScriptingFloatChannel, default=1}
          Scale.Y {type=MovieSceneScriptingFloatChannel, default=1}
          Shear.X {type=MovieSceneScriptingFloatChannel, default=0}
          Shear.Y {type=MovieSceneScriptingFloatChannel, default=0}
         ]
        }]
      }]
    }
    BlackBar_Bottom {
     id=#5
     possessed_class=Image
     tracks [{
       type=MovieSceneFloatTrack
       display_name=Render Opacity
       sections [{type=MovieSceneFloatSection, end_frame=15, start_frame=0, channels [None {type=MovieSceneScriptingFloatChannel, keys [[0,0],[15.053,1]]}]}]
      }]
    }
    L_Myst_Controller {
     id=#6
     possessed_class=CanvasPanel
     tracks [{
       type=MovieScene2DTransformTrack
       sections [{
         type=MovieScene2DTransformSection
         end_frame=566
         start_frame=0
         channels [
          Translation.X {type=MovieSceneScriptingFloatChannel, keys [[0,-859.5015],[15.053,-859.5015],[40.142,0],[566,83.6651]]}
          Translation.Y {type=MovieSceneScriptingFloatChannel, keys [[0,1166.1135],[15.053,1166.1135],[40.142,0],[566,-78.0464]]}
          Angle {type=MovieSceneScriptingFloatChannel, keys [[0,-26.3807],[15.053,-26.3807],[40.142,0],[566,3.5728]]}
          Scale.X {type=MovieSceneScriptingFloatChannel, keys [[0,0.4],[15.053,0.4],[40.142,1],[566,1.1]]}
          Scale.Y {type=MovieSceneScriptingFloatChannel, keys [[0,0.4],[15.053,0.4],[40.142,1],[566,1.1]]}
          Shear.X {type=MovieSceneScriptingFloatChannel, default=0}
          Shear.Y {type=MovieSceneScriptingFloatChannel, default=0}
         ]
        }]
      }]
    }
    BlackBar_Top {
     id=#7
     possessed_class=Image
     tracks [{
       type=MovieSceneFloatTrack
       display_name=Render Opacity
       sections [{type=MovieSceneFloatSection, end_frame=15, start_frame=0, channels [None {type=MovieSceneScriptingFloatChannel, keys [[0,0],[15.053,1]]}]}]
      }]
    }
   ]
   display_rate {numerator=60}
  }
 }
 RightPanelAnim {
  binding_count=11
  length_seconds=9.4
  object=&self:RightPanelAnim
  sequence {
   bindings [
    R_Smoke {
     id=#8
     possessed_class=Image
     tracks [{
       type=MovieScene2DTransformTrack
       sections [{
         type=MovieScene2DTransformSection
         end_frame=564
         start_frame=0
         channels [
          Translation.X {type=MovieSceneScriptingFloatChannel, default=0, const [0,2]}
          Translation.Y {type=MovieSceneScriptingFloatChannel, default=0, keys [[0,0],[564,-27.5976]]}
          Angle {type=MovieSceneScriptingFloatChannel, default=0}
          Scale.X {type=MovieSceneScriptingFloatChannel, keys [[0,1],[564,1.05]]}
          Scale.Y {type=MovieSceneScriptingFloatChannel, keys [[0,1],[564,1.05]]}
          Shear.X {type=MovieSceneScriptingFloatChannel, default=0}
          Shear.Y {type=MovieSceneScriptingFloatChannel, default=0}
         ]
        }]
      }]
    }
    R_Myst_R_Hand_Controller {
     id=#9
     possessed_class=CanvasPanel
     tracks [{
       type=MovieScene2DTransformTrack
       sections [{
         type=MovieScene2DTransformSection
         end_frame=564
         start_frame=0
         channels [
          Translation.X {type=MovieSceneScriptingFloatChannel, default=0}
          Translation.Y {type=MovieSceneScriptingFloatChannel, default=0}
          Angle {type=MovieSceneScriptingFloatChannel, keys [[0,0],[564,-32.1323]]}
          Scale.X {type=MovieSceneScriptingFloatChannel, default=1}
          Scale.Y {type=MovieSceneScriptingFloatChannel, default=1}
          Shear.X {type=MovieSceneScriptingFloatChannel, default=0}
          Shear.Y {type=MovieSceneScriptingFloatChannel, default=0}
         ]
        }]
      }]
    }
    R_Myst_Controller {
     id=#10
     possessed_class=CanvasPanel
     tracks [{
       type=MovieScene2DTransformTrack
       sections [{
         type=MovieScene2DTransformSection
         end_frame=564
         start_frame=0
         channels [
          Translation.X {type=MovieSceneScriptingFloatChannel, keys [[0,695.9243],[15,695.9243],[30,0],[564,-55.8486]]}
          Translation.Y {type=MovieSceneScriptingFloatChannel, keys [[0,1683.7428],[15,1683.7428],[30,0],[564,-63.74]]}
          Angle {type=MovieSceneScriptingFloatChannel, keys [[0,-10.0792],[15,-10.0792],[30,0],[564,-1.1996]]}
          Scale.X {type=MovieSceneScriptingFloatChannel, keys [[0,0.8],[15,0.8],[30,1],[564,1.1]]}
          Scale.Y {type=MovieSceneScriptingFloatChannel, keys [[0,0.8],[15,0.8],[30,1],[564,1.1]]}
          Shear.X {type=MovieSceneScriptingFloatChannel, default=0}
          Shear.Y {type=MovieSceneScriptingFloatChannel, default=0}
         ]
        }]
      }]
    }
    BlackBar_Bottom {
     id=#11
     possessed_class=Image
     tracks [{
       type=MovieSceneFloatTrack
       display_name=Render Opacity
       sections [{
         type=MovieSceneFloatSection
         end_frame=483
         start_frame=0
         channels [None {type=MovieSceneScriptingFloatChannel, keys [[0,1],[473,1],[483,0]]}]
        }]
      }]
    }
    BG_Crows {
     id=#12
     possessed_class=Image
     tracks [{
       type=MovieScene2DTransformTrack
       sections [{
         type=MovieScene2DTransformSection
         end_frame=564
         start_frame=0
         channels [
          Translation.X {type=MovieSceneScriptingFloatChannel, default=0, const [0,2]}
          Translation.Y {type=MovieSceneScriptingFloatChannel, default=0, keys [[0,0],[564,-36.9725]]}
          Angle {type=MovieSceneScriptingFloatChannel, keys [[0,0],[564,10.6173]]}
          Scale.X {type=MovieSceneScriptingFloatChannel, keys [[0,1.1],[564,1.2]]}
          Scale.Y {type=MovieSceneScriptingFloatChannel, keys [[0,1.1],[564,1.2]]}
          Shear.X {type=MovieSceneScriptingFloatChannel, default=0}
          Shear.Y {type=MovieSceneScriptingFloatChannel, default=0}
         ]
        }]
      }]
    }
    L_Crow_Controller {
     id=#13
     possessed_class=CanvasPanel
     tracks [{
       type=MovieScene2DTransformTrack
       sections [{
         type=MovieScene2DTransformSection
         end_frame=564
         start_frame=0
         channels [
          Translation.X {type=MovieSceneScriptingFloatChannel, keys [[0,-791.5851],[23,-791.5851],[38,0],[564,167.005]]}
          Translation.Y {type=MovieSceneScriptingFloatChannel, keys [[0,821.7805],[23,821.7805],[38,0],[564,-129.818]]}
          Angle {type=MovieSceneScriptingFloatChannel, keys [[0,15.7863],[23,15.7863],[38,0],[564,-8.245]]}
          Scale.X {type=MovieSceneScriptingFloatChannel, keys [[0,0.7],[15,0.7],[30,0.8],[564,0.9]]}
          Scale.Y {type=MovieSceneScriptingFloatChannel, keys [[0,0.7],[15,0.7],[30,0.8],[564,0.9]]}
          Shear.X {type=MovieSceneScriptingFloatChannel, default=0}
          Shear.Y {type=MovieSceneScriptingFloatChannel, default=0}
         ]
        }]
      }]
    }
    R_Frame {
     id=#14
     possessed_class=CanvasPanel
     tracks [
      {
       type=MovieScene2DTransformTrack
       sections [{
         type=MovieScene2DTransformSection
         end_frame=564
         start_frame=0
         channels [
          Translation.X {type=MovieSceneScriptingFloatChannel, const [0,3]}
          Translation.Y {type=MovieSceneScriptingFloatChannel, keys [[0,0],[473,0],[483,1271.9932]]}
          Angle {type=MovieSceneScriptingFloatChannel, default=0}
          Scale.X {type=MovieSceneScriptingFloatChannel, default=0.55}
          Scale.Y {type=MovieSceneScriptingFloatChannel, default=0.55}
          Shear.X {type=MovieSceneScriptingFloatChannel, default=0}
          Shear.Y {type=MovieSceneScriptingFloatChannel, default=0}
         ]
        }]
      }
      {
       type=MovieSceneFloatTrack
       display_name=Render Opacity
       sections [{
         type=MovieSceneFloatSection
         end_frame=483
         start_frame=0
         channels [None {type=MovieSceneScriptingFloatChannel, keys [[0,0],[15,1],[473,1],[483,0]]}]
        }]
      }
     ]
    }
    R_BG_Controller {
     id=#15
     possessed_class=CanvasPanel
     tracks [{
       type=MovieScene2DTransformTrack
       sections [{
         type=MovieScene2DTransformSection
         end_frame=15
         start_frame=0
         channels [
          Translation.X {type=MovieSceneScriptingFloatChannel, const [0,2]}
          Translation.Y {type=MovieSceneScriptingFloatChannel, keys [[0,-2287.2463],[15,0]]}
          Angle {type=MovieSceneScriptingFloatChannel, default=0}
          Scale.X {type=MovieSceneScriptingFloatChannel, default=1}
          Scale.Y {type=MovieSceneScriptingFloatChannel, default=1}
          Shear.X {type=MovieSceneScriptingFloatChannel, default=0}
          Shear.Y {type=MovieSceneScriptingFloatChannel, default=0}
         ]
        }]
      }]
    }
    BlackBar_Top {
     id=#16
     possessed_class=Image
     tracks [{
       type=MovieSceneFloatTrack
       display_name=Render Opacity
       sections [{
         type=MovieSceneFloatSection
         end_frame=483
         start_frame=0
         channels [None {type=MovieSceneScriptingFloatChannel, keys [[0,1],[473,1],[483,0]]}]
        }]
      }]
    }
    R_Myst_L_Hand_Controller {
     id=#17
     possessed_class=CanvasPanel
     tracks [{
       type=MovieScene2DTransformTrack
       sections [{
         type=MovieScene2DTransformSection
         end_frame=564
         start_frame=0
         channels [
          Translation.X {type=MovieSceneScriptingFloatChannel, default=0}
          Translation.Y {type=MovieSceneScriptingFloatChannel, default=0}
          Angle {type=MovieSceneScriptingFloatChannel, keys [[0,0],[564,18.1579]]}
          Scale.X {type=MovieSceneScriptingFloatChannel, default=1}
          Scale.Y {type=MovieSceneScriptingFloatChannel, default=1}
          Shear.X {type=MovieSceneScriptingFloatChannel, default=0}
          Shear.Y {type=MovieSceneScriptingFloatChannel, default=0}
         ]
        }]
      }]
    }
    R_Crow_Controller {
     id=#18
     possessed_class=CanvasPanel
     tracks [{
       type=MovieScene2DTransformTrack
       sections [{
         type=MovieScene2DTransformSection
         end_frame=564
         start_frame=0
         channels [
          Translation.X {type=MovieSceneScriptingFloatChannel, keys [[0,-684.5891],[27,-684.5891],[45,0],[564,98.2496]]}
          Translation.Y {type=MovieSceneScriptingFloatChannel, keys [[0,-505.1443],[27,-505.1443],[45,0],[564,-61.1664]]}
          Angle {type=MovieSceneScriptingFloatChannel, keys [[0,31.1551],[27,31.1551],[45,0],[564,-5.2126]]}
          Scale.X {type=MovieSceneScriptingFloatChannel, keys [[0,0.5],[27,0.5],[45,0.8],[564,0.9]]}
          Scale.Y {type=MovieSceneScriptingFloatChannel, keys [[0,0.5],[27,0.5],[45,0.8],[564,0.9]]}
          Shear.X {type=MovieSceneScriptingFloatChannel, default=0}
          Shear.Y {type=MovieSceneScriptingFloatChannel, default=0}
         ]
        }]
      }]
    }
   ]
   display_rate {numerator=60}
  }
 }
]
graphs [EventGraph {nodes [
   n1 = SetRightPanelVisiblity!() -> (OutputDelegate, then, InVisibility:ESlateVisibility Enum=SelfHitTestInvisible) {is=title K2Node_CustomEvent}
   > SetVisibility(self=Get R_Frame() -> (R_Frame:Canvas Panel Object Reference), InVisibility=@n1.InVisibility)
   n2 = SetLeftPanelVisibility!() -> (OutputDelegate, then, InVisibility:ESlateVisibility Enum=SelfHitTestInvisible) {is=title K2Node_CustomEvent}
   > SetVisibility(
    self=[
     Get BlackBar_Top() -> (BlackBar_Top:Image Object Reference)
     Get BlackBar_Bottom() -> (BlackBar_Bottom:Image Object Reference)
     Get L_Frame() -> (L_Frame:Canvas Panel Object Reference)
    ]
    InVisibility=@n2.InVisibility
   )
   Event Construct()
   > SetLeftPanelVisibility!(execute, self, InVisibility:ESlateVisibility Enum=Hidden) -> (then) {is=title K2Node_CallFunction}
   > SetRightPanelVisiblity!(execute, self, InVisibility:ESlateVisibility Enum=Hidden) -> (then) {is=title K2Node_CallFunction}
   Event HeroPanel_UnloadTextures() {is=title K2Node_Event}
   n3 = Make Array!(
    "[0]":Image Object Reference=Get R_Smoke() -> (R_Smoke:Image Object Reference)
    "[1]":Image Object Reference=Get R_Myst_R_Hand() -> (R_Myst_R_Hand:Image Object Reference)
    "[2]":Image Object Reference=Get R_Myst_L_Hand() -> (R_Myst_L_Hand:Image Object Reference)
    "[3]":Image Object Reference=Get R_Myst_Body() -> (R_Myst_Body:Image Object Reference)
    "[4]":Image Object Reference=Get R_Crow() -> (R_Crow:Image Object Reference)
    "[5]":Image Object Reference=Get R_BG() -> (R_BG:Image Object Reference)
    "[6]":Image Object Reference=Get L_Myst_R_Hand() -> (L_Myst_R_Hand:Image Object Reference)
    "[7]":Image Object Reference=Get L_Myst_L_Hand() -> (L_Myst_L_Hand:Image Object Reference)
    "[8]":Image Object Reference=Get L_Myst_Body() -> (L_Myst_Body:Image Object Reference)
    "[9]":Image Object Reference=Get L_Crow() -> (L_Crow:Image Object Reference)
    "[10]":Image Object Reference=Get L_BG() -> (L_BG:Image Object Reference)
    "[11]":Image Object Reference=Get BG_Crows() -> (BG_Crows:Image Object Reference)
   ) -> (Array:Array of Image Object References)
   > n4 = For Each Loop(Array:Array of Image Object References=@n3) -> (Array Element:Image Object Reference)
   CollectGarbage!(execute=@n4.Completed) -> (then)
   > PrintString!(execute, InString=UNLOADED HERO PANEL 3 TEXTURES, bPrintToScreen, bPrintToLog, TextColor=(R=0.9931,G=0.9931,B=1,A=1), Duration=3, Key) -> (then)
   Is Valid(exec=@n4.LoopBody, InputObject=@n4)
   > SetBrushFromTexture!(execute, self:Image Object Reference=@n4, Texture:Texture 2D Object Reference, bMatchSize:bool=true) -> (then)
   Event HeroPanel_LoadTextures() {is=title K2Node_Event}
   > n5 = For Each Loop(Array=@n3)
   > n6 = Branch(Condition=IsMobilePlatform!() -> (ReturnValue:bool))
   SetBrushFromSoftTexture!(
    execute=@n6.else
    self:Image Object Reference=@n5
    SoftTexture:Texture 2D Soft Object Reference="Get (a copy)"(
     Array:Array of Texture 2D Soft Object References=Get TexturesHigh() -> (TexturesHigh:Array of Texture 2D Soft Object References)
     Dimension 1=@n5.Array Index
    ) -> (Output:Texture 2D Soft Object Reference)
    bMatchSize:bool=false
   ) -> (then)
   n7 = Event PlayPistolCanaryHeroPanelAnimation!() -> (OutputDelegate, then, HeroPanel:E_PistolCanary_HeroPanelSelect Enum=NewEnumerator0, PanelSide:E_PistolCanary_PanelSide Enum=NewEnumerator0) {is=title K2Node_Event}
   > PlayAnimation(InAnimation=Select!(
     NewEnumerator0:Widget Animation Object Reference=Get LeftPanelAnim() -> (LeftPanelAnim:Widget Animation Object Reference)
     NewEnumerator1:Widget Animation Object Reference=Get RightPanelAnim() -> (RightPanelAnim:Widget Animation Object Reference)
     NewEnumerator2:Widget Animation Object Reference
     Index:E_PistolCanary_PanelSide Enum=@n7.PanelSide
    ) -> (ReturnValue:Widget Animation Object Reference))
   > n8 = Switch on E_PistolCanary_PanelSide!(execute, Selection:E_PistolCanary_PanelSide Enum=@n7.PanelSide) -> (NewEnumerator0:Exec, NewEnumerator1:Exec, NewEnumerator2:Exec) {is=title K2Node_SwitchEnum}
   > SetLeftPanelVisibility(InVisibility=HitTestInvisible)
   SetRightPanelVisiblity(execute=@n8.NewEnumerator1, InVisibility=HitTestInvisible)
   Event PreConstruct()
   Event Tick()
  ]}]
variables [TexturesHigh {
  type="array<Texture2D>"
  default [
   /PistolCanary/Layout/ComicPanels_v2/ComicPanelTextures/Panel6/R_Frame/P6_BG_Smoke_50percent
   /PistolCanary/Layout/ComicPanels_v2/ComicPanelTextures/Panel6/R_Frame/P6_R_Myst_R_Hand
   /PistolCanary/Layout/ComicPanels_v2/ComicPanelTextures/Panel6/R_Frame/P6_R_Myst_L_Hand
   /PistolCanary/Layout/ComicPanels_v2/ComicPanelTextures/Panel6/R_Frame/P6_R_Myst_Body_50percent
   /PistolCanary/Layout/ComicPanels_v2/ComicPanelTextures/Panel6/R_Frame/P6_R_Crow_50percent
   /PistolCanary/Layout/ComicPanels_v2/ComicPanelTextures/Panel6/R_Frame/P6_R_BG_80percent
   /PistolCanary/Layout/ComicPanels_v2/ComicPanelTextures/Panel6/L_Frame/L_Myst_R_Hand
   /PistolCanary/Layout/ComicPanels_v2/ComicPanelTextures/Panel6/L_Frame/L_Myst_L_Hand
   /PistolCanary/Layout/ComicPanels_v2/ComicPanelTextures/Panel6/L_Frame/L_Myst_Body_60percent
   /PistolCanary/Layout/ComicPanels_v2/ComicPanelTextures/Panel6/R_Frame/P6_L_Crow_50percent
   /PistolCanary/Layout/ComicPanels_v2/ComicPanelTextures/Panel6/L_Frame/P6_L_BG_80percent
   /PistolCanary/Layout/ComicPanels_v2/ComicPanelTextures/Panel6/R_Frame/P6_BG_Crows_60percent
  ]
 }]
widgets {
 info {inheritedWidgetCount=0, parentClass=&UserWidget, rootWidgetClass=&CanvasPanel, widgetCount=27}
 widgets [
  Panel_06 {widgetClassPath=&CanvasPanel, widget=Panel_06, properties {clipping=ClipToBounds}}
  L_Frame {
   widgetClassPath=&CanvasPanel
   bIsVariable=true
   parent=Panel_06
   widget=L_Frame
   properties {renderTransform {angle=0, scale=<0.55 0.55>, shear=<0 0>, translation=<0 0>}}
   slot {
    class=CanvasPanelSlot
    properties {bAutoSize=true, layoutData {alignment=<0 0>, anchors {maximum=<0 0.5>, minimum=<0 0.5>}, offsets=<ltrb 0.9609 -540 100 30>}}
   }
  }
  L_BG_Controller {
   widgetClassPath=&CanvasPanel
   parent=L_Frame
   widget=L_BG_Controller
   properties {clipping=ClipToBounds}
   slot {
    class=CanvasPanelSlot
    properties {bAutoSize=true, layoutData {alignment=<0 0>, anchors {maximum=<0.5 0.5>, minimum=<0.5 0.5>}, offsets=<ltrb -1746 -2072 169.0552 176.201>}}
   }
  }
  L_BG {
   widgetClassPath=&Image
   bIsVariable=true
   parent=L_BG_Controller
   widget=L_BG
   properties {
    clipping=ClipToBounds
    visibility=HitTestInvisible
    brush {
     bIsDynamicallyLoaded=false
     drawAs=Image
     imageType=FullColor
     mirroring=NoMirror
     resourceName=None
     resourceObject=/PistolCanary/Layout/ComicPanels_v2/ComicPanelTextures/Panel6/L_Frame/LowRes/P6_L_BG_80percent_low
     tiling=NoTile
     imageSize=<192 268>
     margin=<ltrb 0 0 0 0>
     outlineSettings {
      bUseBrushTransparency=false
      roundingType=HalfHeightRadius
      width=0
      color {colorUseRule=UseColor_Specified, specifiedColor=<rgba 0 0 0 0>}
      cornerRadii=<0 0 0 0>
     }
     tintColor {colorUseRule=UseColor_Specified, specifiedColor=<rgba 1 1 1 1>}
     uVRegion {bIsValid=false, max=<0 0>, min=<0 0>}
    }
   }
   slot {class=CanvasPanelSlot, properties {layoutData {alignment=<0 0>, anchors {maximum=<0 0>, minimum=<0 0>}, offsets=<ltrb 0 0 1604 2256>}}}
  }
  L_Myst_Controller {
   widgetClassPath=&CanvasPanel
   parent=L_Frame
   widget=L_Myst_Controller
   slot {
    class=CanvasPanelSlot
    properties {layoutData {alignment=<0 0>, anchors {maximum=<0.5 0.5>, minimum=<0.5 0.5>}, offsets=<ltrb -842 -916 169.0552 176.201>}}
   }
  }
  L_Myst_Body {
   widgetClassPath=&Image
   bIsVariable=true
   parent=L_Myst_Controller
   widget=L_Myst_Body
   properties {
    clipping=ClipToBounds
    visibility=HitTestInvisible
    brush {
     bIsDynamicallyLoaded=false
     drawAs=Image
     imageType=FullColor
     mirroring=NoMirror
     resourceName=None
     resourceObject=/PistolCanary/Layout/ComicPanels_v2/ComicPanelTextures/Panel6/L_Frame/LowRes/L_Myst_Body_60percent_low
     tiling=NoTile
     imageSize=<372 504>
     margin=<ltrb 0 0 0 0>
     outlineSettings {
      bUseBrushTransparency=false
      roundingType=HalfHeightRadius
      width=0
      color {colorUseRule=UseColor_Specified, specifiedColor=<rgba 0 0 0 0>}
      cornerRadii=<0 0 0 0>
     }
     tintColor {colorUseRule=UseColor_Specified, specifiedColor=<rgba 1 1 1 1>}
     uVRegion {bIsValid=false, max=<0 0>, min=<0 0>}
    }
   }
   slot {
    class=CanvasPanelSlot
    properties {layoutData {alignment=<0 0>, anchors {maximum=<0 0>, minimum=<0 0>}, offsets=<ltrb -812 -1096 1544 2096>}}
   }
  }
  L_Myst_R_Hand_Controller {
   widgetClassPath=&CanvasPanel
   parent=L_Myst_Controller
   widget=L_Myst_R_Hand_Controller
   slot {
    class=CanvasPanelSlot
    properties {layoutData {alignment=<0 0>, anchors {maximum=<0.5 0.5>, minimum=<0.5 0.5>}, offsets=<ltrb -424.5276 -132.1005 157.9048 147.3333>}}
   }
  }
  L_Myst_R_Hand {
   widgetClassPath=&Image
   bIsVariable=true
   parent=L_Myst_R_Hand_Controller
   widget=L_Myst_R_Hand
   properties {
    clipping=ClipToBounds
    visibility=HitTestInvisible
    brush {
     bIsDynamicallyLoaded=false
     drawAs=Image
     imageType=FullColor
     mirroring=NoMirror
     resourceName=None
     resourceObject=/PistolCanary/Layout/ComicPanels_v2/ComicPanelTextures/Panel6/L_Frame/LowRes/L_Myst_R_Hand_low
     tiling=NoTile
     imageSize=<268 248>
     margin=<ltrb 0 0 0 0>
     outlineSettings {
      bUseBrushTransparency=false
      roundingType=HalfHeightRadius
      width=0
      color {colorUseRule=UseColor_Specified, specifiedColor=<rgba 0 0 0 0>}
      cornerRadii=<0 0 0 0>
     }
     tintColor {colorUseRule=UseColor_Specified, specifiedColor=<rgba 1 1 1 1>}
     uVRegion {bIsValid=false, max=<0 0>, min=<0 0>}
    }
   }
   slot {class=CanvasPanelSlot, properties {layoutData {alignment=<0 0>, anchors {maximum=<0 0>, minimum=<0 0>}, offsets=<ltrb -172 -160 444 412>}}}
  }
  L_Myst_L_Hand_Controller {
   widgetClassPath=&CanvasPanel
   parent=L_Myst_Controller
   widget=L_Myst_L_Hand_Controller
   slot {
    class=CanvasPanelSlot
    properties {layoutData {alignment=<0 0>, anchors {maximum=<0.5 0.5>, minimum=<0.5 0.5>}, offsets=<ltrb 175.4724 -864.1005 157.9048 147.3333>}}
   }
  }
  L_Myst_L_Hand {
   widgetClassPath=&Image
   bIsVariable=true
   parent=L_Myst_L_Hand_Controller
   widget=L_Myst_L_Hand
   properties {
    clipping=ClipToBounds
    visibility=HitTestInvisible
    brush {
     bIsDynamicallyLoaded=false
     drawAs=Image
     imageType=FullColor
     mirroring=NoMirror
     resourceName=None
     resourceObject=/PistolCanary/Layout/ComicPanels_v2/ComicPanelTextures/Panel6/L_Frame/LowRes/L_Myst_L_Hand_low
     tiling=NoTile
     imageSize=<292 260>
     margin=<ltrb 0 0 0 0>
     outlineSettings {
      bUseBrushTransparency=false
      roundingType=HalfHeightRadius
      width=0
      color {colorUseRule=UseColor_Specified, specifiedColor=<rgba 0 0 0 0>}
      cornerRadii=<0 0 0 0>
     }
     tintColor {colorUseRule=UseColor_Specified, specifiedColor=<rgba 1 1 1 1>}
     uVRegion {bIsValid=false, max=<0 0>, min=<0 0>}
    }
   }
   slot {class=CanvasPanelSlot, properties {layoutData {alignment=<0 0>, anchors {maximum=<0 0>, minimum=<0 0>}, offsets=<ltrb -92 -48 292 260>}}}
  }
  R_Frame {
   widgetClassPath=&CanvasPanel
   bIsVariable=true
   parent=Panel_06
   widget=R_Frame
   properties {renderTransform {angle=0, scale=<0.55 0.55>, shear=<0 0>, translation=<0 0>}}
   slot {
    class=CanvasPanelSlot
    properties {bAutoSize=true, layoutData {alignment=<0 0>, anchors {maximum=<1 0.5>, minimum=<1 0.5>}, offsets=<ltrb -1920.9609 -540 100 30>}}
   }
  }
  R_BG_Controller {
   widgetClassPath=&CanvasPanel
   parent=R_Frame
   widget=R_BG_Controller
   properties {clipping=ClipToBounds}
   slot {
    class=CanvasPanelSlot
    properties {bAutoSize=true, layoutData {alignment=<0 0>, anchors {maximum=<0 0>, minimum=<0 0>}, offsets=<ltrb 972 -2004 100 30>}}
   }
  }
  R_BG {
   widgetClassPath=&Image
   bIsVariable=true
   parent=R_BG_Controller
   widget=R_BG
   properties {
    clipping=ClipToBounds
    visibility=HitTestInvisible
    brush {
     bIsDynamicallyLoaded=false
     drawAs=Image
     imageType=FullColor
     mirroring=NoMirror
     resourceName=None
     resourceObject=/PistolCanary/Layout/ComicPanels_v2/ComicPanelTextures/Panel6/R_Frame/LowRes/P6_R_BG_80percent_low
     tiling=NoTile
     imageSize=<212 268>
     margin=<ltrb 0 0 0 0>
     outlineSettings {
      bUseBrushTransparency=false
      roundingType=HalfHeightRadius
      width=0
      color {colorUseRule=UseColor_Specified, specifiedColor=<rgba 0 0 0 0>}
      cornerRadii=<0 0 0 0>
     }
     tintColor {colorUseRule=UseColor_Specified, specifiedColor=<rgba 1 1 1 1>}
     uVRegion {bIsValid=false, max=<0 0>, min=<0 0>}
    }
   }
   slot {class=CanvasPanelSlot, properties {layoutData {alignment=<0 0>, anchors {maximum=<0 0>, minimum=<0 0>}, offsets=<ltrb 0 0 1776 2256>}}}
  }
  BG_Crows {
   widgetClassPath=&Image
   bIsVariable=true
   parent=R_BG_Controller
   widget=BG_Crows
   properties {
    clipping=ClipToBounds
    visibility=HitTestInvisible
    brush {
     bIsDynamicallyLoaded=false
     drawAs=Image
     imageType=FullColor
     mirroring=NoMirror
     resourceName=None
     resourceObject=/PistolCanary/Layout/ComicPanels_v2/ComicPanelTextures/Panel6/R_Frame/LowRes/P6_BG_Crows_60percent_low
     tiling=NoTile
     imageSize=<260 292>
     margin=<ltrb 0 0 0 0>
     outlineSettings {
      bUseBrushTransparency=false
      roundingType=HalfHeightRadius
      width=0
      color {colorUseRule=UseColor_Specified, specifiedColor=<rgba 0 0 0 0>}
      cornerRadii=<0 0 0 0>
     }
     tintColor {colorUseRule=UseColor_Specified, specifiedColor=<rgba 1 1 1 1>}
     uVRegion {bIsValid=false, max=<0 0>, min=<0 0>}
    }
    renderTransform {angle=0, scale=<1.1 1.1>, shear=<0 0>, translation=<0 0>}
   }
   slot {class=CanvasPanelSlot, properties {layoutData {alignment=<0 0>, anchors {maximum=<0 0>, minimum=<0 0>}, offsets=<ltrb 176 420 1076 1208>}}}
  }
  R_Smoke {
   widgetClassPath=&Image
   bIsVariable=true
   parent=R_BG_Controller
   widget=R_Smoke
   properties {
    clipping=ClipToBounds
    visibility=HitTestInvisible
    brush {
     bIsDynamicallyLoaded=false
     drawAs=Image
     imageType=FullColor
     mirroring=NoMirror
     resourceName=None
     resourceObject=/PistolCanary/Layout/ComicPanels_v2/ComicPanelTextures/Panel6/R_Frame/LowRes/P6_BG_Smoke_50percent_low
     tiling=NoTile
     imageSize=<324 336>
     margin=<ltrb 0 0 0 0>
     outlineSettings {
      bUseBrushTransparency=false
      roundingType=HalfHeightRadius
      width=0
      color {colorUseRule=UseColor_Specified, specifiedColor=<rgba 0 0 0 0>}
      cornerRadii=<0 0 0 0>
     }
     tintColor {colorUseRule=UseColor_Specified, specifiedColor=<rgba 1 1 1 1>}
     uVRegion {bIsValid=false, max=<0 0>, min=<0 0>}
    }
   }
   slot {class=CanvasPanelSlot, properties {layoutData {alignment=<0 0>, anchors {maximum=<0 0>, minimum=<0 0>}, offsets=<ltrb 592 1212 1092 1120>}}}
  }
  L_Crow_Controller {
   widgetClassPath=&CanvasPanel
   parent=R_Frame
   widget=L_Crow_Controller
   properties {renderTransform {angle=0, scale=<0.8 0.8>, shear=<0 0>, translation=<0 0>}}
   slot {
    class=CanvasPanelSlot
    properties {layoutData {alignment=<0 0>, anchors {maximum=<0 0>, minimum=<0 0>}, offsets=<ltrb 1276 -112 90.0952 88.6667>}}
   }
  }
  L_Crow {
   widgetClassPath=&Image
   bIsVariable=true
   parent=L_Crow_Controller
   widget=L_Crow
   properties {
    clipping=ClipToBounds
    visibility=HitTestInvisible
    brush {
     bIsDynamicallyLoaded=false
     drawAs=Image
     imageType=FullColor
     mirroring=NoMirror
     resourceName=None
     resourceObject=/PistolCanary/Layout/ComicPanels_v2/ComicPanelTextures/Panel6/R_Frame/LowRes/P6_L_Crow_50percent_low
     tiling=NoTile
     imageSize=<292 252>
     margin=<ltrb 0 0 0 0>
     outlineSettings {
      bUseBrushTransparency=false
      roundingType=HalfHeightRadius
      width=0
      color {colorUseRule=UseColor_Specified, specifiedColor=<rgba 0 0 0 0>}
      cornerRadii=<0 0 0 0>
     }
     tintColor {colorUseRule=UseColor_Specified, specifiedColor=<rgba 1 1 1 1>}
     uVRegion {bIsValid=false, max=<0 0>, min=<0 0>}
    }
   }
   slot {class=CanvasPanelSlot, properties {layoutData {alignment=<0 0>, anchors {maximum=<0 0>, minimum=<0 0>}, offsets=<ltrb -368 -588 968 840>}}}
  }
  R_Crow_Controller {
   widgetClassPath=&CanvasPanel
   parent=R_Frame
   widget=R_Crow_Controller
   properties {renderTransform {angle=0, scale=<0.8 0.8>, shear=<0 0>, translation=<0 0>}}
   slot {
    class=CanvasPanelSlot
    properties {layoutData {alignment=<0 0>, anchors {maximum=<0 0>, minimum=<0 0>}, offsets=<ltrb 1940 -1540 102.0317 90.7831>}}
   }
  }
  R_Crow {
   widgetClassPath=&Image
   bIsVariable=true
   parent=R_Crow_Controller
   widget=R_Crow
   properties {
    clipping=ClipToBounds
    visibility=HitTestInvisible
    brush {
     bIsDynamicallyLoaded=false
     drawAs=Image
     imageType=FullColor
     mirroring=NoMirror
     resourceName=None
     resourceObject=/PistolCanary/Layout/ComicPanels_v2/ComicPanelTextures/Panel6/R_Frame/LowRes/P6_R_Crow_50percent_low
     tiling=NoTile
     imageSize=<408 256>
     margin=<ltrb 0 0 0 0>
     outlineSettings {
      bUseBrushTransparency=false
      roundingType=HalfHeightRadius
      width=0
      color {colorUseRule=UseColor_Specified, specifiedColor=<rgba 0 0 0 0>}
      cornerRadii=<0 0 0 0>
     }
     tintColor {colorUseRule=UseColor_Specified, specifiedColor=<rgba 1 1 1 1>}
     uVRegion {bIsValid=false, max=<0 0>, min=<0 0>}
    }
   }
   slot {class=CanvasPanelSlot, properties {layoutData {alignment=<0 0>, anchors {maximum=<0 0>, minimum=<0 0>}, offsets=<ltrb -740 -540 1356 848>}}}
  }
  R_Myst_Controller {
   widgetClassPath=&CanvasPanel
   parent=R_Frame
   widget=R_Myst_Controller
   slot {
    class=CanvasPanelSlot
    properties {layoutData {alignment=<0 0>, anchors {maximum=<0 0>, minimum=<0 0>}, offsets=<ltrb 1688 -872 102.0529 94.8677>}}
   }
  }
  R_Myst_Body {
   widgetClassPath=&Image
   bIsVariable=true
   parent=R_Myst_Controller
   widget=R_Myst_Body
   properties {
    clipping=ClipToBounds
    visibility=HitTestInvisible
    brush {
     bIsDynamicallyLoaded=false
     drawAs=Image
     imageType=FullColor
     mirroring=NoMirror
     resourceName=None
     resourceObject=/PistolCanary/Layout/ComicPanels_v2/ComicPanelTextures/Panel6/R_Frame/LowRes/P6_R_Myst_Body_50percent_low
     tiling=NoTile
     imageSize=<184 528>
     margin=<ltrb 0 0 0 0>
     outlineSettings {
      bUseBrushTransparency=false
      roundingType=HalfHeightRadius
      width=0
      color {colorUseRule=UseColor_Specified, specifiedColor=<rgba 0 0 0 0>}
      cornerRadii=<0 0 0 0>
     }
     tintColor {colorUseRule=UseColor_Specified, specifiedColor=<rgba 1 1 1 1>}
     uVRegion {bIsValid=false, max=<0 0>, min=<0 0>}
    }
   }
   slot {class=CanvasPanelSlot, properties {layoutData {alignment=<0 0>, anchors {maximum=<0 0>, minimum=<0 0>}, offsets=<ltrb -236 -724 616 1760>}}}
  }
  R_Myst_R_Hand_Controller {
   widgetClassPath=&CanvasPanel
   parent=R_Myst_Controller
   widget=R_Myst_R_Hand_Controller
   slot {
    class=CanvasPanelSlot
    properties {layoutData {alignment=<0 0>, anchors {maximum=<0 0>, minimum=<0 0>}, offsets=<ltrb -252 -20 99.1429 89.4286>}}
   }
  }
  R_Myst_R_Hand {
   widgetClassPath=&Image
   bIsVariable=true
   parent=R_Myst_R_Hand_Controller
   widget=R_Myst_R_Hand
   properties {
    clipping=ClipToBounds
    visibility=HitTestInvisible
    brush {
     bIsDynamicallyLoaded=false
     drawAs=Image
     imageType=FullColor
     mirroring=NoMirror
     resourceName=None
     resourceObject=/PistolCanary/Layout/ComicPanels_v2/ComicPanelTextures/Panel6/R_Frame/LowRes/P6_R_Myst_R_Hand_low
     tiling=NoTile
     imageSize=<112 76>
     margin=<ltrb 0 0 0 0>
     outlineSettings {
      bUseBrushTransparency=false
      roundingType=HalfHeightRadius
      width=0
      color {colorUseRule=UseColor_Specified, specifiedColor=<rgba 0 0 0 0>}
      cornerRadii=<0 0 0 0>
     }
     tintColor {colorUseRule=UseColor_Specified, specifiedColor=<rgba 1 1 1 1>}
     uVRegion {bIsValid=false, max=<0 0>, min=<0 0>}
    }
   }
   slot {class=CanvasPanelSlot, properties {layoutData {alignment=<0 0>, anchors {maximum=<0 0>, minimum=<0 0>}, offsets=<ltrb -48 16 112 76>}}}
  }
  R_Myst_L_Hand_Controller {
   widgetClassPath=&CanvasPanel
   parent=R_Myst_Controller
   widget=R_Myst_L_Hand_Controller
   slot {
    class=CanvasPanelSlot
    properties {layoutData {alignment=<0 0>, anchors {maximum=<0 0>, minimum=<0 0>}, offsets=<ltrb 180 -220 99.1429 89.4286>}}
   }
  }
  R_Myst_L_Hand {
   widgetClassPath=&Image
   bIsVariable=true
   parent=R_Myst_L_Hand_Controller
   widget=R_Myst_L_Hand
   properties {
    clipping=ClipToBounds
    visibility=HitTestInvisible
    brush {
     bIsDynamicallyLoaded=false
     drawAs=Image
     imageType=FullColor
     mirroring=NoMirror
     resourceName=None
     resourceObject=/PistolCanary/Layout/ComicPanels_v2/ComicPanelTextures/Panel6/R_Frame/LowRes/P6_R_Myst_L_Hand_low
     tiling=NoTile
     imageSize=<136 112>
     margin=<ltrb 0 0 0 0>
     outlineSettings {
      bUseBrushTransparency=false
      roundingType=HalfHeightRadius
      width=0
      color {colorUseRule=UseColor_Specified, specifiedColor=<rgba 0 0 0 0>}
      cornerRadii=<0 0 0 0>
     }
     tintColor {colorUseRule=UseColor_Specified, specifiedColor=<rgba 1 1 1 1>}
     uVRegion {bIsValid=false, max=<0 0>, min=<0 0>}
    }
   }
   slot {
    class=CanvasPanelSlot
    properties {bAutoSize=true, layoutData {alignment=<0 0>, anchors {maximum=<0 0>, minimum=<0 0>}, offsets=<ltrb -20 -24 136 112>}}
   }
  }
  BlackBar_Top {
   widgetClassPath=&Image
   bIsVariable=true
   parent=Panel_06
   widget=BlackBar_Top
   properties {brush {
     bIsDynamicallyLoaded=false
     drawAs=Image
     imageType=NoImage
     mirroring=NoMirror
     resourceName=None
     resourceObject=null
     tiling=NoTile
     imageSize=<32 32>
     margin=<ltrb 0 0 0 0>
     outlineSettings {
      bUseBrushTransparency=false
      roundingType=HalfHeightRadius
      width=0
      color {colorUseRule=UseColor_Specified, specifiedColor=<rgba 0 0 0 0>}
      cornerRadii=<0 0 0 0>
     }
     tintColor {colorUseRule=UseColor_Specified, specifiedColor=<rgba 0 0 0 1>}
     uVRegion {bIsValid=false, max=<0 0>, min=<0 0>}
    }}
   slot {
    class=CanvasPanelSlot
    properties {
     zOrder=10
     layoutData {alignment=<0 0>, anchors {maximum=<0.5 0.5>, minimum=<0.5 0.5>}, offsets=<ltrb -1202.0753 -931.8829 2278.6777 393.113>}
    }
   }
  }
  BlackBar_Bottom {
   widgetClassPath=&Image
   bIsVariable=true
   parent=Panel_06
   widget=BlackBar_Bottom
   properties {brush {
     bIsDynamicallyLoaded=false
     drawAs=Image
     imageType=NoImage
     mirroring=NoMirror
     resourceName=None
     resourceObject=null
     tiling=NoTile
     imageSize=<32 32>
     margin=<ltrb 0 0 0 0>
     outlineSettings {
      bUseBrushTransparency=false
      roundingType=HalfHeightRadius
      width=0
      color {colorUseRule=UseColor_Specified, specifiedColor=<rgba 0 0 0 0>}
      cornerRadii=<0 0 0 0>
     }
     tintColor {colorUseRule=UseColor_Specified, specifiedColor=<rgba 0 0 0 1>}
     uVRegion {bIsValid=false, max=<0 0>, min=<0 0>}
    }}
   slot {
    class=CanvasPanelSlot
    properties {zOrder=12, layoutData {alignment=<0 0>, anchors {maximum=<0.5 0.5>, minimum=<0.5 0.5>}, offsets=<ltrb -1154.2546 539.03 2278.6777 393.113>}}
   }
  }
 ]
}
