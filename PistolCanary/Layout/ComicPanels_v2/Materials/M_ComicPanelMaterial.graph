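# M_ComicPanelMaterial is a translucent, unlit UI-domain material for comic panels in the layout system.
# Its graph holds a single texture parameter named Texture, which defaults to a white placeholder texture and can be overridden by material instances.
# The graph shows no other nodes and no connection from the sample to a material output.
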Material /PistolCanary/Layout/ComicPanels_v2/Materials/M_ComicPanelMaterial
graph {
 properties {
  bCastRayTracedShadows=false
  bOutputTranslucentVelocity=true
  bUseTranslucencyVertexFog=false
  blendMode=BLEND_Translucent
  materialDomain=MD_UI
  shadingModel=MSM_Unlit
  usedShadingModels=Unlit
 }
 nodes [TextureSampleParameter2D() {parameterName=Texture, texture=/PistolCanary/Assets/Textures/default_white}]
}
