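# M_FX_AlphaBlend is an unlit translucent particle material for Niagara sprites in the PistolCanary comic panel layout.
# It multiplies particle color alpha by the RGB of a Texture parameter, using the texture color rather than its alpha as the mask.
# The product feeds a DepthFade with an adjustable distance to soften intersections with scene geometry, then is saturated.
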Material /PistolCanary/Layout/ComicPanels_v2/Materials/M_FX_AlphaBlend
graph {
 properties {bUsedWithNiagaraSprites=true, blendMode=BLEND_Translucent, shadingModel=MSM_Unlit, usedShadingModels=Unlit}
 nodes [
  n1 = ParticleColor()
  n2 = TextureSampleParameter2D() {parameterName=Texture}
  Saturate(None=DepthFade(Opacity=Multiply(A=@n1.A, B=@n2.RGB), FadeDistance=ScalarParameter() {defaultValue=50, parameterName=Depth Fade Distance}))
 ]
}
