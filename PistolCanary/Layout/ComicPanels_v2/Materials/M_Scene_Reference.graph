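# M_Scene_Reference is a translucent, unlit, two-sided material, apparently for showing a see-through scene reference image in comic panel layout.
# A Reference texture parameter defaults to the Scene_01_Animated_Reference screenshot.
# Its alpha is multiplied by an Opacity scalar parameter, half strength by default, so both image and transparency can be overridden in material instances.
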Material /PistolCanary/Layout/ComicPanels_v2/Materials/M_Scene_Reference
graph {
 properties {blendMode=BLEND_Translucent, shadingModel=MSM_Unlit, twoSided=true, usedShadingModels=Unlit}
 nodes [Multiply(
   A=TextureSampleParameter2D() {parameterName=Reference, texture=/PistolCanary/Layout/ComicPanels_v2/Screenshots/Scene_01_Animated_Reference}.A
   B=ScalarParameter() {defaultValue=0.5, parameterName=Opacity}
  )]
}
