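# Unlit, masked, two-sided material for Niagara sprite VFX in the ComicPanels_v2 layout, with ray-traced shadow casting turned off.
# Its only node is a Texture parameter defaulting to a comic "Bang" image, so material instances can swap in other sprites.
# The graph shows no math or links from that sample to material outputs.
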Material /PistolCanary/Layout/ComicPanels_v2/Materials/M_VFX_Sprite
graph {
 properties {
  bCastRayTracedShadows=false
  bUsedWithNiagaraSprites=true
  blendMode=BLEND_Masked
  shadingModel=MSM_Unlit
  twoSided=true
  usedShadingModels=Unlit
 }
 nodes [TextureSampleParameter2D() {parameterName=Texture, texture=/PistolCanary/Layout/ComicPanels_v2/Materials/Textures/Comic_Bang_v01}]
}
