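# Unlit material for the background fader of Panel 1 in the comic panel layout.
# Its only node reads the Color parameter from the MPC_ComicIntro parameter collection, so that value is controlled outside the material.
# The graph has no textures or math, and automatic usage-flag setting in the editor is turned off.
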Material /PistolCanary/Layout/ComicPanels_v2/Materials/Panel1_BG_Fader
graph {
 properties {bAutomaticallySetUsageInEditor=false, shadingModel=MSM_Unlit, usedShadingModels=Unlit}
 nodes [CollectionParameter() {collection=&/MPC_ComicIntro, parameterName=Color}]
}
