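# Niagara system for a short comic-style hit effect, built from four one-shot burst sprite emitters with no user parameters.
# Bright flash and comic bang sprites pop up and shrink on a size curve while fading out.
# Sparks launch in a cone, slow with drag, shrink, fade and stretch with speed; their gravity and collision modules are disabled.
# Short-lived elongated speed lines shoot from a cone at random speeds and fade.
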NiagaraSystem /PistolCanary/Layout/ComicPanels_v2/VFX/NE_Hit_FX
dependencies {usedDataInterfaces [&NiagaraDataInterfaceCurve], usedDynamicInputs [], usedModules [&SystemState]}
emitters {
 NE_Bright_Flash {
  emitterUpdateScript [
   &EmitterState()
   &SpawnBurst_Instantaneous(Use Spawn Probability=false, UseLoopCountLimit=false)
  ]
  particleSpawnScript [
   &InitializeParticle(Lifetime=0.2, Sprite Size Mode=::NewEnumerator3, Uniform Sprite Size=150, UsePositionOffset=false, Write Material Random=true)
   &ShapeLocation(MaskSpawnGroup=false, Offset Mode=::NewEnumerator0, Override Seed=false, Sphere Radius=0, UseOwnerScale=true)
  ]
  particleUpdateScript [
   &ParticleState()
   &ScaleSpriteSize(Uniform Curve Sprite Scale=DataInterface() {propertyValues {Curve {
       defaultValue=0
       keys [
        {
         arriveTangent=6.7161
         arriveTangentWeight=0.3244
         interpMode=RCIM_Cubic
         leaveTangent=6.7161
         leaveTangentWeight=0.3244
         tangentMode=RCTM_User
         tangentWeightMode=RCTWM_WeightedBoth
         time=0
         value=0
        }
        {interpMode=RCIM_Cubic, tangentMode=RCTM_User, time=0.3, value=1}
        {interpMode=RCIM_Cubic, tangentMode=RCTM_User, time=1, value=0}
       ]
      }}})
   &ScaleColor(
    Scale Alpha=&FloatFromCurve(FloatCurve=DataInterface() {propertyValues {Curve {keys [<at -0.0046 1.0113 RCIM_Cubic>,<at 1.0011 0.002 RCIM_Cubic>]}}})
    ScaleA=true
    ScaleRGB=false
   )
   &SolveForcesAndVelocity(Clamp Velocity=false, Limit Acceleration=false)
  ]
  rendererClasses [&NiagaraSpriteRendererProperties]
  renderers [{rendererClass=&NiagaraSpriteRendererProperties}]
 }
 NE_Comic_Bang {
  emitterUpdateScript [
   &EmitterState()
   &SpawnBurst_Instantaneous(Use Spawn Probability=false, UseLoopCountLimit=false)
  ]
  particleSpawnScript [
   &InitializeParticle(
    Lifetime=0.2
    Sprite Rotation Mode=::NewEnumerator2
    Sprite Size Max=<150 150>
    Sprite Size Min=<130 130>
    Sprite Size Mode=::NewEnumerator2
    UsePositionOffset=false
    Write Material Random=true
   )
   &ShapeLocation(MaskSpawnGroup=false, Offset Mode=::NewEnumerator0, Override Seed=false, Sphere Radius=0, UseOwnerScale=true)
  ]
  particleUpdateScript [
   &ParticleState()
   &ScaleSpriteSize(Uniform Curve Sprite Scale=DataInterface() {propertyValues {Curve {
       defaultValue=0
       keys [
        {
         arriveTangent=6.7161
         arriveTangentWeight=0.3244
         interpMode=RCIM_Cubic
         leaveTangent=6.7161
         leaveTangentWeight=0.3244
         tangentMode=RCTM_User
         tangentWeightMode=RCTWM_WeightedBoth
         time=0
         value=0
        }
        {interpMode=RCIM_Cubic, tangentMode=RCTM_User, time=0.3, value=1}
        {interpMode=RCIM_Cubic, tangentMode=RCTM_User, time=1, value=0}
       ]
      }}})
   &ScaleColor(
    Scale Alpha=&FloatFromCurve(FloatCurve=DataInterface() {propertyValues {Curve {keys [<at -0.0046 1.0113 RCIM_Cubic>,<at 1.0011 0.002 RCIM_Cubic>]}}})
    ScaleA=true
    ScaleRGB=false
   )
   &SolveForcesAndVelocity(Clamp Velocity=false, Limit Acceleration=false)
  ]
  rendererClasses [&NiagaraSpriteRendererProperties]
  renderers [{rendererClass=&NiagaraSpriteRendererProperties}]
 }
 NE_Sparks {
  emitterUpdateScript [
   &EmitterState()
   &SpawnBurst_Instantaneous(Spawn Count=15, Use Spawn Probability=false, UseLoopCountLimit=false)
  ]
  particleSpawnScript [
   &InitializeParticle(
    Lifetime Max=0.6
    Lifetime Min=0.4
    Lifetime Mode=::NewEnumerator1
    Sprite Rotation Mode=::NewEnumerator2
    Sprite Size Mode=::NewEnumerator1
    Uniform Sprite Size Max=1
    Uniform Sprite Size Min=0.5
    UsePositionOffset=false
    Write Material Random=true
   )
   &ShapeLocation(Cone Angle=120, MaskSpawnGroup=false, Override Seed=false, Shape Primitive=::NewEnumerator5, UseOwnerScale=true)
   &AddVelocity(
    Module.Use Falloff Distance=false
    Override Seed=false
    Velocity Mode=::NewEnumerator1
    Velocity Speed=&RandomRangeFloat(Minimum=100, Maximum=200)
   )
  ]
  particleUpdateScript [
   &ParticleState()
   &GravityForce(Gravity=<0 0 -400>) {enabled=false}
   &Drag(Use Linear Drag=true, Use Rotational Drag=false)
   &ScaleSpriteSize(Scale Factor=&Vector2DFromFloat(Value=&FloatFromCurve(FloatCurve=DataInterface() {propertyValues {Curve {
         defaultValue=1.0113
         keys [{interpMode=RCIM_Cubic, tangentMode=RCTM_User, time=0, value=1}, {interpMode=RCIM_Cubic, tangentMode=RCTM_User, time=1, value=0}]
        }}})))
   &ScaleColor(
    Scale Alpha=&FloatFromCurve(FloatCurve=DataInterface() {propertyValues {Curve {keys [<at -0.0046 1.0113 RCIM_Cubic>,<at 1.0011 0.002 RCIM_Cubic>]}}})
    ScaleA=true
    ScaleRGB=true
   )
   &Collision(
    Age Colliding Particles=false
    KillOnCollision=false
    Module.Control Rest State Via Surface Normal=false
    Module.EnableMaxCollisionCount=false
    Module.Randomize Collision Normal Vector=false
   ) {enabled=false}
   &SolveForcesAndVelocity(Clamp Velocity=false, Limit Acceleration=false)
   &ScaleSpriteSizeBySpeed(Max Scale Factor=<1 20>, Min Scale Factor=<1 2>)
  ]
  rendererClasses [&NiagaraSpriteRendererProperties]
  renderers [{rendererClass=&NiagaraSpriteRendererProperties}]
 }
 NE_Speed_Lines {
  emitterUpdateScript [
   &EmitterState()
   &SpawnBurst_Instantaneous(Spawn Count=7, Use Spawn Probability=false, UseLoopCountLimit=false)
  ]
  particleSpawnScript [
   &InitializeParticle(
    Lifetime Max=0.2
    Lifetime Min=0.1
    Lifetime Mode=::NewEnumerator1
    Sprite Rotation Mode=::NewEnumerator2
    Sprite Size=<2 50>
    Sprite Size Mode=::NewEnumerator4
    UsePositionOffset=false
    Write Material Random=true
   )
   &ShapeLocation(
    Cone Angle=120
    Cone Inner Angle=120
    Cone Length=7
    Flatten Endcaps=true
    MaskSpawnGroup=false
    Override Seed=false
    Shape Primitive=::NewEnumerator5
    UseOwnerScale=true
   )
   &AddVelocity(
    Module.Use Falloff Distance=false
    Override Seed=false
    Velocity Mode=::NewEnumerator1
    Velocity Speed=&RandomRangeFloat(Minimum=250, Maximum=500)
   )
  ]
  particleUpdateScript [
   &ParticleState()
   &ScaleColor(
    Scale Alpha=&FloatFromCurve(FloatCurve=DataInterface() {propertyValues {Curve {keys [<at -0.0046 1.0113 RCIM_Cubic>,<at 1.0011 0.002 RCIM_Cubic>]}}})
    ScaleA=true
    ScaleRGB=true
   )
   &SolveForcesAndVelocity(Clamp Velocity=false, Limit Acceleration=false)
  ]
  rendererClasses [&NiagaraSpriteRendererProperties]
  renderers [{rendererClass=&NiagaraSpriteRendererProperties}]
 }
}
summary {
 systemName=NE_Hit_FX
 emitters [
  {emitterName=NE_Bright_Flash, rendererClasses [&NiagaraSpriteRendererProperties]}
  {emitterName=NE_Sparks, rendererClasses [&NiagaraSpriteRendererProperties]}
  {emitterName=NE_Speed_Lines, rendererClasses [&NiagaraSpriteRendererProperties]}
  {emitterName=NE_Comic_Bang, rendererClasses [&NiagaraSpriteRendererProperties]}
 ]
 userVariables []
}
